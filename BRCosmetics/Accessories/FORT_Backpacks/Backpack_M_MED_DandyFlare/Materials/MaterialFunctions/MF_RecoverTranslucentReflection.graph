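# Material function for the DandyFlare backpack that computes Base Color, Metallic, Specular and Opacity so translucent, water-like surfaces keep their reflections at low opacity.
# An environment BRDF evaluates reflectance from F0, roughness and view angle, then its terms are inverted to derive an unpremultiplied F0.
# That F0 splits into specular and metallic reflectance, while Surface Opacity Mask fades toward opaque diffuse.
# Minimum volume transmittance reads only red and blue, skipping green.
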MaterialFunction /BRCosmetics/Accessories/FORT_Backpacks/Backpack_M_MED_DandyFlare/Materials/MaterialFunctions/MF_RecoverTranslucentReflection
graph {nodes [
  FunctionOutput(None=NamedRerouteUsage()) {outputName="Surface Reflectance (S)", sortPriority=50}
  FunctionOutput(None=Constant()) {outputName=----------------, sortPriority=40}
  FunctionOutput(None=NamedRerouteUsage()) {outputName="Opacity (S)", sortPriority=30}
  FunctionOutput(None=NamedRerouteUsage()) {description=Enables metallic reflectance where diffuse mask is 0, outputName="Metallic (S)", sortPriority=10}
  FunctionOutput(None=NamedRerouteUsage()) {description=Unpremultiplied specular, outputName="Specular (S)", sortPriority=20}
  FunctionOutput(None=NamedRerouteUsage()) {description="Unpremultiplied metallic reflectance where diffuse mask is 0, diffuse color otherwise", outputName="Base Color (V3)"}
  n1 = FunctionInput() {
   bUsePreviewValueAsDefault=true
   description="Diffuse opacity of the water surface; 0 : full visibility of the underwater scene + water volume scattering; 1 : the water surface is a fully opaque lambertian reflector; does not affect surface specular/reflection (always visible)"
   inputName=Surface Opacity Mask
   inputType=FunctionInput_Scalar
   sortPriority=60
  }
  NamedRerouteDeclaration(None=@n1) {name=Surface Opacity Mask, nodeColor=<rgba 0.4297 0 1 1>}
  n2 = MaterialFunctionCall(Float3=NamedRerouteUsage()) {materialFunction=&BreakOutFloat3Components}
  NamedRerouteDeclaration(None=Min(A=Min(A=@n2.R, B=@n2.R), B=@n2.B)) {name="Min Volume Transmittance (S)", nodeColor=<rgba 0 0.5913 1 1>}
  n3 = MaterialFunctionCall!(NdotV=NamedRerouteUsage(), Roughness=NamedRerouteUsage(), F0=NamedRerouteUsage(), F90=NamedRerouteUsage()) {materialFunction=/Game/Global/MaterialFunctions/TechArt/Refraction/MF_EnvironmentBRDF}
  NamedRerouteDeclaration(None=@n3."F0 Integral (S)") {name="EnvBRDF A (S)", nodeColor=<rgba 0 1 0.2969 1>}
  NamedRerouteDeclaration(None=@n3."F90 Integral (S)") {name="EnvBRDF B (S)", nodeColor=<rgba 0 0.313 1 1>}
  NamedRerouteDeclaration(None=@n3."Reflectance (S)") {name="Surface Reflectance (S)", nodeColor=<rgba 1 0.0234 0 1>}
  n4 = FunctionInput() {inputName=Dielectric Opacity, inputType=FunctionInput_Scalar, sortPriority=40, previewValue=<0.04 0 0 1>}
  NamedRerouteDeclaration(None=Saturate(None=@n4)) {name="Opacity In (S)", nodeColor=<rgba 0 0.1739 1 1>}
  n5 = FunctionInput() {bUsePreviewValueAsDefault=true, inputName=Volume Transmittance, sortPriority=70, previewValue=<1 1 1 1>}
  NamedRerouteDeclaration(None=@n5) {name="Volume Transmittance (V3)", nodeColor=<rgba 0 1 0.6016 1>}
  n6 = Divide(A=NamedRerouteUsage(), B=Max(A=NamedRerouteUsage()) {constB=0.0001})
  n7 = NamedRerouteUsage()
  NamedRerouteDeclaration(None=Divide(A=Subtract(A=@n6, B=@n7), B=NamedRerouteUsage())) {name="F0 Dielectric (S)", nodeColor=<rgba 0 0.0625 1 1>}
  n8 = FunctionInput() {inputName=Roughness, inputType=FunctionInput_Scalar, sortPriority=20, previewValue=<0.2 0 0 1>}
  NamedRerouteDeclaration(None=@n8) {name="Roughness (S)", nodeColor=<rgba 1 0 0.5543 1>}
  NamedRerouteDeclaration(None=FunctionInput(Preview=PixelNormalWS()) {description=Provided normal must be normalized, inputName=Normal WS, previewValue=<0 0 1 1>}) {name="Normal WS (V3)", nodeColor=<rgba 0 0.2734 1 1>}
  n9 = CameraVectorWS()
  n10 = NamedRerouteUsage()
  NamedRerouteDeclaration(None=Saturate(None=DotProduct(A=@n9, B=@n10))) {name="NdotV (S)", nodeColor=<rgba 1 0.1875 0 1>}
  n11 = FunctionInput() {bUsePreviewValueAsDefault=true, inputName=Specular, inputType=FunctionInput_Scalar, sortPriority=10, previewValue=<0.5 0 0 1>}
  n12 = Multiply(A=Saturate(None=@n11)) {constB=0.08}
  n13 = FunctionInput() {bUsePreviewValueAsDefault=true, inputName=Reflectance Mult, inputType=FunctionInput_Scalar, sortPriority=30, previewValue=<1 0 0 1>}
  NamedRerouteDeclaration(None=Saturate(None=Multiply(A=@n12, B=@n13))) {name="F0 (S)", nodeColor=<rgba 0.5 1 0 1>}
  n14 = NamedRerouteUsage()
  NamedRerouteDeclaration(None=Saturate(None=Multiply(A=@n14) {constB=50})) {name="F90 (S)", nodeColor=<rgba 1 0.5625 0 1>}
  NamedRerouteDeclaration(None=Saturate(None=FunctionInput() {bUsePreviewValueAsDefault=true, inputName=Diffuse Color, sortPriority=50})) {name="Diffuse Color (V3)", nodeColor=<rgba 1 0.4922 0 1>}
  n15 = NamedRerouteUsage()
  NamedRerouteDeclaration(None=Power(Base=NamedRerouteUsage(), Exp=LinearInterpolate(Alpha=@n15) {constA=1, constB=0.4})) {name="Blend Mask (S)", nodeColor=<rgba 0.4297 0 1 1>}
  n16 = NamedRerouteUsage()
  n17 = LinearInterpolate(Alpha=@n16) {constA=100}
  NamedRerouteDeclaration(None=LinearInterpolate(A=NamedRerouteUsage(), Alpha=Power(Base=@n16, Exp=@n17))) {name="Opacity (S)", nodeColor=<rgba 0.2188 1 0 1>}
  NamedRerouteDeclaration(None=LinearInterpolate(A=NamedRerouteUsage(), B=NamedRerouteUsage(), Alpha=NamedRerouteUsage())) {name="F0 Blend (S)", nodeColor=<rgba 1 0 0.5391 1>}
  NamedRerouteDeclaration(None=MaterialFunctionCall!(Input=NamedRerouteUsage(), Input Low=Constant() {r=0.08}, Input High=Constant() {r=1}) {materialFunction=/Engine/Functions/Engine_MaterialFunctions03/Math/RemapValueRangeNormalized}.Result) {name="Metallic (S)", nodeColor=<rgba 0 0.8594 1 1>}
  n18 = NamedRerouteUsage()
  n19 = NamedRerouteUsage()
  n20 = Power(Base=NamedRerouteUsage(), Exp=OneMinus(None=Multiply(A=@n19, B=@n19)))
  NamedRerouteDeclaration(None=LinearInterpolate(A=Multiply(A=@n18, B=@n20), B=NamedRerouteUsage(), Alpha=NamedRerouteUsage())) {name="BaseColor (S)", nodeColor=<rgba 1 0.8203 0 1>}
  n21 = Divide(A=NamedRerouteUsage()) {constB=0.08}
  NamedRerouteDeclaration(None=LinearInterpolate(A=Saturate(None=@n21), Alpha=NamedRerouteUsage()) {constB=0}) {name="Specular (S)", nodeColor=<rgba 0.5 0 1 1>}
 ]}
