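# Material function for the PixieParts lava body cosmetic that remaps an incoming emission signal into tinted energy emission.
# Input exponent and gain parameters, saturation and a pre-tint exponent shape the base emission, alongside two color inputs for tinting.
# Lerp, power and smoothstep stages add an outer gradient and edge mask, with packed gradient controls in one vector parameter.
# Brightness and saturation-offset inputs adjust the result, the latter via negated Desaturation.
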MaterialFunction /BRCosmetics/Characters/Player/Male/Medium/Bodies/M_MED_PixieParts_Lava/Materials/MaterialFunctions/MF_PIxieParts_EmissionRemap
graph {nodes [
  n1 = LinearInterpolate(A=NamedRerouteUsage(), B=NamedRerouteUsage(), Alpha=NamedRerouteUsage())
  n2 = Power(Base=@n1, Exp=NamedRerouteUsage())
  n3 = NamedRerouteUsage()
  n4 = NamedRerouteUsage()
  n5 = NamedRerouteUsage()
  n6 = NamedRerouteUsage()
  n7 = Multiply(A=LinearInterpolate(A=@n1, B=Multiply(A=@n2, B=@n3), Alpha=SmoothStep(Min=@n4, Max=@n5, Value=@n6)), B=NamedRerouteUsage())
  n8 = SmoothStep(Min=NamedRerouteUsage(), Max=NamedRerouteUsage(), Value=NamedRerouteUsage())
  FunctionOutput(None=Desaturation(None=Multiply(A=@n7, B=@n8), Fraction=NamedRerouteUsage())) {outputName="Emission Out (V3)"}
  n9 = VectorParameter() {
   group=PixieParts | Materials | Energy
   parameterName=Emission - Outer Gradient
   sortPriority=30
   channelNames {a=Mult, b=Exp, g=Gradient Max, r=Gradient Min}
   defaultValue=<rgba 0.25 -0.1 1 1>
  }
  NamedRerouteDeclaration(None=@n9.Gradient Min) {name="Outer Gradient Min (S)", nodeColor=<rgba 0.8281 0 1 1>}
  n10 = FunctionInput() {bUsePreviewValueAsDefault=true, inputName=Emission Brightness, inputType=FunctionInput_Scalar, sortPriority=40, previewValue=<1 1 1 1>}
  NamedRerouteDeclaration(None=@n10) {name="Emission Brightness (S)", nodeColor=<rgba 1 0.1739 0 1>}
  n11 = FunctionInput() {bUsePreviewValueAsDefault=true, inputName=Emission Pre-Tint Exp, inputType=FunctionInput_Scalar, sortPriority=10, previewValue=<1 1 1 1>}
  NamedRerouteDeclaration(None=@n11) {name="Emission Pre-Tint Exp (S)", nodeColor=<rgba 1 0.1739 0 1>}
  n12 = Multiply(A=Power(Base=NamedRerouteUsage(), Exp=NamedRerouteUsage()), B=NamedRerouteUsage())
  NamedRerouteDeclaration(None=Power(Base=Saturate(None=@n12), Exp=NamedRerouteUsage())) {name="Base Emission (S)", nodeColor=<rgba 1 0 0 1>}
  n13 = FunctionInput() {bUsePreviewValueAsDefault=true, inputName=Emission Color, sortPriority=20, previewValue=<1 1 1 1>}
  NamedRerouteDeclaration(None=Saturate(None=@n13)) {name="Emission Color (V3)", nodeColor=<rgba 1 0.1739 0 1>}
  n14 = FunctionInput() {
   bUsePreviewValueAsDefault=true
   inputName=Emission Saturation Offset
   inputType=FunctionInput_Scalar
   sortPriority=50
   previewValue=<0 1 1 1>
  }
  NamedRerouteDeclaration(None=Multiply(A=@n14) {constB=-1}) {name="Emission Desaturation (S)", nodeColor=<rgba 1 0.1739 0 1>}
  NamedRerouteDeclaration(None=ScalarParameter() {defaultValue=1, group=PixieParts | Materials | Energy, parameterName=Emission - Input Exp, sortPriority=10}) {name="Emission In Exp (S)", nodeColor=<rgba 0.1217 1 0 1>}
  n15 = FunctionInput() {bUsePreviewValueAsDefault=true, inputName=Emission Color 2, sortPriority=30, previewValue=<1 1 1 1>}
  NamedRerouteDeclaration(None=Saturate(None=@n15)) {name="Emission Color 2 (V3)", nodeColor=<rgba 1 0.1739 0 1>}
  NamedRerouteDeclaration(None=ScalarParameter() {defaultValue=1, group=PixieParts | Materials | Energy, parameterName=Emission - Input Gain, sortPriority=20}) {name="Emission In Gain (S)", nodeColor=<rgba 0.1217 1 0 1>}
  NamedRerouteDeclaration(None=@n9.Gradient Max) {name="Outer Gradient Max (S)", nodeColor=<rgba 0.8281 0 1 1>}
  NamedRerouteDeclaration(None=@n9.Exp) {name="Outer Gradient Exp (S)", nodeColor=<rgba 0.8281 0 1 1>}
  NamedRerouteDeclaration(None=@n9.Mult) {name="Outer Gradient Mult (S)", nodeColor=<rgba 0.8281 0 1 1>}
  NamedRerouteDeclaration(None=FunctionInput() {inputName=Emission In}) {name="Emission In (S)", nodeColor=<rgba 1 0.1739 0 1>}
  NamedRerouteDeclaration(None=ScalarParameter() {group=PixieParts | Materials | Energy, parameterName=Emission - Mask Smoothstep Min, sortPriority=40}) {name="Outer Edge Min (S)", nodeColor=<rgba 0 0.487 1 1>}
  NamedRerouteDeclaration(None=ScalarParameter() {defaultValue=0.025, group=PixieParts | Materials | Energy, parameterName=Emission - Mask Smoothstep Max, sortPriority=50}) {name="Outer Edge Max (S)", nodeColor=<rgba 0 0.487 1 1>}
 ]}
