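# Material function for the PixieParts Lava character body that turns incoming material attributes into a simple tinted material with remapped channels.
# Base color is desaturated, shaped by S-curve, exponent and gain from a packed vector input, then multiplied by a Tint.
# Specular and roughness each receive offset, S-curve and gain from packed inputs, then saturate.
# Metallic is replaced by a scalar input and emissive by a constant zero, linked through named reroutes.
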MaterialFunction /BRCosmetics/Characters/Player/Male/Medium/Bodies/M_MED_PixieParts_Lava/Materials/MaterialFunctions/MF_PixieParts_SimpleMaterial
graph {nodes [
  n1 = SetMaterialAttributes!(
   MaterialAttributes=NamedRerouteUsage()
   BaseColor=Multiply(A=NamedRerouteUsage(), B=NamedRerouteUsage())
   Metallic=NamedRerouteUsage()
   Specular=NamedRerouteUsage()
   Roughness=NamedRerouteUsage()
   EmissiveColor=Constant()
  ) {attributeSetTypes [
    #1
    #2
    #3
    #4
    #5
   ]}
  FunctionOutput(None=@n1) {outputName="Material Out (MA)"}
  NamedRerouteDeclaration(None=Saturate(None=FunctionInput() {bUsePreviewValueAsDefault=true, inputName=Metallic, inputType=FunctionInput_Scalar, sortPriority=30})) {name="Metallic (S)", nodeColor=<rgba 0.1679 0.6019 1 1>}
  MaterialExpressionFunctionInput_4 = FunctionInput() {
   desc="\r\nR : Offset\r\nG : S-Curve Exp\r\nB : Gain\r\n"
   description="R : Offset; G : S-Curve Exp; B : Gain"
   inputName=Roughness Remaps
   sortPriority=50
   previewValue=<0 1 0 1>
  }
  n2 = Convert(None=@MaterialExpressionFunctionInput_4) {
   nodeName=Break Float3
   convertInputs [{
     type=Vector3
     expressionInput {
      expression=&self:MaterialExpressionFunctionInput_4
      inputName=None
      mask=0
      maskA=0
      maskB=0
      maskG=0
      maskR=0
      outputIndex=0
     }
    }]
   convertOutputs [{type=Scalar}, {type=Scalar}, {type=Scalar}]
  }
  NamedRerouteDeclaration(None=@n2) {name="Roughness Mult (S)", nodeColor=<rgba 0 0.5217 1 1>}
  NamedRerouteDeclaration(None=@n2) {name="Roughness Offset (S)", nodeColor=<rgba 0 1 0.2843 1>}
  NamedRerouteDeclaration(None=@n2) {name="Roughness S-Curve Exp (S)", nodeColor=<rgba 0.0174 1 0 1>}
  MaterialExpressionFunctionInput_1 = FunctionInput() {
   desc="\r\nR : Desat\r\nG : S-Curve Exp\r\nB : Exp\r\nA  :Gain\r\n"
   description="R : Desat; G : S-Curve Exp; B : Exp; A  :Gain"
   inputName=Albedo Remaps
   inputType=FunctionInput_Vector4
   sortPriority=10
   previewValue=<1 1 1 1>
  }
  n3 = Convert(None=@MaterialExpressionFunctionInput_1) {
   nodeName=Break Float4
   convertInputs [{
     type=Vector4
     defaultValue=<rgba 0 0 0 1>
     expressionInput {
      expression=&self:MaterialExpressionFunctionInput_1
      inputName=None
      mask=0
      maskA=0
      maskB=0
      maskG=0
      maskR=0
      outputIndex=0
     }
    }]
   convertOutputs [{type=Scalar}, {type=Scalar}, {type=Scalar}, {type=Scalar}]
  }
  NamedRerouteDeclaration(None=@n3) {name="Albedo Gain (S)", nodeColor=<rgba 1 0.2783 0 1>}
  NamedRerouteDeclaration(None=@n3) {name="Albedo Exp (S)", nodeColor=<rgba 0 0.5217 1 1>}
  NamedRerouteDeclaration(None=@n3) {name="Albedo Desat (S)", nodeColor=<rgba 0 1 0.2843 1>}
  NamedRerouteDeclaration(None=@n3) {name="Albedo S-Curve (S)", nodeColor=<rgba 0.0174 1 0 1>}
  n4 = GetMaterialAttributes(None=NamedRerouteUsage()) {attributeGetTypes [#1, #3, #4]}
  NamedRerouteDeclaration(None=@n4.BaseColor) {name="BaseColor In (V3)", nodeColor=<rgba 0.0547 1 0 1>}
  n5 = MaterialFunctionCall(In=Desaturation(None=NamedRerouteUsage(), Fraction=NamedRerouteUsage()), Power=NamedRerouteUsage()) {materialFunction=&SCurve}
  n6 = NamedRerouteUsage()
  NamedRerouteDeclaration(None=Multiply(A=NamedRerouteUsage(), B=Power(Base=@n5.Result, Exp=@n6))) {name="GreyScale Albedo (S)", nodeColor=<rgba 1 0 0.375 1>}
  NamedRerouteDeclaration(None=@n4.Specular) {name="Specular In (S)", nodeColor=<rgba 1 0.1172 0 1>}
  NamedRerouteDeclaration(None=@n4.Roughness) {name="Roughness In (S)", nodeColor=<rgba 0 0.2031 1 1>}
  n7 = MaterialFunctionCall(In=Add(A=NamedRerouteUsage(), B=NamedRerouteUsage()), Power=NamedRerouteUsage()) {materialFunction=&SCurve}
  n8 = NamedRerouteUsage()
  NamedRerouteDeclaration(None=Saturate(None=Multiply(A=@n7.Result, B=@n8))) {name="Specular (S)", nodeColor=<rgba 1 0.1172 0 1>}
  MaterialExpressionFunctionInput_2 = FunctionInput() {
   desc="\r\nR : Offset\r\nG : S-Curve Exp\r\nB : Gain\r\n"
   description="R : Offset; G : S-Curve Exp; B : Gain"
   inputName=Specular Remaps
   sortPriority=40
   previewValue=<0 1 0 1>
  }
  n9 = Convert(None=@MaterialExpressionFunctionInput_2) {
   nodeName=Break Float3
   convertInputs [{
     type=Vector3
     expressionInput {
      expression=&self:MaterialExpressionFunctionInput_2
      inputName=None
      mask=0
      maskA=0
      maskB=0
      maskG=0
      maskR=0
      outputIndex=0
     }
    }]
   convertOutputs [{type=Scalar}, {type=Scalar}, {type=Scalar}]
  }
  NamedRerouteDeclaration(None=@n9) {name="Specular Mult (S)", nodeColor=<rgba 0 0.5217 1 1>}
  n10 = MaterialFunctionCall(In=Add(A=NamedRerouteUsage(), B=NamedRerouteUsage()), Power=NamedRerouteUsage()) {materialFunction=&SCurve}
  n11 = NamedRerouteUsage()
  NamedRerouteDeclaration(None=Saturate(None=Multiply(A=@n10.Result, B=@n11))) {name="Roughness (S)", nodeColor=<rgba 0 0.2031 1 1>}
  NamedRerouteDeclaration(None=@n9) {name="Specular Offset (S)", nodeColor=<rgba 0 1 0.2843 1>}
  NamedRerouteDeclaration(None=@n9) {name="Specular S-Curve Exp (S)", nodeColor=<rgba 0.0174 1 0 1>}
  NamedRerouteDeclaration(None=Saturate(None=FunctionInput() {bUsePreviewValueAsDefault=true, inputName=Tint, sortPriority=20, previewValue=<1 1 1 1>})) {name="Color (V3)", nodeColor=<rgba 0.1679 0.6019 1 1>}
  NamedRerouteDeclaration(None=FunctionInput() {inputName=Material In, inputType=FunctionInput_MaterialAttributes}) {name="Material In (MA)", nodeColor=<rgba 0.4531 1 0 1>}
 ]}
