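# Material function for the PixieParts Lava player body that builds its Base Layer, Armor and Stone material attributes from base and stone inputs.
# Base Layer lerps color toward an optional constant, applies exponent and gain, reshapes specular and roughness with exponents, and zeroes metallic.
# Armor and Stone use MF_PixieParts_SimpleMaterial with parameterized albedo, specular and roughness remaps.
# Stone also blends in a decal variant with gain-scaled color and fixed low roughness.
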MaterialFunction /BRCosmetics/Characters/Player/Male/Medium/Bodies/M_MED_PixieParts_Lava/Materials/MaterialFunctions/MF_PixieParts_BaseMaterials
graph {nodes [
  FunctionOutput(None=NamedRerouteUsage()) {outputName="Base Layer Material Out (MA)"}
  FunctionOutput(None=NamedRerouteUsage()) {outputName="Armor Material Out (MA)", sortPriority=10}
  FunctionOutput(None=NamedRerouteUsage()) {outputName="Stone Material Out (MA)", sortPriority=20}
  NamedRerouteDeclaration(None=FunctionInput() {inputName=Base Material In, inputType=FunctionInput_MaterialAttributes}) {name="Base Material In (MA)", nodeColor=<rgba 0.4531 1 0 1>}
  NamedRerouteDeclaration(None=FunctionInput() {inputName=Stone Material In, inputType=FunctionInput_MaterialAttributes, sortPriority=10}) {name="Stone Material In (MA)", nodeColor=<rgba 0.4531 1 0 1>}
  n1 = MaterialFunctionCall!(
   Material In=NamedRerouteUsage()
   Albedo Remaps=NamedRerouteUsage()
   Tint
   Metallic
   Specular Remaps=NamedRerouteUsage()
   Roughness Remaps=NamedRerouteUsage()
  ) {materialFunction=&/MF_PixieParts_SimpleMaterial}
  n2 = GetMaterialAttributes(None=@n1."Material Out (MA)") {attributeGetTypes [#1]}
  n3 = Multiply(A=@n2.BaseColor, B=NamedRerouteUsage())
  n4 = Constant() {r=0.2}
  NamedRerouteDeclaration(None=BlendMaterialAttributes(
    A=@n1."Material Out (MA)"
    B=SetMaterialAttributes!(MaterialAttributes=@n2.MaterialAttributes, BaseColor=@n3, Roughness=@n4) {attributeSetTypes [#1, #2]}
    Alpha=NamedRerouteUsage()
   ) {vertexAttributeBlendType=UseA}) {name="Stone Material (MA)", nodeColor=<rgba 0.4531 1 0 1>}
  NamedRerouteDeclaration(None=ScalarParameter() {defaultValue=0.6, group=PixieParts | Materials | Stone, parameterName=Stone Decal Gain, sliderMax=1, sortPriority=40}) {name="Stone Decal Gain (S)", nodeColor=<rgba 0.1679 1 0.4718 1>}
  n5 = VectorParameter() {
   group=PixieParts | Materials | Base Layer
   parameterName=Base Layer Color
   sortPriority=10
   channelNames {a=Use Contant Color, b=Constant Color B, g=Constant Color G, r=Constant Color R}
   defaultValue=<rgba 0.01 0.01 0.01 1>
  }
  NamedRerouteDeclaration(None=Saturate(None=@n5.Use Contant Color)) {name="Base Layer Use Constant Color (S)", nodeColor=<rgba 0 0.0859 1 1>}
  n6 = VectorParameter() {
   group=PixieParts | Materials | Base Layer
   parameterName=Base Layer Params
   sortPriority=20
   channelNames {a=Roughness Exp, b=Specular Exp, g=Color Gain, r=Color Exp}
   defaultValue=<rgba 1 1 1 1>
  }
  NamedRerouteDeclaration(None=@n6.Color Exp) {name="Base Layer Color Exp (S)", nodeColor=<rgba 0 0.0859 1 1>}
  NamedRerouteDeclaration(None=@n6.Color Gain) {name="Base Layer Color Gain (S)", nodeColor=<rgba 0 0.0859 1 1>}
  NamedRerouteDeclaration(None=@n6.Specular Exp) {name="Base Layer Specular Exp (S)", nodeColor=<rgba 1 0.1172 0 1>}
  NamedRerouteDeclaration(None=@n6.Roughness Exp) {name="Base Layer Roughness Exp (S)", nodeColor=<rgba 1 0.1172 0 1>}
  n7 = GetMaterialAttributes(None=NamedRerouteUsage()) {attributeGetTypes [#1, #3, #2]}
  NamedRerouteDeclaration(None=@n7.BaseColor) {name="Base Layer Color In (V3)", nodeColor=<rgba 0.0547 1 0 1>}
  NamedRerouteDeclaration(None=@n7.Specular) {name="Base Layer Specular In (S)", nodeColor=<rgba 1 0.1172 0 1>}
  NamedRerouteDeclaration(None=@n7.Roughness) {name="Base Layer Roughness In (S)", nodeColor=<rgba 0 0.2031 1 1>}
  n8 = Power(Base=LinearInterpolate(A=NamedRerouteUsage(), B=NamedRerouteUsage(), Alpha=NamedRerouteUsage()), Exp=NamedRerouteUsage())
  n9 = NamedRerouteUsage()
  NamedRerouteDeclaration(None=Saturate(None=Multiply(A=@n8, B=@n9))) {name="Base Layer Albedo (V3)", nodeColor=<rgba 1 0 0.375 1>}
  NamedRerouteDeclaration(None=Power(Base=NamedRerouteUsage(), Exp=NamedRerouteUsage())) {name="Base Layer Specular (S)", nodeColor=<rgba 1 0.1172 0 1>}
  NamedRerouteDeclaration(None=Power(Base=NamedRerouteUsage(), Exp=NamedRerouteUsage())) {name="Base Layer Roughness (S)", nodeColor=<rgba 0 0.2031 1 1>}
  n10 = SetMaterialAttributes!(
   MaterialAttributes=NamedRerouteUsage()
   BaseColor=NamedRerouteUsage()
   Metallic=Constant()
   Specular=NamedRerouteUsage()
   Roughness=NamedRerouteUsage()
  ) {attributeSetTypes [
    #1
    #4
    #3
    #2
   ]}
  NamedRerouteDeclaration(None=@n10) {name="Base Layer Material (MA)", nodeColor=<rgba 0.4531 1 0 1>}
  n11 = VectorParameter() {
   group=PixieParts | Materials | Armor
   parameterName=Armor Albedo Remap
   sortPriority=10
   channelNames {a=Gain, b=Exp, g=S-Curve Exp, r=Desat}
   defaultValue=<rgba 1 1 1 1>
  }
  NamedRerouteDeclaration(None=@n11.RGBA) {name="Armor Albedo Remap (V4)", nodeColor=<rgba 0.1679 0.6019 1 1>}
  n12 = FunctionInput() {bUsePreviewValueAsDefault=true, inputName=Decals Mask In, inputType=FunctionInput_Scalar, sortPriority=20, previewValue=<0 0 0 0>}
  NamedRerouteDeclaration(None=@n12) {name="Decals Mask In (S)", nodeColor=<rgba 0 1 0.5078 1>}
  n13 = VectorParameter() {
   group=PixieParts | Materials | Armor
   parameterName=Armor Roughness Remap
   sortPriority=30
   channelNames {a=n/a, b=Mult, g=S-Curve Exp, r=Offset}
   defaultValue=<rgba 0 1 1 0>
  }
  NamedRerouteDeclaration(None=@n13.RGB) {name="Armor Roughness Remap (V3)", nodeColor=<rgba 0.1679 0.6019 1 1>}
  n14 = VectorParameter() {
   group=PixieParts | Materials | Armor
   parameterName=Armor Specular Remap
   sortPriority=20
   channelNames {a=n/a, b=Mult, g=S-Curve Exp, r=Offset}
   defaultValue=<rgba 0 1 1 0>
  }
  NamedRerouteDeclaration(None=@n14.RGB) {name="Armor Specular Remap (V3)", nodeColor=<rgba 0.1679 0.6019 1 1>}
  n15 = VectorParameter() {
   group=PixieParts | Materials | Stone
   parameterName=Stone Albedo Remap
   sortPriority=10
   channelNames {a=Gain, b=Exp, g=S-Curve Exp, r=Desat}
   defaultValue=<rgba 1 1 1 1>
  }
  NamedRerouteDeclaration(None=@n15.RGBA) {name="Stone Albedo Remap (V4)", nodeColor=<rgba 0.1679 1 0.4718 1>}
  n16 = VectorParameter() {
   group=PixieParts | Materials | Stone
   parameterName=Stone Roughness Remap
   sortPriority=30
   channelNames {a=n/a, b=Mult, g=S-Curve Exp, r=Offset}
   defaultValue=<rgba 0 1 1 0>
  }
  NamedRerouteDeclaration(None=@n16.RGB) {name="Stone Roughness Remap (V3)", nodeColor=<rgba 0.1679 1 0.4718 1>}
  n17 = VectorParameter() {
   group=PixieParts | Materials | Stone
   parameterName=Stone Specular Remap
   sortPriority=20
   channelNames {a=n/a, b=Mult, g=S-Curve Exp, r=Offset}
   defaultValue=<rgba 0 1 1 0>
  }
  NamedRerouteDeclaration(None=@n17.RGB) {name="Stone Specular Remap (V3)", nodeColor=<rgba 0.1679 1 0.4718 1>}
  n18 = MaterialFunctionCall(
   Material In=NamedRerouteUsage()
   Albedo Remaps=NamedRerouteUsage()
   Specular Remaps=NamedRerouteUsage()
   Roughness Remaps=NamedRerouteUsage()
  ) {materialFunction=&MF_PixieParts_SimpleMaterial}
  NamedRerouteDeclaration(None=@n18."Material Out (MA)") {name="Armor Material (MA)", nodeColor=<rgba 0.4531 1 0 1>}
  NamedRerouteDeclaration(None=Saturate(None=@n5.RGB)) {name="Base Layer Constant Color (V3)", nodeColor=<rgba 0 0.0859 1 1>}
 ]}
