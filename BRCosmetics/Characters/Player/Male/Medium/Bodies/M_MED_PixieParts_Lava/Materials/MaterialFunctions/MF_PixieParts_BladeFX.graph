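# Material function for the PixieParts Lava character's blade effect, outputting the effect, its normalized UVs and a blade mask.
# Parameters remap a UV island with offset, scale and optional fmod wrap, feeding the borrowed OutlineLove gadget fire function.
# The mask combines a UV range test with a step on vertex-interpolated, pre-skinning local Z.
# The mask is output separately, not applied to the effect.
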MaterialFunction /BRCosmetics/Characters/Player/Male/Medium/Bodies/M_MED_PixieParts_Lava/Materials/MaterialFunctions/MF_PixieParts_BladeFX
graph {nodes [
  FunctionOutput(None=NamedRerouteUsage()) {outputName="Blade FX (V3)"}
  FunctionOutput(None=NamedRerouteUsage()) {outputName="Blade UV [Normalized] (V2)", sortPriority=10}
  FunctionOutput(None=NamedRerouteUsage()) {outputName="Blade Mask (S)", sortPriority=20}
  n1 = VectorParameter() {
   group=PixieParts | DEV
   parameterName=Blade Mask
   sortPriority=110
   channelNames {a=n/a, b=Local Z min, g=U max, r=U min}
   defaultValue=<rgba 0.8 1 19 0>
  }
  n2 = AppendVector(A=@n1.U min, B=Constant())
  n3 = Constant() {r=1}
  n4 = Divide(A=Subtract(A=TextureCoordinate(), B=@n2), B=Subtract(A=AppendVector(A=@n1.U max, B=@n3), B=@n2))
  MaterialExpressionMultiply_0 = Multiply(A=OneMinus(None=Step(X=@n4) {constY=1}), B=Step(X=@n4))
  n5 = Convert(None=@MaterialExpressionMultiply_0) {
   nodeName=Break Float2
   convertInputs [{
     type=Vector2
     expressionInput {
      expression=&self:MaterialExpressionMultiply_0
      inputName=None
      mask=0
      maskA=0
      maskB=0
      maskG=0
      maskR=0
      outputIndex=0
     }
    }]
   convertOutputs [{type=Scalar}, {type=Scalar}]
  }
  n6 = VertexInterpolator(VS=LocalPosition() {includedOffsets=ExcludeOffsets, localOrigin=InstancePreSkinning}.Z)
  NamedRerouteDeclaration(None=Multiply(A=Multiply(A=@n5, B=@n5), B=Step(Y=@n1.Local Z min, X=@n6.PS))) {name="Blade Mask (S)", nodeColor=<rgba 1 0.1172 0 1>}
  n7 = VectorParameter() {
   group=PixieParts | DEV
   parameterName=FX UV Island
   sortPriority=110
   channelNames {a=V max, b=U max, g=V min, r=U min}
   defaultValue=<rgba 0.8 0 0.9 1>
  }
  n8 = AppendVector(A=@n7.U min, B=@n7.U min)
  n9 = Subtract(A=TextureCoordinate(), B=@n8)
  n10 = Subtract(A=AppendVector(A=@n7.U max, B=@n7.U max), B=@n8)
  n11 = VectorParameter() {
   group=PixieParts | DEV
   parameterName=FX UV Controls
   sortPriority=120
   channelNames {a=V scale, b=U scale, g=V offset, r=U offset}
   defaultValue=<rgba 0 0 1 1>
  }
  n12 = Divide(A=Subtract(A=Divide(A=@n9, B=@n10), B=AppendVector(A=@n11.U offset, B=@n11.U offset)), B=AppendVector(A=@n11.U scale, B=@n11.U scale))
  n13 = VectorParameter() {
   group=PixieParts | DEV
   parameterName=FX UV Fmod
   sortPriority=120
   channelNames {a=V fmod, b=U fmod, g=Enable V fmod, r=Enable U fmod}
   defaultValue=<rgba 0 0 1 1>
  }
  n14 = LinearInterpolate(A=@n12, B=Fmod(A=@n12, B=AppendVector(A=@n13.U fmod, B=@n13.U fmod)), Alpha=AppendVector(A=@n13.Enable U fmod, B=@n13.Enable U fmod))
  NamedRerouteDeclaration(None=MaterialFunctionCall!(UV=NamedRerouteDeclaration(None=@n14) {name="Blade UV [Normalized] (V2)", nodeColor=<rgba 0 1 0.3438 1>}) {materialFunction=/BRCosmetics/Animation/Props/Gadgets_OutlineLove/Materials/MF_Gadget_OutlineLove_Fire}."Result (V3)") {name="Blade FX (V3)", nodeColor=<rgba 0 0.2734 1 1>}
 ]}
