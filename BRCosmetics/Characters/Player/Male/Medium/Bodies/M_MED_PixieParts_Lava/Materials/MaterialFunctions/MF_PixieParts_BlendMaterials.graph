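# Material function that blends layered materials for the PixieParts Lava player character body.
# Chained BlendMaterialAttributes nodes build a composite from masked materials, with inputs for a base layer, five materials, three decals and an energy layer.
# Named reroutes link the composite, decal and final blend stages, each blend keeping vertex attributes from input A.
# The final stage adds a rerouted value, apparently Energy Emission, to emissive before output.
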MaterialFunction /BRCosmetics/Characters/Player/Male/Medium/Bodies/M_MED_PixieParts_Lava/Materials/MaterialFunctions/MF_PixieParts_BlendMaterials
graph {nodes [
  n1 = GetMaterialAttributes(None=BlendMaterialAttributes(A=NamedRerouteUsage(), B=NamedRerouteUsage(), Alpha=NamedRerouteUsage()) {vertexAttributeBlendType=UseA}) {attributeGetTypes [#1]}
  n2 = NamedRerouteUsage()
  FunctionOutput(None=SetMaterialAttributes!(MaterialAttributes=@n1.MaterialAttributes, EmissiveColor=Add(A=@n1.EmissiveColor, B=@n2)) {attributeSetTypes [#1]}) {outputName="Material Out (MA)"}
  NamedRerouteDeclaration(None=FunctionInput() {inputName=Decal Material 2 In, inputType=FunctionInput_MaterialAttributes, sortPriority=320}) {name="Decal Material 2 In (MA)", nodeColor=<rgba 1 1 1 1>}
  NamedRerouteDeclaration(None=FunctionInput() {bUsePreviewValueAsDefault=true, inputName=Mask Decal 1, inputType=FunctionInput_Scalar, sortPriority=310}) {name="Mask Decal 1 (S)", nodeColor=<rgba 0 1 0.1524 1>}
  NamedRerouteDeclaration(None=FunctionInput() {bUsePreviewValueAsDefault=true, inputName=Mask Decal 2, inputType=FunctionInput_Scalar, sortPriority=330}) {name="Mask Decal 2 (S)", nodeColor=<rgba 0 1 0.1524 1>}
  NamedRerouteDeclaration(None=FunctionInput() {inputName=Decal Material 3 In, inputType=FunctionInput_MaterialAttributes, sortPriority=340}) {name="Decal Material 3 In (MA)", nodeColor=<rgba 1 1 1 1>}
  NamedRerouteDeclaration(None=FunctionInput() {bUsePreviewValueAsDefault=true, inputName=Mask Decal 3, inputType=FunctionInput_Scalar, sortPriority=350}) {name="Mask Decal 3 (S)", nodeColor=<rgba 0 1 0.1524 1>}
  n3 = NamedRerouteUsage()
  n4 = NamedRerouteUsage()
  n5 = NamedRerouteUsage()
  n6 = BlendMaterialAttributes(
   A=BlendMaterialAttributes(A=BlendMaterialAttributes(A=@n3, B=@n4, Alpha=@n5) {vertexAttributeBlendType=UseA}, B=NamedRerouteUsage(), Alpha=NamedRerouteUsage()) {vertexAttributeBlendType=UseA}
   B=NamedRerouteUsage()
   Alpha=NamedRerouteUsage()
  ) {vertexAttributeBlendType=UseA}
  n7 = NamedRerouteUsage()
  n8 = NamedRerouteUsage()
  NamedRerouteDeclaration(None=BlendMaterialAttributes(A=BlendMaterialAttributes(A=@n6, B=@n7, Alpha=@n8) {vertexAttributeBlendType=UseA}, B=NamedRerouteUsage(), Alpha=NamedRerouteUsage()) {vertexAttributeBlendType=UseA}) {name="Material Composite (MA)", nodeColor=<rgba 1 1 1 1>}
  NamedRerouteDeclaration(None=FunctionInput() {inputName=Decal Material 1 In, inputType=FunctionInput_MaterialAttributes, sortPriority=300}) {name="Decal Material 1 In (MA)", nodeColor=<rgba 1 1 1 1>}
  n9 = BlendMaterialAttributes(A=NamedRerouteUsage(), B=NamedRerouteUsage(), Alpha=NamedRerouteUsage()) {vertexAttributeBlendType=UseA}
  n10 = NamedRerouteUsage()
  n11 = NamedRerouteUsage()
  NamedRerouteDeclaration(None=BlendMaterialAttributes(A=BlendMaterialAttributes(A=@n9, B=@n10, Alpha=@n11) {vertexAttributeBlendType=UseA}, B=NamedRerouteUsage(), Alpha=NamedRerouteUsage()) {vertexAttributeBlendType=UseA}) {name="Material with Decals (MA)", nodeColor=<rgba 1 1 1 1>}
  NamedRerouteDeclaration(None=FunctionInput() {inputName=Material 1 In, inputType=FunctionInput_MaterialAttributes, sortPriority=100}) {name="Material 1 In (MA)", nodeColor=<rgba 1 1 1 1>}
  NamedRerouteDeclaration(None=FunctionInput() {inputName=Material 2 In, inputType=FunctionInput_MaterialAttributes, sortPriority=120}) {name="Material 2 In (MA)", nodeColor=<rgba 1 1 1 1>}
  NamedRerouteDeclaration(None=FunctionInput() {inputName=Material 4 In, inputType=FunctionInput_MaterialAttributes, sortPriority=160}) {name="Material 4 In (MA)", nodeColor=<rgba 1 1 1 1>}
  NamedRerouteDeclaration(None=FunctionInput() {inputName=Material 3 In, inputType=FunctionInput_MaterialAttributes, sortPriority=140}) {name="Material 3 In (MA)", nodeColor=<rgba 1 1 1 1>}
  NamedRerouteDeclaration(None=FunctionInput() {inputName=Material 5 In, inputType=FunctionInput_MaterialAttributes, sortPriority=180}) {name="Material 5 In (MA)", nodeColor=<rgba 1 1 1 1>}
  NamedRerouteDeclaration(None=FunctionInput() {inputName=Base Layer Material In, inputType=FunctionInput_MaterialAttributes}) {name="Base Layer Material In (MA)", nodeColor=<rgba 0.4531 1 0 1>}
  NamedRerouteDeclaration(None=FunctionInput() {bUsePreviewValueAsDefault=true, inputName=Mask Material 1, inputType=FunctionInput_Scalar, sortPriority=110}) {name="Mask Material 1 (S)", nodeColor=<rgba 0.5703 0 1 1>}
  NamedRerouteDeclaration(None=FunctionInput() {bUsePreviewValueAsDefault=true, inputName=Mask Material 2, inputType=FunctionInput_Scalar, sortPriority=130}) {name="Mask Material 2 (S)", nodeColor=<rgba 0.5703 0 1 1>}
  NamedRerouteDeclaration(None=FunctionInput() {bUsePreviewValueAsDefault=true, inputName=Mask Material 3, inputType=FunctionInput_Scalar, sortPriority=150}) {name="Mask Material 3 (S)", nodeColor=<rgba 0.5703 0 1 1>}
  NamedRerouteDeclaration(None=FunctionInput() {bUsePreviewValueAsDefault=true, inputName=Mask Material 4, inputType=FunctionInput_Scalar, sortPriority=170}) {name="Mask Material 4 (S)", nodeColor=<rgba 0 0.4819 1 1>}
  NamedRerouteDeclaration(None=FunctionInput() {bUsePreviewValueAsDefault=true, inputName=Mask Material 5, inputType=FunctionInput_Scalar, sortPriority=190}) {name="Mask Material 5 (S)", nodeColor=<rgba 0 0.4819 1 1>}
  NamedRerouteDeclaration(None=FunctionInput() {bUsePreviewValueAsDefault=true, inputName=Mask Energy, inputType=FunctionInput_Scalar, sortPriority=220}) {name="Mask Energy (S)", nodeColor=<rgba 0 0.4819 1 1>}
  n12 = FunctionInput() {bUsePreviewValueAsDefault=true, inputName=Energy Emission, sortPriority=210, previewValue=<1 1 1 1>}
  NamedRerouteDeclaration(None=@n12) {name="Energy Emission (V3)", nodeColor=<rgba 1 0 0 1>}
  NamedRerouteDeclaration(None=FunctionInput() {inputName=Energy Material In, inputType=FunctionInput_MaterialAttributes, sortPriority=200}) {name="Energy Material In (MA)", nodeColor=<rgba 1 1 1 1>}
 ]}
