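# Material function for the PixieParts Lava character body that applies optional clear coat shading to incoming material attributes.
# It reads Displacement, SubsurfaceColor and Refraction as packed data channels for coat intensity, under-coat metallic multiplier, under-coat roughness, roughness mix and enable.
# It sets the Clear Coat shading model, scales Metallic, and cross-lerps material and under-coat roughness between Roughness and CustomData1.
# The "Use Clear Coat" static switch gates an attribute blend for the output.
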MaterialFunction /BRCosmetics/Characters/Player/Male/Medium/Bodies/M_MED_PixieParts_Lava/Materials/MaterialFunctions/MF_PixieParts_ClearCoat
graph {nodes [
  n1 = NamedRerouteUsage()
  n2 = NamedRerouteUsage()
  n3 = NamedRerouteUsage()
  FunctionOutput(None=StaticSwitchParameter(True=BlendMaterialAttributes(A=@n1, B=@n2, Alpha=@n3) {vertexAttributeBlendType=UseA}, False=NamedRerouteUsage()) {group=PixieParts | Globals | Clear Coat, parameterName=Use Clear Coat, sortPriority=0}) {outputName="Material Out (MA)"}
  MaterialExpressionGetMaterialAttributes_0 = GetMaterialAttributes(None=NamedRerouteDeclaration(None=FunctionInput() {inputName=Material In, inputType=FunctionInput_MaterialAttributes}) {name="Material In (MA)", nodeColor=<rgba 0.4531 1 0 1>}) {attributeGetTypes [
    #1
    #2
    #3
    #4
    #5
   ]}
  NamedRerouteDeclaration(None=@MaterialExpressionGetMaterialAttributes_0.Metallic) {name="Metallic In (S)", nodeColor=<rgba 0 1 0.7826 1>}
  NamedRerouteDeclaration(None=@MaterialExpressionGetMaterialAttributes_0.Displacement) {name="Clear Coat Intensity (S)", nodeColor=<rgba 1 0 0.5859 1>}
  NamedRerouteDeclaration(None=@MaterialExpressionGetMaterialAttributes_0.Roughness) {name="Roughness In (S)", nodeColor=<rgba 0 0.3906 1 1>}
  n4 = Convert(None=@MaterialExpressionGetMaterialAttributes_0.SubsurfaceColor) {
   nodeName=Break Float3
   convertInputs [{
     type=Vector3
     expressionInput {
      expression=&self:MaterialExpressionGetMaterialAttributes_0
      inputName=None
      mask=0
      maskA=0
      maskB=0
      maskG=0
      maskR=0
      outputIndex=5
     }
    }]
   convertOutputs [{type=Scalar}, {type=Scalar}, {type=Scalar}]
  }
  NamedRerouteDeclaration(None=@n4) {name="Under Coat Metallic Mult (S)", nodeColor=<rgba 1 0 0.5859 1>}
  NamedRerouteDeclaration(None=@n4) {name="Under Coat Roughness (S)", nodeColor=<rgba 1 0 0.5859 1>}
  NamedRerouteDeclaration(None=@n4) {name="Top Coat Material Roughness Mix (S)", nodeColor=<rgba 1 0 0.5859 1>}
  n5 = NamedRerouteUsage()
  n6 = NamedRerouteUsage()
  n7 = NamedRerouteUsage()
  n8 = SetMaterialAttributes!(
   MaterialAttributes=NamedRerouteUsage()
   CustomData0=NamedRerouteUsage()
   CustomData1=LinearInterpolate(A=@n5, B=@n6, Alpha=@n7)
   Roughness=LinearInterpolate(A=@n6, B=@n5, Alpha=@n7)
   Metallic=Multiply(A=NamedRerouteUsage(), B=NamedRerouteUsage())
   ShadingModel=ShadingModel() {shadingModel=MSM_ClearCoat}
  ) {attributeSetTypes [
    #6
    #7
    #2
    #1
    #8
   ]}
  NamedRerouteDeclaration(None=@n8) {name="Material Out (MA)", nodeColor=<rgba 0 1 0.0859 1>}
  NamedRerouteDeclaration(None=@MaterialExpressionGetMaterialAttributes_0.Refraction) {name="Enable Clear Coat (S)", nodeColor=<rgba 1 0 0.5859 1>}
 ]}
