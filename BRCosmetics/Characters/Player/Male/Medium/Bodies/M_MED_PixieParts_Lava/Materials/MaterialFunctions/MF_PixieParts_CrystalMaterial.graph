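# Material function that builds the crystal material for PixieParts character cosmetics, with a faked refractive interior.
# A tiled detail normal blends into the incoming normal, and a bump-offset copy of the same texture forms the interior normal.
# Refraction and reflection vectors sample the skylight, blended by an approximate total internal reflection term, then tinted and masked by Fresnel and GMap for emissive.
# Metal LUT, clear coat and thin-film parameters are packed into repurposed material attributes.
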MaterialFunction /BRCosmetics/Characters/Player/Male/Medium/Bodies/M_MED_PixieParts_Lava/Materials/MaterialFunctions/MF_PixieParts_CrystalMaterial
graph {expressions {
  MaterialExpressionCameraVectorWS_1 {class=MaterialExpressionCameraVectorWS, inputs [], properties {}}
  MaterialExpressionConstant_0 {class=MaterialExpressionConstant}
  MaterialExpressionConstant_1 {class=MaterialExpressionConstant, properties {r=0.73}}
  MaterialExpressionConstant_2 {class=MaterialExpressionConstant, properties {r=1}}
  MaterialExpressionConstant_3 {class=MaterialExpressionConstant, properties {r=0.78}}
  MaterialExpressionConvert_1 {
   class=MaterialExpressionConvert
   inputs [None {from=MaterialExpressionNamedRerouteUsage_9}]
   properties {
    convertInputs [{
      type=Vector4
      defaultValue=<rgba 0 0 0 1>
      expressionInput {
       expression=&self:MaterialExpressionNamedRerouteUsage_9
       inputName=None
       mask=0
       maskA=0
       maskB=0
       maskG=0
       maskR=0
       outputIndex=0
      }
     }]
    convertOutputs [{type=Scalar}, {type=Vector3}]
   }
  }
  MaterialExpressionConvert_3 {
   class=MaterialExpressionConvert
   inputs [
    None {from=MaterialExpressionScalarParameter_4}
    None {from=MaterialExpressionMultiply_8}
   ]
   properties {
    nodeName=Make Vector2
    convertInputs [
     {
      type=Scalar
      expressionInput {
       expression=&self:MaterialExpressionScalarParameter_4
       inputName=None
       mask=0
       maskA=0
       maskB=0
       maskG=0
       maskR=0
       outputIndex=0
      }
     }
     {
      type=Scalar
      expressionInput {
       expression=&self:MaterialExpressionMultiply_8
       inputName=None
       mask=0
       maskA=0
       maskB=0
       maskG=0
       maskR=0
       outputIndex=0
      }
     }
    ]
    convertOutputs [{type=Vector2}]
   }
  }
  MaterialExpressionDotProduct_2 {
   class=MaterialExpressionDotProduct
   inputs [
    A {from=MaterialExpressionNamedRerouteUsage_1, output=""}
    B {from=MaterialExpressionNamedRerouteUsage_2, output=""}
   ]
   properties {}
  }
  MaterialExpressionFresnel_0 {
   class=MaterialExpressionFresnel
   inputs [
    ExponentIn {from=MaterialExpressionNamedRerouteUsage_24, output=""}
    BaseReflectFractionIn {from="", output=""}
    Normal {from=MaterialExpressionNamedRerouteUsage_4, output=""}
   ]
   properties {baseReflectFraction=0}
  }
  MaterialExpressionFunctionInput_2 {
   class=MaterialExpressionFunctionInput
   inputs [Preview {from="", output=""}]
   properties {bUsePreviewValueAsDefault=true, inputName=GMap In, inputType=FunctionInput_Scalar, sortPriority=10, previewValue=<1 1 0 1>}
  }
  MaterialExpressionFunctionInput_5 {
   class=MaterialExpressionFunctionInput
   inputs [Preview {from="", output=""}]
   properties {inputName=Material In, inputType=FunctionInput_MaterialAttributes}
  }
  MaterialExpressionFunctionOutput_0 {
   class=MaterialExpressionFunctionOutput
   inputs [None {from=MaterialExpressionSetMaterialAttributes_3, output=""}]
   properties {outputName="Material Out (MA)"}
  }
  MaterialExpressionGetMaterialAttributes_0 {
   class=MaterialExpressionGetMaterialAttributes
   inputs [None {from=MaterialExpressionNamedRerouteDeclaration_190, output=""}]
   properties {attributeGetTypes [#1]}
  }
  MaterialExpressionLinearInterpolate_0 {
   class=MaterialExpressionLinearInterpolate
   inputs [
    A {from=MaterialExpressionSkyLightEnvMapSample_1}
    B {from=MaterialExpressionSkyLightEnvMapSample_0}
    Alpha {from=MaterialExpressionNamedRerouteUsage_3}
   ]
  }
  MaterialExpressionLinearInterpolate_8 {
   class=MaterialExpressionLinearInterpolate
   inputs [
    A {from=""}
    B {from=""}
    Alpha {from=MaterialExpressionMaterialFunctionCall_8, output=Result}
   ]
   properties {constA=0.1, constB=0.9}
  }
  MaterialExpressionMaterialFunctionCall_1 {
   class=MaterialExpressionMaterialFunctionCall
   inputs [
    BaseNormal {from=MaterialExpressionNamedRerouteUsage_15}
    AdditionalNormal {from=MaterialExpressionMaterialFunctionCall_4, output="Result (V3)"}
   ]
   properties {materialFunction=&BlendAngleCorrectedNormals}
  }
  MaterialExpressionMaterialFunctionCall_2 {
   class=MaterialExpressionMaterialFunctionCall
   inputs [
    Ray Direction {from=MaterialExpressionNamedRerouteUsage_8}
    Surface Normal {from=MaterialExpressionNamedRerouteUsage_7}
    Refractive Index Origin {from=MaterialExpressionNamedRerouteUsage_25}
    Refractive Index Target {from=MaterialExpressionConstant_2}
   ]
   properties {materialFunction=/Engine/Functions/Engine_MaterialFunctions01/Vectors/Refract}
  }
  MaterialExpressionMaterialFunctionCall_3 {
   class=MaterialExpressionMaterialFunctionCall
   inputs [
    Ray Direction {from=MaterialExpressionMultiply_1}
    Surface Normal {from=MaterialExpressionNamedRerouteUsage_10}
    Refractive Index Origin {from=""}
    Refractive Index Target {from=MaterialExpressionNamedRerouteUsage_23}
   ]
   properties {materialFunction=&Refract}
  }
  MaterialExpressionMaterialFunctionCall_4 {
   class=MaterialExpressionMaterialFunctionCall
   inputs [
    Normal {from=MaterialExpressionTextureSample_0, output=RGB}
    Intensity {from=MaterialExpressionNamedRerouteUsage_18}
   ]
   properties {materialFunction=/Game/Global/MaterialFunctions/MF_NormalIntensity}
  }
  MaterialExpressionMaterialFunctionCall_5 {
   class=MaterialExpressionMaterialFunctionCall
   inputs [
    UV Coord {from=MaterialExpressionMultiply_0}
    Height {from=MaterialExpressionNamedRerouteUsage_37}
    ReferencePlane {from=MaterialExpressionConstant_0}
    HeightRatio {from=""}
    Normal {from=MaterialExpressionNamedRerouteUsage_38}
   ]
   properties {materialFunction=/Game/Global/MaterialFunctions/MF_BumpOffset_Accurate}
  }
  MaterialExpressionMaterialFunctionCall_6 {
   class=MaterialExpressionMaterialFunctionCall
   inputs [
    Normal {from=MaterialExpressionTextureSample_2, output=RGB}
    Intensity {from=MaterialExpressionNamedRerouteUsage_47}
   ]
   properties {materialFunction=&MF_NormalIntensity}
  }
  MaterialExpressionMaterialFunctionCall_8 {
   class=MaterialExpressionMaterialFunctionCall
   inputs [
    Input {from=MaterialExpressionMultiply_3}
    Input Low {from=MaterialExpressionConstant_1}
    Input High {from=MaterialExpressionConstant_3}
   ]
   properties {materialFunction=&RemapValue_0-1}
  }
  MaterialExpressionMultiply_0 {
   class=MaterialExpressionMultiply
   inputs [
    A {from=MaterialExpressionTextureCoordinate_1}
    B {from=MaterialExpressionNamedRerouteUsage_45}
   ]
  }
  MaterialExpressionMultiply_1 {
   class=MaterialExpressionMultiply
   inputs [
    A {from=MaterialExpressionCameraVectorWS_1}
    B {from=""}
   ]
   properties {constB=-1}
  }
  MaterialExpressionMultiply_2 {
   class=MaterialExpressionMultiply
   inputs [
    A {from=MaterialExpressionTextureCoordinate_0}
    B {from=MaterialExpressionNamedRerouteUsage_17}
   ]
  }
  MaterialExpressionMultiply_3 {
   class=MaterialExpressionMultiply
   inputs [
    A {from=MaterialExpressionDotProduct_2}
    B {from=""}
   ]
   properties {constB=-1}
  }
  MaterialExpressionMultiply_4 {
   class=MaterialExpressionMultiply
   inputs [
    A {from=MaterialExpressionLinearInterpolate_0}
    B {from=MaterialExpressionNamedRerouteUsage_30}
   ]
  }
  MaterialExpressionMultiply_5 {
   class=MaterialExpressionMultiply
   inputs [
    A {from=MaterialExpressionMultiply_4}
    B {from=MaterialExpressionNamedRerouteUsage_12}
   ]
  }
  MaterialExpressionMultiply_6 {
   class=MaterialExpressionMultiply
   inputs [
    A {from=MaterialExpressionMultiply_5}
    B {from=MaterialExpressionPower_1}
   ]
  }
  MaterialExpressionMultiply_8 {
   class=MaterialExpressionMultiply
   inputs [
    A {from=MaterialExpressionScalarParameter_4}
    B {from=MaterialExpressionScalarParameter_18}
   ]
  }
  MaterialExpressionMultiply_9 {
   class=MaterialExpressionMultiply
   inputs [
    A {from=MaterialExpressionScalarParameter_2}
    B {from=MaterialExpressionNamedRerouteUsage_48}
   ]
  }
  MaterialExpressionNamedRerouteDeclaration_0 {
   class=MaterialExpressionNamedRerouteDeclaration
   inputs [None {from=MaterialExpressionNormalize_1}]
   properties {name="Interior Normal WS (V3)", nodeColor=<rgba 0 0.2266 1 1>}
  }
  MaterialExpressionNamedRerouteDeclaration_1 {
   class=MaterialExpressionNamedRerouteDeclaration
   inputs [None {from=MaterialExpressionMaterialFunctionCall_3, output=Vector}]
   properties {name="Refraction Vector WS (V3)", nodeColor=<rgba 1 0 0.4219 1>}
  }
  MaterialExpressionNamedRerouteDeclaration_10 {
   class=MaterialExpressionNamedRerouteDeclaration
   inputs [None {from=MaterialExpressionVectorParameter_3, output=RGB}]
   properties {name="Metal LUT (V3)", nodeColor=<rgba 0 0.9062 1 1>}
  }
  MaterialExpressionNamedRerouteDeclaration_11 {
   class=MaterialExpressionNamedRerouteDeclaration
   inputs [None {from=MaterialExpressionConvert_3}]
   properties {name="Detail Normal Tiling (V2)", nodeColor=<rgba 1 0 0.8672 1>}
  }
  MaterialExpressionNamedRerouteDeclaration_119 {
   class=MaterialExpressionNamedRerouteDeclaration
   inputs [None {from=MaterialExpressionConvert_1}]
   properties {name="Clear Coat ##1 (S)", nodeColor=<rgba 1 0 0.5859 1>}
  }
  MaterialExpressionNamedRerouteDeclaration_12 {
   class=MaterialExpressionNamedRerouteDeclaration
   inputs [None {from=MaterialExpressionScalarParameter_11}]
   properties {name="Detail Normal Intensity (S)", nodeColor=<rgba 0 1 0.8125 1>}
  }
  MaterialExpressionNamedRerouteDeclaration_13 {
   class=MaterialExpressionNamedRerouteDeclaration
   inputs [None {from=MaterialExpressionScalarParameter_0}]
   properties {name="Interior Depth (S)", nodeColor=<rgba 1 0.4922 0 1>}
  }
  MaterialExpressionNamedRerouteDeclaration_139 {
   class=MaterialExpressionNamedRerouteDeclaration
   inputs [None {from=MaterialExpressionConvert_1}]
   properties {name="Clear Coat ##2 (V3)", nodeColor=<rgba 1 0 0.5859 1>}
  }
  MaterialExpressionNamedRerouteDeclaration_14 {
   class=MaterialExpressionNamedRerouteDeclaration
   inputs [None {from=MaterialExpressionMultiply_9}]
   properties {name="Interior Normal Tiling (V2)", nodeColor=<rgba 0.1719 0 1 1>}
  }
  MaterialExpressionNamedRerouteDeclaration_16 {
   class=MaterialExpressionNamedRerouteDeclaration
   inputs [None {from=MaterialExpressionScalarParameter_3}]
   properties {name="IOR (S)", nodeColor=<rgba 0 1 0.0859 1>}
  }
  MaterialExpressionNamedRerouteDeclaration_17 {
   class=MaterialExpressionNamedRerouteDeclaration
   inputs [None {from=MaterialExpressionScalarParameter_8}]
   properties {name="Fresnel Exponent (S)", nodeColor=<rgba 1 0 0.2344 1>}
  }
  MaterialExpressionNamedRerouteDeclaration_18 {
   class=MaterialExpressionNamedRerouteDeclaration
   inputs [None {from=MaterialExpressionScalarParameter_9}]
   properties {name="Interior Reflection Roughness (S)", nodeColor=<rgba 0 0.5547 1 1>}
  }
  MaterialExpressionNamedRerouteDeclaration_19 {
   class=MaterialExpressionNamedRerouteDeclaration
   inputs [None {from=MaterialExpressionScalarParameter_10}]
   properties {name="Interior Refraction Roughness (S)", nodeColor=<rgba 0 1 0.1565 1>}
  }
  MaterialExpressionNamedRerouteDeclaration_190 {
   class=MaterialExpressionNamedRerouteDeclaration
   inputs [None {from=MaterialExpressionFunctionInput_5}]
   properties {name="Material In (MA)", nodeColor=<rgba 0.4531 1 0 1>}
  }
  MaterialExpressionNamedRerouteDeclaration_2 {
   class=MaterialExpressionNamedRerouteDeclaration
   inputs [None {from=MaterialExpressionLinearInterpolate_8}]
   properties {name="TIR (S)", nodeColor=<rgba 0 0.7422 1 1>}
  }
  MaterialExpressionNamedRerouteDeclaration_20 {
   class=MaterialExpressionNamedRerouteDeclaration
   inputs [None {from=MaterialExpressionMaterialFunctionCall_2, output=Vector}]
   properties {name="Interior Refraction Vector WS (V3)", nodeColor=<rgba 0.8047 0 1 1>}
  }
  MaterialExpressionNamedRerouteDeclaration_21 {
   class=MaterialExpressionNamedRerouteDeclaration
   inputs [None {from=MaterialExpressionReflectionVectorWS_0}]
   properties {name="Interior Reflection Vector WS (V3)", nodeColor=<rgba 0 1 0.4844 1>}
  }
  MaterialExpressionNamedRerouteDeclaration_22 {
   class=MaterialExpressionNamedRerouteDeclaration
   inputs [None {from=MaterialExpressionVectorParameter_4, output=RGB}]
   properties {name="Thin Film ##1 (V3)", nodeColor=<rgba 0 1 0.5043 1>}
  }
  MaterialExpressionNamedRerouteDeclaration_23 {
   class=MaterialExpressionNamedRerouteDeclaration
   inputs [None {from=MaterialExpressionVectorParameter_0, output=RGB}]
   properties {name="Interior Skylight Tint (V3)", nodeColor=<rgba 1 0.1641 0 1>}
  }
  MaterialExpressionNamedRerouteDeclaration_24 {
   class=MaterialExpressionNamedRerouteDeclaration
   inputs [None {from=MaterialExpressionVectorParameter_1, output=RGB}]
   properties {name="BaseColor (V3)", nodeColor=<rgba 0.4522 0 1 1>}
  }
  MaterialExpressionNamedRerouteDeclaration_25 {
   class=MaterialExpressionNamedRerouteDeclaration
   inputs [None {from=MaterialExpressionScalarParameter_5}]
   properties {name="Metallic (S)", nodeColor=<rgba 0.4766 1 0 1>}
  }
  MaterialExpressionNamedRerouteDeclaration_26 {
   class=MaterialExpressionNamedRerouteDeclaration
   inputs [None {from=MaterialExpressionScalarParameter_6}]
   properties {name="Specular (S)", nodeColor=<rgba 1 0 0.7969 1>}
  }
  MaterialExpressionNamedRerouteDeclaration_27 {
   class=MaterialExpressionNamedRerouteDeclaration
   inputs [None {from=MaterialExpressionScalarParameter_7}]
   properties {name="Roughness (S)", nodeColor=<rgba 0 1 0.8828 1>}
  }
  MaterialExpressionNamedRerouteDeclaration_28 {
   class=MaterialExpressionNamedRerouteDeclaration
   inputs [None {from=MaterialExpressionVectorParameter_5, output=RGB}]
   properties {name="Thin Film ##2 (V3)", nodeColor=<rgba 0.2957 1 0 1>}
  }
  MaterialExpressionNamedRerouteDeclaration_29 {
   class=MaterialExpressionNamedRerouteDeclaration
   inputs [None {from=MaterialExpressionFunctionInput_2}]
   properties {name="GMap In (S)", nodeColor=<rgba 1 0.1875 0 1>}
  }
  MaterialExpressionNamedRerouteDeclaration_3 {
   class=MaterialExpressionNamedRerouteDeclaration
   inputs [None {from=MaterialExpressionNormalize_0}]
   properties {name="Surface Normal WS (V3)", nodeColor=<rgba 0 0.2266 1 1>}
  }
  MaterialExpressionNamedRerouteDeclaration_31 {
   class=MaterialExpressionNamedRerouteDeclaration
   inputs [None {from=MaterialExpressionScalarParameter_16}]
   properties {name="GMap Exp (S)", nodeColor=<rgba 0 0.7891 1 1>}
  }
  MaterialExpressionNamedRerouteDeclaration_32 {
   class=MaterialExpressionNamedRerouteDeclaration
   inputs [None {from=MaterialExpressionMaterialFunctionCall_6, output="Result (V3)"}]
   properties {name="Interior Normal TS (V3)", nodeColor=<rgba 0 0.2266 1 1>}
  }
  MaterialExpressionNamedRerouteDeclaration_33 {
   class=MaterialExpressionNamedRerouteDeclaration
   inputs [None {from=MaterialExpressionScalarParameter_12}]
   properties {name="Interior Normal Intensity (S)", nodeColor=<rgba 0.5703 0 1 1>}
  }
  MaterialExpressionNamedRerouteDeclaration_4 {
   class=MaterialExpressionNamedRerouteDeclaration
   inputs [None {from=MaterialExpressionTextureObjectParameter_0}]
   properties {name="Detail Normal Texture (TO)", nodeColor=<rgba 1 0 0.75 1>}
  }
  MaterialExpressionNamedRerouteDeclaration_5 {
   class=MaterialExpressionNamedRerouteDeclaration
   inputs [None {from=MaterialExpressionVectorParameter_2, output=RGBA}]
   properties {name="Clear Coat (V4)", nodeColor=<rgba 1 0 0.5859 1>}
  }
  MaterialExpressionNamedRerouteDeclaration_6 {
   class=MaterialExpressionNamedRerouteDeclaration
   inputs [None {from=MaterialExpressionMaterialFunctionCall_1, output=Result}]
   properties {name="Surface Normal TS (V3)", nodeColor=<rgba 0 0.2266 1 1>}
  }
  MaterialExpressionNamedRerouteDeclaration_7 {
   class=MaterialExpressionNamedRerouteDeclaration
   inputs [None {from=MaterialExpressionOneMinus_0}]
   properties {name="Fresnel Mask (S)", nodeColor=<rgba 1 0.0469 0 1>}
  }
  MaterialExpressionNamedRerouteDeclaration_8 {
   class=MaterialExpressionNamedRerouteDeclaration
   inputs [None {from=MaterialExpressionMultiply_6}]
   properties {name="Interior Result (V3)", nodeColor=<rgba 0 0.2734 1 1>}
  }
  MaterialExpressionNamedRerouteDeclaration_9 {
   class=MaterialExpressionNamedRerouteDeclaration
   inputs [None {from=MaterialExpressionGetMaterialAttributes_0, output=Normal}]
   properties {name="Normal In TS (V3)", nodeColor=<rgba 0 0.2266 1 1>}
  }
  MaterialExpressionNamedRerouteUsage_0 {class=MaterialExpressionNamedRerouteUsage}
  MaterialExpressionNamedRerouteUsage_1 {class=MaterialExpressionNamedRerouteUsage}
  MaterialExpressionNamedRerouteUsage_10 {class=MaterialExpressionNamedRerouteUsage}
  MaterialExpressionNamedRerouteUsage_11 {class=MaterialExpressionNamedRerouteUsage}
  MaterialExpressionNamedRerouteUsage_12 {class=MaterialExpressionNamedRerouteUsage}
  MaterialExpressionNamedRerouteUsage_13 {class=MaterialExpressionNamedRerouteUsage}
  MaterialExpressionNamedRerouteUsage_14 {class=MaterialExpressionNamedRerouteUsage}
  MaterialExpressionNamedRerouteUsage_15 {class=MaterialExpressionNamedRerouteUsage}
  MaterialExpressionNamedRerouteUsage_16 {class=MaterialExpressionNamedRerouteUsage}
  MaterialExpressionNamedRerouteUsage_17 {class=MaterialExpressionNamedRerouteUsage}
  MaterialExpressionNamedRerouteUsage_18 {class=MaterialExpressionNamedRerouteUsage}
  MaterialExpressionNamedRerouteUsage_19 {class=MaterialExpressionNamedRerouteUsage}
  MaterialExpressionNamedRerouteUsage_2 {class=MaterialExpressionNamedRerouteUsage}
  MaterialExpressionNamedRerouteUsage_23 {class=MaterialExpressionNamedRerouteUsage}
  MaterialExpressionNamedRerouteUsage_24 {class=MaterialExpressionNamedRerouteUsage}
  MaterialExpressionNamedRerouteUsage_25 {class=MaterialExpressionNamedRerouteUsage}
  MaterialExpressionNamedRerouteUsage_26 {class=MaterialExpressionNamedRerouteUsage}
  MaterialExpressionNamedRerouteUsage_27 {class=MaterialExpressionNamedRerouteUsage}
  MaterialExpressionNamedRerouteUsage_28 {class=MaterialExpressionNamedRerouteUsage}
  MaterialExpressionNamedRerouteUsage_29 {class=MaterialExpressionNamedRerouteUsage}
  MaterialExpressionNamedRerouteUsage_3 {class=MaterialExpressionNamedRerouteUsage}
  MaterialExpressionNamedRerouteUsage_30 {class=MaterialExpressionNamedRerouteUsage}
  MaterialExpressionNamedRerouteUsage_31 {class=MaterialExpressionNamedRerouteUsage}
  MaterialExpressionNamedRerouteUsage_32 {class=MaterialExpressionNamedRerouteUsage}
  MaterialExpressionNamedRerouteUsage_33 {class=MaterialExpressionNamedRerouteUsage}
  MaterialExpressionNamedRerouteUsage_34 {class=MaterialExpressionNamedRerouteUsage}
  MaterialExpressionNamedRerouteUsage_35 {class=MaterialExpressionNamedRerouteUsage}
  MaterialExpressionNamedRerouteUsage_36 {class=MaterialExpressionNamedRerouteUsage}
  MaterialExpressionNamedRerouteUsage_37 {class=MaterialExpressionNamedRerouteUsage}
  MaterialExpressionNamedRerouteUsage_38 {class=MaterialExpressionNamedRerouteUsage}
  MaterialExpressionNamedRerouteUsage_4 {class=MaterialExpressionNamedRerouteUsage}
  MaterialExpressionNamedRerouteUsage_40 {class=MaterialExpressionNamedRerouteUsage}
  MaterialExpressionNamedRerouteUsage_43 {class=MaterialExpressionNamedRerouteUsage}
  MaterialExpressionNamedRerouteUsage_44 {class=MaterialExpressionNamedRerouteUsage}
  MaterialExpressionNamedRerouteUsage_45 {class=MaterialExpressionNamedRerouteUsage}
  MaterialExpressionNamedRerouteUsage_46 {class=MaterialExpressionNamedRerouteUsage}
  MaterialExpressionNamedRerouteUsage_47 {class=MaterialExpressionNamedRerouteUsage}
  MaterialExpressionNamedRerouteUsage_48 {class=MaterialExpressionNamedRerouteUsage}
  MaterialExpressionNamedRerouteUsage_5 {class=MaterialExpressionNamedRerouteUsage}
  MaterialExpressionNamedRerouteUsage_7 {class=MaterialExpressionNamedRerouteUsage}
  MaterialExpressionNamedRerouteUsage_8 {class=MaterialExpressionNamedRerouteUsage}
  MaterialExpressionNamedRerouteUsage_9 {class=MaterialExpressionNamedRerouteUsage}
  MaterialExpressionNormalize_0 {class=MaterialExpressionNormalize, inputs [VectorInput {from=MaterialExpressionTransform_1, output=""}]}
  MaterialExpressionNormalize_1 {class=MaterialExpressionNormalize, inputs [VectorInput {from=MaterialExpressionTransform_0, output=""}]}
  MaterialExpressionOneMinus_0 {class=MaterialExpressionOneMinus, inputs [None {from=MaterialExpressionFresnel_0}]}
  MaterialExpressionPower_1 {
   class=MaterialExpressionPower
   inputs [
    Base {from=MaterialExpressionNamedRerouteUsage_40}
    Exp {from=MaterialExpressionNamedRerouteUsage_43}
   ]
  }
  MaterialExpressionReflectionVectorWS_0 {
   class=MaterialExpressionReflectionVectorWS
   inputs [CustomWorldNormal {from=MaterialExpressionNamedRerouteUsage_0, output=""}]
   properties {}
  }
  MaterialExpressionScalarParameter_0 {
   class=MaterialExpressionScalarParameter
   properties {defaultValue=50, group=PixieParts | Materials | Crystal, parameterName=Crystal - Interior Depth, sortPriority=330}
  }
  MaterialExpressionScalarParameter_10 {
   class=MaterialExpressionScalarParameter
   properties {defaultValue=0.2, group=PixieParts | Materials | Crystal, parameterName=Crystal - Interior Refraction Roughness, sortPriority=370}
  }
  MaterialExpressionScalarParameter_11 {
   class=MaterialExpressionScalarParameter
   properties {defaultValue=1, group=PixieParts | Materials | Crystal, parameterName=Crystal - Normal Intensity, sortPriority=230}
  }
  MaterialExpressionScalarParameter_12 {
   class=MaterialExpressionScalarParameter
   properties {defaultValue=1, group=PixieParts | Materials | Crystal, parameterName=Crystal - Interior Normal Intensity, sortPriority=350}
  }
  MaterialExpressionScalarParameter_16 {
   class=MaterialExpressionScalarParameter
   properties {defaultValue=1, group=PixieParts | Materials | Crystal, parameterName=Crystal - GMap Exp, sliderMax=10, sortPriority=240}
  }
  MaterialExpressionScalarParameter_18 {
   class=MaterialExpressionScalarParameter
   properties {defaultValue=1, group=PixieParts | Materials | Crystal, parameterName=Crystal - Normal V Tiling Factor, sortPriority=220}
  }
  MaterialExpressionScalarParameter_2 {
   class=MaterialExpressionScalarParameter
   properties {defaultValue=1, group=PixieParts | Materials | Crystal, parameterName=Crystal - Interior Normal Tiling Gain, sortPriority=340}
  }
  MaterialExpressionScalarParameter_3 {
   class=MaterialExpressionScalarParameter
   properties {defaultValue=1.5, group=PixieParts | Materials | Crystal, parameterName=Crystal - IOR, sortPriority=320}
  }
  MaterialExpressionScalarParameter_4 {
   class=MaterialExpressionScalarParameter
   properties {defaultValue=3, group=PixieParts | Materials | Crystal, parameterName=Crystal - Normal Tiling, sortPriority=210}
  }
  MaterialExpressionScalarParameter_5 {
   class=MaterialExpressionScalarParameter
   properties {defaultValue=0.75, group=PixieParts | Materials | Crystal, parameterName=Crystal - Metallic, sliderMax=1, sortPriority=20}
  }
  MaterialExpressionScalarParameter_6 {
   class=MaterialExpressionScalarParameter
   properties {defaultValue=1, group=PixieParts | Materials | Crystal, parameterName=Crystal - Specular, sliderMax=1, sortPriority=30}
  }
  MaterialExpressionScalarParameter_7 {
   class=MaterialExpressionScalarParameter
   properties {defaultValue=0.1, group=PixieParts | Materials | Crystal, parameterName=Crystal - Roughness, sliderMax=1, sortPriority=40}
  }
  MaterialExpressionScalarParameter_8 {
   class=MaterialExpressionScalarParameter
   properties {defaultValue=5, group=PixieParts | Materials | Crystal, parameterName=Crystal - Fresnel Exponent, sortPriority=310}
  }
  MaterialExpressionScalarParameter_9 {
   class=MaterialExpressionScalarParameter
   properties {defaultValue=0.1, group=PixieParts | Materials | Crystal, parameterName=Crystal - Interior Reflection Roughness, sortPriority=360}
  }
  MaterialExpressionSetMaterialAttributes_0 {
   class=MaterialExpressionSetMaterialAttributes
   inputs [
    MaterialAttributes {from=MaterialExpressionNamedRerouteUsage_35, output=""}
    BaseColor {from=MaterialExpressionNamedRerouteUsage_31, output=""}
    Metallic {from=MaterialExpressionNamedRerouteUsage_32, output=""}
    Specular {from=MaterialExpressionNamedRerouteUsage_33, output=""}
    Roughness {from=MaterialExpressionNamedRerouteUsage_34, output=""}
    EmissiveColor {from=MaterialExpressionNamedRerouteUsage_13, output=""}
    Normal {from=MaterialExpressionNamedRerouteUsage_11, output=""}
   ]
   properties {attributeSetTypes [
     #2
     #3
     #4
     #5
     #6
     #1
    ]}
  }
  MaterialExpressionSetMaterialAttributes_3 {
   class=MaterialExpressionSetMaterialAttributes
   inputs [
    MaterialAttributes {from=MaterialExpressionSetMaterialAttributes_0, output=""}
    Tangent {from=MaterialExpressionNamedRerouteUsage_36, output=""}
    Displacement {from=MaterialExpressionNamedRerouteDeclaration_119, output=""}
    SubsurfaceColor {from=MaterialExpressionNamedRerouteDeclaration_139, output=""}
    BentNormal {from=MaterialExpressionNamedRerouteUsage_14, output=""}
    TransmittanceColor {from=MaterialExpressionNamedRerouteUsage_16, output=""}
   ]
   properties {attributeSetTypes [
     #7
     #8
     #9
     #10
     #11
    ]}
  }
  MaterialExpressionSkyLightEnvMapSample_0 {
   class=MaterialExpressionSkyLightEnvMapSample
   inputs [
    Direction {from=MaterialExpressionNamedRerouteUsage_29, output=""}
    Roughness {from=MaterialExpressionNamedRerouteUsage_27, output=""}
   ]
   properties {}
  }
  MaterialExpressionSkyLightEnvMapSample_1 {
   class=MaterialExpressionSkyLightEnvMapSample
   inputs [
    Direction {from=MaterialExpressionNamedRerouteUsage_28, output=""}
    Roughness {from=MaterialExpressionNamedRerouteUsage_26, output=""}
   ]
   properties {}
  }
  MaterialExpressionTextureCoordinate_0 {class=MaterialExpressionTextureCoordinate}
  MaterialExpressionTextureCoordinate_1 {class=MaterialExpressionTextureCoordinate}
  MaterialExpressionTextureObjectParameter_0 {
   class=MaterialExpressionTextureObjectParameter
   inputs [
    UVs {from="", output=""}
    Apply View MipBias {from="", output=""}
   ]
   properties {
    group=PixieParts | Materials | Crystal
    parameterName=Crystal - Normal
    samplerType=SAMPLERTYPE_Normal
    sortPriority=200
    texture=/Game/Global/Materials/MaterialGym/Materials/TechArtPatterns/Textures/T_Triangles_N
   }
  }
  MaterialExpressionTextureSample_0 {
   class=MaterialExpressionTextureSample
   inputs [
    UVs {from=MaterialExpressionMultiply_2}
    Tex {from=MaterialExpressionNamedRerouteUsage_5}
    Apply View MipBias {from=""}
   ]
   properties {
    samplerSource=SSM_Wrap_WorldGroupSettings
    samplerType=SAMPLERTYPE_Normal
    texture=&T_Triangles_N
   }
  }
  MaterialExpressionTextureSample_2 {
   class=MaterialExpressionTextureSample
   inputs [
    UVs {from=MaterialExpressionMaterialFunctionCall_5, output=Result}
    Tex {from=MaterialExpressionNamedRerouteUsage_44}
    Apply View MipBias {from=""}
   ]
   properties {
    samplerSource=SSM_Wrap_WorldGroupSettings
    samplerType=SAMPLERTYPE_Normal
    texture=&T_Triangles_N
   }
  }
  MaterialExpressionTransform_0 {class=MaterialExpressionTransform, inputs [None {from=MaterialExpressionNamedRerouteUsage_46}]}
  MaterialExpressionTransform_1 {class=MaterialExpressionTransform, inputs [None {from=MaterialExpressionNamedRerouteUsage_19}]}
  MaterialExpressionVectorParameter_0 {
   class=MaterialExpressionVectorParameter
   properties {
    group=PixieParts | Materials | Crystal
    parameterName=Crystal - Interior Skylight Tint
    sortPriority=380
    defaultValue=<rgba 0.151 0.2229 0.2396 1>
   }
  }
  MaterialExpressionVectorParameter_1 {
   class=MaterialExpressionVectorParameter
   properties {group=PixieParts | Materials | Crystal, parameterName=Crystal - Color, sortPriority=10, defaultValue=<rgba 0.4271 0.4271 0.4271 0>}
  }
  MaterialExpressionVectorParameter_2 {
   class=MaterialExpressionVectorParameter
   properties {
    group=PixieParts | Materials | Crystal
    parameterName=Crystal - Clear Coat
    sortPriority=110
    channelNames {a=Top Coat Material Roughness Mix, b=Under Coat Roughness, g=Under Coat Metallic Mult, r=Clear Coat Intensity}
    defaultValue=<rgba 0 1 0.1 0>
   }
  }
  MaterialExpressionVectorParameter_3 {
   class=MaterialExpressionVectorParameter
   properties {
    group=PixieParts | Materials | Crystal
    parameterName=Crystal - Metal LUT
    sortPriority=100
    channelNames {a=n/a, b=n/a, g=LUT Channel, r=LUT Intensity}
   }
  }
  MaterialExpressionVectorParameter_4 {
   class=MaterialExpressionVectorParameter
   properties {
    group=PixieParts | Materials | Crystal
    parameterName="Crystal - Thin Film ##1"
    sortPriority=120
    channelNames {a=n/a, b=Roughness Influence, g=Thin Film Exponent, r=Thin Film Intensity}
    defaultValue=<rgba 0 2 0 0>
   }
  }
  MaterialExpressionVectorParameter_5 {
   class=MaterialExpressionVectorParameter
   properties {
    group=PixieParts | Materials | Crystal
    parameterName="Crystal - Thin Film ##2"
    sortPriority=130
    channelNames {a=n/a, b=Thin Film Offset, g=Thin Film Scale, r=Thin Film Wrap}
    defaultValue=<rgba 0.5 1 0 0>
   }
  }
 }}
