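# Material function for the PixieParts Lava player body that builds three decal materials from one incoming material attributes set.
# A shared vector parameter supplies metallic, specular and roughness, while three color parameters provide a separate base color for each output.
# All three outputs also share constant "disabled" values for Metal LUT, clear coat and thin film, written into attribute slots such as Tangent, Displacement and BentNormal.
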MaterialFunction /BRCosmetics/Characters/Player/Male/Medium/Bodies/M_MED_PixieParts_Lava/Materials/MaterialFunctions/MF_PixieParts_DecalMaterials
graph {nodes [
  n1 = Convert!() {desc=disabled, convertOutputs [{type=Scalar}, {type=Vector3, defaultValue=<rgba 1 0.1 0 1>}]}
  n2 = SetMaterialAttributes!(
   MaterialAttributes=SetMaterialAttributes!(MaterialAttributes=NamedRerouteUsage(), Metallic=NamedRerouteUsage(), Specular=NamedRerouteUsage(), Roughness=NamedRerouteUsage()) {attributeSetTypes [#1, #2, #3]}
   Tangent=NamedRerouteDeclaration(None=Convert() {desc=disabled, convertOutputs [{type=Vector3}]}) {name="Metal LUT (V3)", nodeColor=<rgba 0 0.9062 1 1>}
   Displacement=NamedRerouteDeclaration(None=@n1) {name="Clear Coat ##1 (S)", nodeColor=<rgba 1 0 0.5859 1>}
   SubsurfaceColor=NamedRerouteDeclaration(None=@n1) {name="Clear Coat ##2 (V3)", nodeColor=<rgba 1 0 0.5859 1>}
   BentNormal=NamedRerouteDeclaration(None=Convert() {desc=disabled, convertOutputs [{type=Vector3, defaultValue=<rgba 0 2 0 1>}]}) {name="Thin Film ##1 (V3)", nodeColor=<rgba 0 1 0.5043 1>}
   TransmittanceColor=NamedRerouteDeclaration(None=Convert() {desc=disabled, convertOutputs [{type=Vector3, defaultValue=<rgba 0.5 1 0 1>}]}) {name="Thin Film ##2 (V3)", nodeColor=<rgba 0.2957 1 0 1>}
  ) {attributeSetTypes [
    #4
    #5
    #6
    #7
    #8
   ]}
  FunctionOutput(None=SetMaterialAttributes!(MaterialAttributes=@n2, BaseColor=NamedRerouteUsage()) {attributeSetTypes [#9]}) {outputName="Decal Material 1 Out (MA)"}
  FunctionOutput(None=SetMaterialAttributes(MaterialAttributes=@n2, BaseColor=NamedRerouteUsage()) {attributeSetTypes [#9]}) {outputName="Decal Material 2 Out (MA)", sortPriority=10}
  FunctionOutput(None=SetMaterialAttributes(MaterialAttributes=@n2, BaseColor=NamedRerouteUsage()) {attributeSetTypes [#9]}) {outputName="Decal Material 3 Out (MA)", sortPriority=20}
  n3 = VectorParameter() {
   group=PixieParts | Materials | Decals
   parameterName=Decals - Color1
   sortPriority=20
   channelNames {a=n/a, b=Tint B, g=Tint G, r=Tint R}
   defaultValue=<rgba 0.5 0.5 0.5 0>
  }
  NamedRerouteDeclaration(None=@n3.RGB) {name="Decals - Color1 (V3)", nodeColor=<rgba 0 0.0859 1 1>}
  n4 = VectorParameter() {
   group=PixieParts | Materials | Decals
   parameterName=Decals - Color2
   sortPriority=30
   channelNames {a=n/a, b=Tint B, g=Tint G, r=Tint R}
   defaultValue=<rgba 0.5 0.5 0.5 0>
  }
  NamedRerouteDeclaration(None=@n4.RGB) {name="Decals - Color2 (V3)", nodeColor=<rgba 0 0.0859 1 1>}
  n5 = VectorParameter() {
   group=PixieParts | Materials | Decals
   parameterName=Decals - Color3
   sortPriority=40
   channelNames {a=n/a, b=Tint B, g=Tint G, r=Tint R}
   defaultValue=<rgba 0.5 0.5 0.5 0>
  }
  NamedRerouteDeclaration(None=@n5.RGB) {name="Decals - Color3 (V3)", nodeColor=<rgba 0 0.0859 1 1>}
  n6 = VectorParameter() {
   group=PixieParts | Materials | Decals
   parameterName=Decals Properties
   sortPriority=10
   channelNames {a=n/a, b=Roughness, g=Specular, r=Metallic}
   defaultValue=<rgba 0 0.5 0.5 0>
  }
  NamedRerouteDeclaration(None=@n6.Metallic) {name="Decals - Metallic (S)", nodeColor=<rgba 1 0.1172 0 1>}
  NamedRerouteDeclaration(None=@n6.Specular) {name="Decals - Specular (S)", nodeColor=<rgba 1 0.1172 0 1>}
  NamedRerouteDeclaration(None=@n6.Roughness) {name="Decals - Roughness (S)", nodeColor=<rgba 1 0.1172 0 1>}
  NamedRerouteDeclaration(None=FunctionInput() {inputName=Material In, inputType=FunctionInput_MaterialAttributes}) {name="Material In (MA)", nodeColor=<rgba 0.4531 1 0 1>}
 ]}
