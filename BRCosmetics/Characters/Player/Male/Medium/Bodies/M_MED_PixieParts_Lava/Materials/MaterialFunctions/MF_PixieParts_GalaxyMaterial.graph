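# Galaxy look material function for the PixieParts Lava character body, reusing the VileSting galaxy function from another cosmetic.
# It zeroes the incoming emissive, runs the galaxy function with a full mask, then blends emissive toward bright blue by a cubed scalar, apparently the optional Fresnel input.
# It overwrites Tangent, Displacement, SubsurfaceColor, BentNormal and TransmittanceColor with constants labeled Metal LUT, Clear Coat and Thin Film, marked disabled.
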MaterialFunction /BRCosmetics/Characters/Player/Male/Medium/Bodies/M_MED_PixieParts_Lava/Materials/MaterialFunctions/MF_PixieParts_GalaxyMaterial
graph {nodes [
  n1 = NamedRerouteUsage()
  n2 = Constant()
  n3 = GetMaterialAttributes(None=MaterialFunctionCall!(
    Material=SetMaterialAttributes!(MaterialAttributes=@n1, EmissiveColor=@n2) {attributeSetTypes [#1]}
    TechArtMask=Constant() {r=1}
   ) {materialFunction=/BRCosmetics/Characters/Player/Female/Medium/Bodies/F_MED_VileSting_Crude/Skins/Point/Materials/MF_GalaxyLook_VileSting}.Result) {attributeGetTypes [#1]}
  n4 = Multiply(A=Constant3Vector() {constant=<rgba 0 0.4371 1 1>}) {constB=1.2}
  n5 = Power(Base=NamedRerouteUsage()) {constExponent=3}
  n6 = Convert!() {desc=disabled, convertOutputs [{type=Scalar}, {type=Vector3, defaultValue=<rgba 1 0.1 0 1>}]}
  n7 = SetMaterialAttributes!(
   MaterialAttributes=SetMaterialAttributes!(MaterialAttributes=@n3.MaterialAttributes, EmissiveColor=LinearInterpolate(A=@n3.EmissiveColor, B=@n4, Alpha=@n5)) {attributeSetTypes [#1]}
   Tangent=NamedRerouteDeclaration(None=Convert() {desc=disabled, convertOutputs [{type=Vector3}]}) {name="Metal LUT (V3)", nodeColor=<rgba 0 0.9062 1 1>}
   Displacement=NamedRerouteDeclaration(None=@n6) {name="Clear Coat ##1 (S)", nodeColor=<rgba 1 0 0.5859 1>}
   SubsurfaceColor=NamedRerouteDeclaration(None=@n6) {name="Clear Coat ##2 (V3)", nodeColor=<rgba 1 0 0.5859 1>}
   BentNormal=NamedRerouteDeclaration(None=Convert() {desc=disabled, convertOutputs [{type=Vector3, defaultValue=<rgba 0 2 0 1>}]}) {name="Thin Film ##1 (V3)", nodeColor=<rgba 0 1 0.5043 1>}
   TransmittanceColor=NamedRerouteDeclaration(None=Convert() {desc=disabled, convertOutputs [{type=Vector3, defaultValue=<rgba 0.5 1 0 1>}]}) {name="Thin Film ##2 (V3)", nodeColor=<rgba 0.2957 1 0 1>}
  ) {attributeSetTypes [
    #2
    #3
    #4
    #5
    #6
   ]}
  FunctionOutput(None=@n7) {outputName="Material Out (MA)"}
  NamedRerouteDeclaration(None=FunctionInput() {inputName=Material In, inputType=FunctionInput_MaterialAttributes}) {name="Material In (MA)", nodeColor=<rgba 0.4531 1 0 1>}
  n8 = FunctionInput() {bUsePreviewValueAsDefault=true, inputName=Fresnel In, inputType=FunctionInput_Scalar, sortPriority=50, previewValue=<0 1 0 1>}
  NamedRerouteDeclaration(None=@n8) {name="Fresnel In (S)", nodeColor=<rgba 0 1 0.0156 1>}
 ]}
