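# Material function that layers a tintable gold metal surface onto incoming material attributes for the PixieParts Lava player body cosmetic.
# Base color blends dark, light and facing-up gold colors using smoothstep masks, including one from pre-skinned normal Z, with GMap and decal-mask inputs alongside offset and gain controls.
# It sets full metallic, low roughness and no emission, then packs clear coat, metal LUT and thin-film parameters into repurposed attributes such as Tangent, Displacement and SubsurfaceColor.
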MaterialFunction /BRCosmetics/Characters/Player/Male/Medium/Bodies/M_MED_PixieParts_Lava/Materials/MaterialFunctions/MF_PixieParts_GoldMaterial
graph {nodes [
  n1 = Saturate(None=Add(A=NamedRerouteUsage(), B=NamedRerouteUsage()))
  n2 = NamedRerouteUsage()
  n3 = NamedRerouteUsage()
  n4 = LinearInterpolate(
   A=NamedRerouteUsage()
   B=NamedRerouteUsage()
   Alpha=SmoothStep(Min=NamedRerouteUsage(), Max=NamedRerouteUsage(), Value=NamedRerouteUsage()) {constMin=0.5}
  )
  n5 = NamedRerouteUsage()
  n6 = VertexInterpolator(VS=PreSkinnedNormal())
  n7 = SmoothStep(Min=NamedRerouteUsage(), Max=NamedRerouteUsage(), Value=Saturate(None=ComponentMask(None=@n6.PS) {b=true})) {constMin=0.5}
  n8 = Constant() {r=1}
  n9 = SetMaterialAttributes!(
   MaterialAttributes=NamedRerouteUsage()
   BaseColor=Multiply(A=LinearInterpolate(A=@n1, B=@n2, Alpha=@n3) {constA=1, constB=0.68}, B=LinearInterpolate(A=@n4, B=@n5, Alpha=@n7))
   Specular=@n8
   Roughness=LinearInterpolate(Alpha=NamedRerouteUsage()) {constA=0.25, constB=0.1}
   Metallic=@n8
   EmissiveColor=Constant()
  ) {attributeSetTypes [
    #1
    #2
    #3
    #4
    #5
   ]}
  MaterialExpressionNamedRerouteUsage_10 = NamedRerouteUsage()
  n10 = Convert(None=@MaterialExpressionNamedRerouteUsage_10) {
   convertInputs [{
     type=Vector4
     defaultValue=<rgba 0 0 0 1>
     expressionInput {
      expression=&self:MaterialExpressionNamedRerouteUsage_10
      inputName=None
      mask=0
      maskA=0
      maskB=0
      maskG=0
      maskR=0
      outputIndex=0
     }
    }]
   convertOutputs [{type=Scalar}, {type=Vector3}]
  }
  n11 = SetMaterialAttributes!(
   MaterialAttributes=@n9
   Tangent=NamedRerouteUsage()
   Displacement=NamedRerouteDeclaration(None=@n10) {name="Clear Coat ##1 (S)", nodeColor=<rgba 1 0 0.5859 1>}
   SubsurfaceColor=NamedRerouteDeclaration(None=@n10) {name="Clear Coat ##2 (V3)", nodeColor=<rgba 1 0 0.5859 1>}
   BentNormal=NamedRerouteUsage()
   TransmittanceColor=NamedRerouteUsage()
  ) {attributeSetTypes [
    #6
    #7
    #8
    #9
    #10
   ]}
  FunctionOutput(None=@n11) {outputName="Material Out (MA)"}
  n12 = ScalarParameter() {defaultValue=0.3, group=PixieParts | Materials | Gold, parameterName=Gold - GMap Offset, sliderMax=1, sliderMin=-1, sortPriority=10}
  NamedRerouteDeclaration(None=@n12) {name="Gold GMap Offset (S)", nodeColor=<rgba 0 1 0.6953 1>}
  n13 = ScalarParameter() {defaultValue=0.68, group=PixieParts | Materials | Gold, parameterName=Gold - Decal Gain, sliderMax=1, sortPriority=20}
  NamedRerouteDeclaration(None=@n13) {name="Gold Decal Gain (S)", nodeColor=<rgba 1 0.375 0 1>}
  n14 = VectorParameter() {group=PixieParts | Materials | Gold, parameterName=Gold - Color Dark, sortPriority=100, defaultValue=<rgba 0.6514 0.3372 0.0561 1>}
  NamedRerouteDeclaration(None=@n14.RGB) {name="Gold Color - Dark (V3)", nodeColor=<rgba 1 0.0348 0 1>}
  n15 = VectorParameter() {group=PixieParts | Materials | Gold, parameterName=Gold - Color Light, sortPriority=110, defaultValue=<rgba 0.7231 0.4179 0.0999 1>}
  NamedRerouteDeclaration(None=@n15.RGB) {name="Gold Color - Light (V3)", nodeColor=<rgba 1 0.8696 0 1>}
  n16 = VectorParameter() {group=PixieParts | Materials | Gold, parameterName=Gold - Color Facing Up, sortPriority=120, defaultValue=<rgba 0.8073 0.5146 0.1598 1>}
  NamedRerouteDeclaration(None=@n16.RGB) {name="Gold Color - Facing Up (V3)", nodeColor=<rgba 1 0.7468 0.1679 1>}
  n17 = VectorParameter() {
   group=PixieParts | Materials | Gold
   parameterName=Gold - Clear Coat
   sortPriority=210
   channelNames {a=Top Coat Material Roughness Mix, b=Under Coat Roughness, g=Under Coat Metallic Mult, r=Clear Coat Intensity}
   defaultValue=<rgba 0 1 0.1 0>
  }
  NamedRerouteDeclaration(None=@n17.RGBA) {name="Clear Coat (V4)", nodeColor=<rgba 1 0 0.5859 1>}
  n18 = FunctionInput() {bUsePreviewValueAsDefault=true, inputName=Decals Mask In, inputType=FunctionInput_Scalar, sortPriority=20, previewValue=<0 0 0 0>}
  NamedRerouteDeclaration(None=@n18) {name="Decals Mask In (S)", nodeColor=<rgba 0 1 0.5078 1>}
  n19 = VectorParameter() {
   group=PixieParts | Materials | Gold
   parameterName=Gold - Metal LUT
   sortPriority=200
   channelNames {a=n/a, b=n/a, g=LUT Channel, r=LUT Intensity}
  }
  NamedRerouteDeclaration(None=@n19.RGB) {name="Metal LUT (V3)", nodeColor=<rgba 0 0.9062 1 1>}
  n20 = VectorParameter() {
   group=PixieParts | Materials | Gold
   parameterName="Gold - Thin Film ##1"
   sortPriority=220
   channelNames {a=n/a, b=Roughness Influence, g=Thin Film Exponent, r=Thin Film Intensity}
   defaultValue=<rgba 0 2 0 0>
  }
  NamedRerouteDeclaration(None=@n20.RGB) {name="Thin Film ##1 (V3)", nodeColor=<rgba 0 1 0.5043 1>}
  n21 = VectorParameter() {
   group=PixieParts | Materials | Gold
   parameterName="Gold - Thin Film ##2"
   sortPriority=230
   channelNames {a=n/a, b=Thin Film Offset, g=Thin Film Scale, r=Thin Film Wrap}
   defaultValue=<rgba 0.5 1 0 0>
  }
  NamedRerouteDeclaration(None=@n21.RGB) {name="Thin Film ##2 (V3)", nodeColor=<rgba 0.2957 1 0 1>}
  NamedRerouteDeclaration(None=FunctionInput() {inputName=Material In, inputType=FunctionInput_MaterialAttributes}) {name="Material In (MA)", nodeColor=<rgba 0.4531 1 0 1>}
  n22 = FunctionInput() {bUsePreviewValueAsDefault=true, inputName=GMap In, inputType=FunctionInput_Scalar, sortPriority=10, previewValue=<1 1 0 1>}
  NamedRerouteDeclaration(None=@n22) {name="GMap In (S)", nodeColor=<rgba 1 0.1875 0 1>}
  NamedRerouteDeclaration(None=ScalarParameter() {defaultValue=0.5, group=PixieParts | Materials | Gold, parameterName=Gold - Curve Min, sliderMax=1, sortPriority=20}) {name="Gold Curve Min (S)", nodeColor=<rgba 0 0.6957 1 1>}
  NamedRerouteDeclaration(None=ScalarParameter() {defaultValue=1, group=PixieParts | Materials | Gold, parameterName=Gold - Curve Max, sliderMax=1, sortPriority=40}) {name="Gold Curve Max (S)", nodeColor=<rgba 0 1 0.0522 1>}
 ]}
