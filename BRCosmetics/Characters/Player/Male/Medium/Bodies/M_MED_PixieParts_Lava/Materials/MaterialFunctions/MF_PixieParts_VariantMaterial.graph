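# Material function that builds PixieParts (Lava) body variant materials by remapping an incoming material attribute set.
# It treats the red channel of BaseColor as greyscale albedo, shaped by an exponent and colored by a saturated tint.
# Specular and Roughness get power-and-multiply adjustments, while Metallic comes from a function input.
# An AO input fades in from white by a mix amount, and named reroutes carry values to the output attributes.
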MaterialFunction /BRCosmetics/Characters/Player/Male/Medium/Bodies/M_MED_PixieParts_Lava/Materials/MaterialFunctions/MF_PixieParts_VariantMaterial
graph {nodes [
  n1 = SetMaterialAttributes!(
   MaterialAttributes=NamedRerouteUsage()
   BaseColor=NamedRerouteUsage()
   Metallic=NamedRerouteUsage()
   Specular=NamedRerouteUsage()
   Roughness=NamedRerouteUsage()
  ) {attributeSetTypes [
    #1
    #2
    #3
    #4
   ]}
  FunctionOutput(None=@n1) {outputName="Material Out (MA)"}
  NamedRerouteDeclaration(None=NamedRerouteUsage()) {name="Metallic Out (S)", nodeColor=<rgba 1 0.5217 0 1>}
  n2 = NamedRerouteUsage()
  n3 = NamedRerouteUsage()
  NamedRerouteDeclaration(None=Multiply(A=Power(Base=@n2, Exp=@n3), B=NamedRerouteUsage())) {name="Specular Out (S)", nodeColor=<rgba 1 0.5217 0 1>}
  n4 = FunctionInput() {bUsePreviewValueAsDefault=true, inputName=Metallic, inputType=FunctionInput_Scalar, sortPriority=120, previewValue=<0 1 0 0>}
  NamedRerouteDeclaration(None=Saturate(None=@n4)) {name="Metallic (S)", nodeColor=<rgba 1 0 0.5391 1>}
  n5 = FunctionInput() {bUsePreviewValueAsDefault=true, inputName=Specular Exp, inputType=FunctionInput_Scalar, sortPriority=130, previewValue=<1 1 0 0>}
  NamedRerouteDeclaration(None=@n5) {name="Specular Exp (S)", nodeColor=<rgba 0 0.5217 1 1>}
  n6 = NamedRerouteUsage()
  n7 = NamedRerouteUsage()
  NamedRerouteDeclaration(None=Multiply(A=Power(Base=@n6, Exp=@n7), B=NamedRerouteUsage())) {name="Roughness Out (S)", nodeColor=<rgba 1 0.5217 0 1>}
  n8 = FunctionInput() {bUsePreviewValueAsDefault=true, inputName=Roughness Exp, inputType=FunctionInput_Scalar, sortPriority=140, previewValue=<1 1 0 0>}
  NamedRerouteDeclaration(None=@n8) {name="Roughness Exp (S)", nodeColor=<rgba 0 0.5217 1 1>}
  n9 = FunctionInput() {bUsePreviewValueAsDefault=true, inputName="Albedo Exp (Pre-Tint)", inputType=FunctionInput_Scalar, sortPriority=100, previewValue=<1 1 1 1>}
  NamedRerouteDeclaration(None=@n9) {name="Albedo Exp (S)", nodeColor=<rgba 0 0.5217 1 1>}
  n10 = GetMaterialAttributes(None=NamedRerouteDeclaration(None=FunctionInput() {inputName=Material In, inputType=FunctionInput_MaterialAttributes}) {name="Material In (MA)", nodeColor=<rgba 0.1016 1 0 1>}) {attributeGetTypes [#1, #3, #4]}
  NamedRerouteDeclaration(None=ComponentMask(None=@n10.BaseColor) {r=true}) {name="Albedo In (S)", nodeColor=<rgba 0.5938 0 1 1>}
  NamedRerouteDeclaration(None=@n10.Specular) {name="Specular In (S)", nodeColor=<rgba 0.5938 0 1 1>}
  NamedRerouteDeclaration(None=@n10.Roughness) {name="Roughness In (S)", nodeColor=<rgba 0.5938 0 1 1>}
  n11 = FunctionInput() {bUsePreviewValueAsDefault=true, inputName=Albedo Tint, sortPriority=110, previewValue=<1 1 1 1>}
  NamedRerouteDeclaration(None=Saturate(None=@n11)) {name="Albedo Tint (V3)", nodeColor=<rgba 0.7287 0 1 1>}
  n12 = FunctionInput() {bUsePreviewValueAsDefault=true, inputName=AO Mix, inputType=FunctionInput_Scalar, sortPriority=20, previewValue=<0 1 0 0>}
  n13 = FunctionInput() {bUsePreviewValueAsDefault=true, inputName=AO In, inputType=FunctionInput_Scalar, sortPriority=10, previewValue=<1 0 0 1>}
  NamedRerouteDeclaration(None=LinearInterpolate(B=@n13, Alpha=Saturate(None=@n12)) {constA=1}) {name="AO (S)", nodeColor=<rgba 1 0 0.0469 1>}
  n14 = Power(Base=NamedRerouteUsage(), Exp=NamedRerouteUsage())
  n15 = NamedRerouteUsage()
  NamedRerouteDeclaration(None=Multiply(A=Multiply(A=@n14, B=@n15), B=NamedRerouteUsage())) {name="BaseColor Out (V3)", nodeColor=<rgba 1 0.5217 0 1>}
 ]}
