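# Material function that builds five variant materials plus an energy material, mask and emission for the PixieParts Lava male medium body cosmetic.
# Each slot blends two source materials (armor and stone inputs, chosen by Material ID), then takes parameter-driven tint, surface, clear coat, metal LUT and thin-film settings, with override sets.
# Energy emission is remapped with two colors.
# Slot 5 adds a visor, which suppresses the energy mask on armor.
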MaterialFunction /BRCosmetics/Characters/Player/Male/Medium/Bodies/M_MED_PixieParts_Lava/Materials/MaterialFunctions/MF_PixieParts_MaterialVariants
graph {expressions {
  MaterialExpressionAdd_0 {
   class=MaterialExpressionAdd
   inputs [
    A {from=MaterialExpressionDotProduct_0}
    B {from=MaterialExpressionNamedRerouteUsage_25}
   ]
  }
  MaterialExpressionBlendMaterialAttributes_0 {
   class=MaterialExpressionBlendMaterialAttributes
   inputs [
    A {from=MaterialExpressionSetMaterialAttributes_8, output=""}
    B {from=MaterialExpressionNamedRerouteUsage_152, output=""}
    Alpha {from=MaterialExpressionNamedRerouteUsage_17, output=""}
   ]
   properties {vertexAttributeBlendType=UseA}
  }
  MaterialExpressionBlendMaterialAttributes_12 {
   class=MaterialExpressionBlendMaterialAttributes
   inputs [
    A {from=MaterialExpressionNamedRerouteUsage_134, output=""}
    B {from=MaterialExpressionNamedRerouteUsage_133, output=""}
    Alpha {from=MaterialExpressionNamedRerouteUsage_128, output=""}
   ]
   properties {vertexAttributeBlendType=UseA}
  }
  MaterialExpressionBlendMaterialAttributes_13 {
   class=MaterialExpressionBlendMaterialAttributes
   inputs [
    A {from=MaterialExpressionNamedRerouteUsage_136, output=""}
    B {from=MaterialExpressionNamedRerouteUsage_135, output=""}
    Alpha {from=MaterialExpressionNamedRerouteUsage_129, output=""}
   ]
   properties {vertexAttributeBlendType=UseA}
  }
  MaterialExpressionBlendMaterialAttributes_14 {
   class=MaterialExpressionBlendMaterialAttributes
   inputs [
    A {from=MaterialExpressionNamedRerouteUsage_138, output=""}
    B {from=MaterialExpressionNamedRerouteUsage_137, output=""}
    Alpha {from=MaterialExpressionNamedRerouteUsage_130, output=""}
   ]
   properties {vertexAttributeBlendType=UseA}
  }
  MaterialExpressionBlendMaterialAttributes_15 {
   class=MaterialExpressionBlendMaterialAttributes
   inputs [
    A {from=MaterialExpressionNamedRerouteUsage_140, output=""}
    B {from=MaterialExpressionNamedRerouteUsage_139, output=""}
    Alpha {from=MaterialExpressionNamedRerouteUsage_131, output=""}
   ]
   properties {vertexAttributeBlendType=UseA}
  }
  MaterialExpressionBlendMaterialAttributes_16 {
   class=MaterialExpressionBlendMaterialAttributes
   inputs [
    A {from=MaterialExpressionNamedRerouteUsage_142, output=""}
    B {from=MaterialExpressionNamedRerouteUsage_141, output=""}
    Alpha {from=MaterialExpressionMultiply_2, output=""}
   ]
   properties {vertexAttributeBlendType=UseA}
  }
  MaterialExpressionConstant_0 {class=MaterialExpressionConstant}
  MaterialExpressionConvert_1 {
   class=MaterialExpressionConvert
   inputs [None {from=MaterialExpressionSwitch_2}]
   properties {
    convertInputs [{
      type=Vector4
      defaultValue=<rgba 0 0 0 1>
      expressionInput {
       expression=&self:MaterialExpressionSwitch_2
       inputName=None
       mask=0
       maskA=0
       maskB=0
       maskG=0
       maskR=0
       outputIndex=0
      }
     }]
    convertOutputs [{type=Scalar}, {type=Vector3}]
   }
  }
  MaterialExpressionConvert_10 {
   class=MaterialExpressionConvert
   inputs [
    None {from=MaterialExpressionNamedRerouteUsage_0}
    None {from=MaterialExpressionNamedRerouteUsage_22}
    None {from=MaterialExpressionNamedRerouteUsage_23}
    None {from=MaterialExpressionNamedRerouteUsage_24}
   ]
   properties {
    nodeName=Make Vector4
    convertInputs [
     {
      type=Scalar
      expressionInput {
       expression=&self:MaterialExpressionNamedRerouteUsage_0
       inputName=None
       mask=0
       maskA=0
       maskB=0
       maskG=0
       maskR=0
       outputIndex=0
      }
     }
     {
      type=Scalar
      expressionInput {
       expression=&self:MaterialExpressionNamedRerouteUsage_22
       inputName=None
       mask=0
       maskA=0
       maskB=0
       maskG=0
       maskR=0
       outputIndex=0
      }
     }
     {
      type=Scalar
      expressionInput {
       expression=&self:MaterialExpressionNamedRerouteUsage_23
       inputName=None
       mask=0
       maskA=0
       maskB=0
       maskG=0
       maskR=0
       outputIndex=0
      }
     }
     {
      type=Scalar
      expressionInput {
       expression=&self:MaterialExpressionNamedRerouteUsage_24
       inputName=None
       mask=0
       maskA=0
       maskB=0
       maskG=0
       maskR=0
       outputIndex=0
      }
     }
    ]
    convertOutputs [{type=Vector4, defaultValue=<rgba 0 0 0 1>}]
   }
  }
  MaterialExpressionConvert_11 {class=MaterialExpressionConvert, inputs [], properties {convertOutputs [{type=Vector4, defaultValue=<rgba 1 1 1 1>}]}}
  MaterialExpressionConvert_19 {
   class=MaterialExpressionConvert
   inputs [None {from=MaterialExpressionSwitch_34}]
   properties {
    nodeName=Break Float4
    convertInputs [{
      type=Vector4
      defaultValue=<rgba 0 0 0 1>
      expressionInput {
       expression=&self:MaterialExpressionSwitch_34
       inputName=None
       mask=0
       maskA=0
       maskB=0
       maskG=0
       maskR=0
       outputIndex=0
      }
     }]
    convertOutputs [{type=Scalar}, {type=Scalar}, {type=Scalar}, {type=Scalar}]
   }
  }
  MaterialExpressionConvert_2 {
   class=MaterialExpressionConvert
   inputs [None {from=MaterialExpressionSwitch_6}]
   properties {
    convertInputs [{
      type=Vector4
      defaultValue=<rgba 0 0 0 1>
      expressionInput {
       expression=&self:MaterialExpressionSwitch_6
       inputName=None
       mask=0
       maskA=0
       maskB=0
       maskG=0
       maskR=0
       outputIndex=0
      }
     }]
    convertOutputs [{type=Scalar}, {type=Vector3}]
   }
  }
  MaterialExpressionConvert_28 {
   class=MaterialExpressionConvert
   inputs [None {from=MaterialExpressionNamedRerouteUsage_115}]
   properties {
    convertInputs [{
      type=Vector4
      defaultValue=<rgba 0 0 0 1>
      expressionInput {
       expression=&self:MaterialExpressionNamedRerouteUsage_115
       inputName=None
       mask=0
       maskA=0
       maskB=0
       maskG=0
       maskR=0
       outputIndex=0
      }
     }]
    convertOutputs [{type=Scalar}, {type=Vector3}]
   }
  }
  MaterialExpressionConvert_29 {
   class=MaterialExpressionConvert
   inputs [None {from=MaterialExpressionNamedRerouteUsage_117}]
   properties {
    convertInputs [{
      type=Vector4
      defaultValue=<rgba 0 0 0 1>
      expressionInput {
       expression=&self:MaterialExpressionNamedRerouteUsage_117
       inputName=None
       mask=0
       maskA=0
       maskB=0
       maskG=0
       maskR=0
       outputIndex=0
      }
     }]
    convertOutputs [{type=Scalar}, {type=Vector3}]
   }
  }
  MaterialExpressionConvert_3 {
   class=MaterialExpressionConvert
   inputs [None {from=MaterialExpressionSwitch_12}]
   properties {
    convertInputs [{
      type=Vector4
      defaultValue=<rgba 0 0 0 1>
      expressionInput {
       expression=&self:MaterialExpressionSwitch_12
       inputName=None
       mask=0
       maskA=0
       maskB=0
       maskG=0
       maskR=0
       outputIndex=0
      }
     }]
    convertOutputs [{type=Scalar}, {type=Vector3}]
   }
  }
  MaterialExpressionConvert_30 {
   class=MaterialExpressionConvert
   inputs [None {from=MaterialExpressionNamedRerouteUsage_119}]
   properties {
    convertInputs [{
      type=Vector4
      defaultValue=<rgba 0 0 0 1>
      expressionInput {
       expression=&self:MaterialExpressionNamedRerouteUsage_119
       inputName=None
       mask=0
       maskA=0
       maskB=0
       maskG=0
       maskR=0
       outputIndex=0
      }
     }]
    convertOutputs [{type=Scalar}, {type=Vector3}]
   }
  }
  MaterialExpressionConvert_31 {
   class=MaterialExpressionConvert
   inputs [None {from=MaterialExpressionNamedRerouteUsage_121}]
   properties {
    convertInputs [{
      type=Vector4
      defaultValue=<rgba 0 0 0 1>
      expressionInput {
       expression=&self:MaterialExpressionNamedRerouteUsage_121
       inputName=None
       mask=0
       maskA=0
       maskB=0
       maskG=0
       maskR=0
       outputIndex=0
      }
     }]
    convertOutputs [{type=Scalar}, {type=Vector3}]
   }
  }
  MaterialExpressionConvert_32 {
   class=MaterialExpressionConvert
   inputs [None {from=MaterialExpressionNamedRerouteUsage_123}]
   properties {
    convertInputs [{
      type=Vector4
      defaultValue=<rgba 0 0 0 1>
      expressionInput {
       expression=&self:MaterialExpressionNamedRerouteUsage_123
       inputName=None
       mask=0
       maskA=0
       maskB=0
       maskG=0
       maskR=0
       outputIndex=0
      }
     }]
    convertOutputs [{type=Scalar}, {type=Vector3}]
   }
  }
  MaterialExpressionConvert_36 {
   class=MaterialExpressionConvert
   inputs [None {from=MaterialExpressionNamedRerouteUsage_118}]
   properties {
    nodeName=Break Float4
    convertInputs [{
      type=Vector4
      defaultValue=<rgba 0 0 0 1>
      expressionInput {
       expression=&self:MaterialExpressionNamedRerouteUsage_118
       inputName=None
       mask=0
       maskA=0
       maskB=0
       maskG=0
       maskR=0
       outputIndex=0
      }
     }]
    convertOutputs [{type=Scalar}, {type=Scalar}, {type=Scalar}, {type=Scalar}]
   }
  }
  MaterialExpressionConvert_37 {
   class=MaterialExpressionConvert
   inputs [None {from=MaterialExpressionNamedRerouteUsage_116}]
   properties {
    nodeName=Break Float4
    convertInputs [{
      type=Vector4
      defaultValue=<rgba 0 0 0 1>
      expressionInput {
       expression=&self:MaterialExpressionNamedRerouteUsage_116
       inputName=None
       mask=0
       maskA=0
       maskB=0
       maskG=0
       maskR=0
       outputIndex=0
      }
     }]
    convertOutputs [{type=Scalar}, {type=Scalar}, {type=Scalar}, {type=Scalar}]
   }
  }
  MaterialExpressionConvert_38 {
   class=MaterialExpressionConvert
   inputs [None {from=MaterialExpressionNamedRerouteUsage_120}]
   properties {
    nodeName=Break Float4
    convertInputs [{
      type=Vector4
      defaultValue=<rgba 0 0 0 1>
      expressionInput {
       expression=&self:MaterialExpressionNamedRerouteUsage_120
       inputName=None
       mask=0
       maskA=0
       maskB=0
       maskG=0
       maskR=0
       outputIndex=0
      }
     }]
    convertOutputs [{type=Scalar}, {type=Scalar}, {type=Scalar}, {type=Scalar}]
   }
  }
  MaterialExpressionConvert_39 {
   class=MaterialExpressionConvert
   inputs [None {from=MaterialExpressionNamedRerouteUsage_122}]
   properties {
    nodeName=Break Float4
    convertInputs [{
      type=Vector4
      defaultValue=<rgba 0 0 0 1>
      expressionInput {
       expression=&self:MaterialExpressionNamedRerouteUsage_122
       inputName=None
       mask=0
       maskA=0
       maskB=0
       maskG=0
       maskR=0
       outputIndex=0
      }
     }]
    convertOutputs [{type=Scalar}, {type=Scalar}, {type=Scalar}, {type=Scalar}]
   }
  }
  MaterialExpressionConvert_4 {
   class=MaterialExpressionConvert
   inputs [None {from=MaterialExpressionSwitch_23}]
   properties {
    convertInputs [{
      type=Vector4
      defaultValue=<rgba 0 0 0 1>
      expressionInput {
       expression=&self:MaterialExpressionSwitch_23
       inputName=None
       mask=0
       maskA=0
       maskB=0
       maskG=0
       maskR=0
       outputIndex=0
      }
     }]
    convertOutputs [{type=Scalar}, {type=Vector3}]
   }
  }
  MaterialExpressionConvert_40 {
   class=MaterialExpressionConvert
   inputs [None {from=MaterialExpressionNamedRerouteUsage_124}]
   properties {
    nodeName=Break Float4
    convertInputs [{
      type=Vector4
      defaultValue=<rgba 0 0 0 1>
      expressionInput {
       expression=&self:MaterialExpressionNamedRerouteUsage_124
       inputName=None
       mask=0
       maskA=0
       maskB=0
       maskG=0
       maskR=0
       outputIndex=0
      }
     }]
    convertOutputs [{type=Scalar}, {type=Scalar}, {type=Scalar}, {type=Scalar}]
   }
  }
  MaterialExpressionConvert_5 {
   class=MaterialExpressionConvert
   inputs [None {from=MaterialExpressionSwitch_30}]
   properties {
    convertInputs [{
      type=Vector4
      defaultValue=<rgba 0 0 0 1>
      expressionInput {
       expression=&self:MaterialExpressionSwitch_30
       inputName=None
       mask=0
       maskA=0
       maskB=0
       maskG=0
       maskR=0
       outputIndex=0
      }
     }]
    convertOutputs [{type=Scalar}, {type=Vector3}]
   }
  }
  MaterialExpressionConvert_6 {
   class=MaterialExpressionConvert
   inputs []
   properties {desc=disabled, convertOutputs [{type=Scalar}, {type=Vector3, defaultValue=<rgba 1 0.1 0 1>}]}
  }
  MaterialExpressionConvert_7 {class=MaterialExpressionConvert, inputs [], properties {desc=disabled, convertOutputs [{type=Vector3}]}}
  MaterialExpressionConvert_8 {class=MaterialExpressionConvert, inputs [], properties {desc=disabled, convertOutputs [{type=Vector3, defaultValue=<rgba 0 2 0 1>}]}}
  MaterialExpressionConvert_9 {class=MaterialExpressionConvert, inputs [], properties {desc=disabled, convertOutputs [{type=Vector3, defaultValue=<rgba 0.5 1 0 1>}]}}
  MaterialExpressionDotProduct_0 {
   class=MaterialExpressionDotProduct
   inputs [
    A {from=MaterialExpressionConvert_10, output=""}
    B {from=MaterialExpressionConvert_11, output=""}
   ]
   properties {}
  }
  MaterialExpressionFunctionInput_0 {
   class=MaterialExpressionFunctionInput
   inputs [Preview {from="", output=""}]
   properties {inputName=Mask Energy, inputType=FunctionInput_Scalar, sortPriority=210}
  }
  MaterialExpressionFunctionInput_1 {
   class=MaterialExpressionFunctionInput
   inputs [Preview {from="", output=""}]
   properties {inputName=Base Material In, inputType=FunctionInput_MaterialAttributes}
  }
  MaterialExpressionFunctionInput_2 {
   class=MaterialExpressionFunctionInput
   inputs [Preview {from="", output=""}]
   properties {bUsePreviewValueAsDefault=true, inputName=Visor Override Mask, inputType=FunctionInput_Scalar, sortPriority=230}
  }
  MaterialExpressionFunctionInput_20 {
   class=MaterialExpressionFunctionInput
   inputs [Preview {from="", output=""}]
   properties {
    desc="\r\n0 : Armor\r\n1 : Stone\r\n"
    description="0 : Armor; 1 : Stone;"
    inputName=Slot 1 - Material ID
    inputType=FunctionInput_Scalar
    sortPriority=110
   }
  }
  MaterialExpressionFunctionInput_21 {
   class=MaterialExpressionFunctionInput
   inputs [Preview {from="", output=""}]
   properties {inputName=Armor Material In, inputType=FunctionInput_MaterialAttributes, sortPriority=10}
  }
  MaterialExpressionFunctionInput_22 {
   class=MaterialExpressionFunctionInput
   inputs [Preview {from="", output=""}]
   properties {
    desc="\r\n0 : Armor\r\n1 : Stone\r\n"
    description="0 : Armor; 1 : Stone;"
    inputName=Slot 2 - Material ID
    inputType=FunctionInput_Scalar
    sortPriority=120
   }
  }
  MaterialExpressionFunctionInput_23 {
   class=MaterialExpressionFunctionInput
   inputs [Preview {from="", output=""}]
   properties {
    desc="\r\n0 : Armor\r\n1 : Stone\r\n"
    description="0 : Armor; 1 : Stone;"
    inputName=Slot 3 - Material ID
    inputType=FunctionInput_Scalar
    sortPriority=130
   }
  }
  MaterialExpressionFunctionInput_24 {
   class=MaterialExpressionFunctionInput
   inputs [Preview {from="", output=""}]
   properties {
    desc="\r\n0 : Armor\r\n1 : Stone\r\n"
    description="0 : Armor; 1 : Stone;"
    inputName=Slot 4 - Material ID
    inputType=FunctionInput_Scalar
    sortPriority=140
   }
  }
  MaterialExpressionFunctionInput_25 {
   class=MaterialExpressionFunctionInput
   inputs [Preview {from="", output=""}]
   properties {
    desc="\r\n0 : Armor\r\n1 : Stone\r\n"
    description="0 : Armor; 1 : Stone;"
    inputName=Slot 5 - Material ID
    inputType=FunctionInput_Scalar
    sortPriority=150
   }
  }
  MaterialExpressionFunctionInput_26 {
   class=MaterialExpressionFunctionInput
   inputs [Preview {from="", output=""}]
   properties {inputName=Stone Material In, inputType=FunctionInput_MaterialAttributes, sortPriority=20}
  }
  MaterialExpressionFunctionInput_28 {
   class=MaterialExpressionFunctionInput
   inputs [Preview {from="", output=""}]
   properties {bUsePreviewValueAsDefault=true, inputName=AO, inputType=FunctionInput_Scalar, sortPriority=200, previewValue=<1 0 0 1>}
  }
  MaterialExpressionFunctionInput_29 {
   class=MaterialExpressionFunctionInput
   inputs [Preview {from="", output=""}]
   properties {inputName=Enable Material ID Overrides, inputType=FunctionInput_Scalar, sortPriority=100}
  }
  MaterialExpressionFunctionInput_3 {
   class=MaterialExpressionFunctionInput
   inputs [Preview {from="", output=""}]
   properties {bUsePreviewValueAsDefault=true, inputName=Enable Visor, inputType=FunctionInput_Scalar, sortPriority=220}
  }
  MaterialExpressionFunctionOutput_0 {
   class=MaterialExpressionFunctionOutput
   inputs [None {from=MaterialExpressionBlendMaterialAttributes_0, output=""}]
   properties {outputName="Energy Material Out (V3)", sortPriority=50}
  }
  MaterialExpressionFunctionOutput_1 {
   class=MaterialExpressionFunctionOutput
   inputs [None {from=MaterialExpressionNamedRerouteUsage_26, output=""}]
   properties {outputName="Mask Energy Out (S)", sortPriority=60}
  }
  MaterialExpressionFunctionOutput_16 {
   class=MaterialExpressionFunctionOutput
   inputs [None {from=MaterialExpressionSetMaterialAttributes_1, output=""}]
   properties {outputName="Material 1 Out (MA)"}
  }
  MaterialExpressionFunctionOutput_17 {
   class=MaterialExpressionFunctionOutput
   inputs [None {from=MaterialExpressionSetMaterialAttributes_2, output=""}]
   properties {outputName="Material 2 Out (MA)", sortPriority=10}
  }
  MaterialExpressionFunctionOutput_18 {
   class=MaterialExpressionFunctionOutput
   inputs [None {from=MaterialExpressionSetMaterialAttributes_3, output=""}]
   properties {outputName="Material 3 Out (MA)", sortPriority=20}
  }
  MaterialExpressionFunctionOutput_19 {
   class=MaterialExpressionFunctionOutput
   inputs [None {from=MaterialExpressionSetMaterialAttributes_4, output=""}]
   properties {outputName="Material 4 Out (MA)", sortPriority=30}
  }
  MaterialExpressionFunctionOutput_20 {
   class=MaterialExpressionFunctionOutput
   inputs [None {from=MaterialExpressionNamedRerouteDeclaration_1, output=""}]
   properties {outputName="Material 5 Out (MA)", sortPriority=40}
  }
  MaterialExpressionFunctionOutput_21 {
   class=MaterialExpressionFunctionOutput
   inputs [None {from=MaterialExpressionLinearInterpolate_0, output=""}]
   properties {outputName="Energy Emission Out (V3)", sortPriority=70}
  }
  MaterialExpressionGetMaterialAttributes_1 {
   class=MaterialExpressionGetMaterialAttributes
   inputs [None {from=MaterialExpressionNamedRerouteUsage_2, output=""}]
   properties {attributeGetTypes [#1]}
  }
  MaterialExpressionGetMaterialAttributes_7 {
   class=MaterialExpressionGetMaterialAttributes
   inputs [None {from=MaterialExpressionNamedRerouteUsage_80, output=""}]
   properties {attributeGetTypes [#1]}
  }
  MaterialExpressionLinearInterpolate_0 {
   class=MaterialExpressionLinearInterpolate
   inputs [
    A {from=MaterialExpressionNamedRerouteUsage_19}
    B {from=MaterialExpressionGetMaterialAttributes_1, output=EmissiveColor}
    Alpha {from=MaterialExpressionNamedRerouteUsage_18}
   ]
  }
  MaterialExpressionLinearInterpolate_1 {
   class=MaterialExpressionLinearInterpolate
   inputs [
    A {from=MaterialExpressionNamedRerouteUsage_38}
    B {from=MaterialExpressionSwitch_0}
    Alpha {from=MaterialExpressionSaturate_2}
   ]
  }
  MaterialExpressionLinearInterpolate_2 {
   class=MaterialExpressionLinearInterpolate
   inputs [
    A {from=MaterialExpressionNamedRerouteUsage_113}
    B {from=MaterialExpressionNamedRerouteUsage_103}
    Alpha {from=MaterialExpressionNamedRerouteUsage_4}
   ]
  }
  MaterialExpressionLinearInterpolate_3 {
   class=MaterialExpressionLinearInterpolate
   inputs [
    A {from=MaterialExpressionNamedRerouteUsage_114}
    B {from=MaterialExpressionNamedRerouteUsage_104}
    Alpha {from=MaterialExpressionNamedRerouteUsage_5}
   ]
  }
  MaterialExpressionLinearInterpolate_4 {
   class=MaterialExpressionLinearInterpolate
   inputs [
    A {from=MaterialExpressionNamedRerouteUsage_67}
    B {from=MaterialExpressionNamedRerouteUsage_46}
    Alpha {from=MaterialExpressionNamedRerouteUsage_8}
   ]
  }
  MaterialExpressionLinearInterpolate_5 {
   class=MaterialExpressionLinearInterpolate
   inputs [
    A {from=MaterialExpressionNamedRerouteUsage_68}
    B {from=MaterialExpressionNamedRerouteUsage_47}
    Alpha {from=MaterialExpressionNamedRerouteUsage_10}
   ]
  }
  MaterialExpressionLinearInterpolate_6 {
   class=MaterialExpressionLinearInterpolate
   inputs [
    A {from=MaterialExpressionNamedRerouteUsage_69}
    B {from=MaterialExpressionNamedRerouteUsage_48}
    Alpha {from=MaterialExpressionNamedRerouteUsage_12}
   ]
  }
  MaterialExpressionLinearInterpolate_7 {
   class=MaterialExpressionLinearInterpolate
   inputs [
    A {from=MaterialExpressionNamedRerouteUsage_70}
    B {from=MaterialExpressionNamedRerouteUsage_49}
    Alpha {from=MaterialExpressionNamedRerouteUsage_15}
   ]
  }
  MaterialExpressionMaterialFunctionCall_16 {
   class=MaterialExpressionMaterialFunctionCall
   inputs [
    Material In {from=MaterialExpressionNamedRerouteUsage_143}
    AO In {from=MaterialExpressionNamedRerouteUsage_161}
    AO Mix {from=MaterialExpressionConvert_37}
    "Albedo Exp (Pre-Tint)" {from=MaterialExpressionConvert_28}
    Albedo Tint {from=MaterialExpressionConvert_28}
    Metallic {from=MaterialExpressionConvert_37}
    Specular Exp {from=MaterialExpressionConvert_37}
    Roughness Exp {from=MaterialExpressionConvert_37}
   ]
   properties {materialFunction=&/MF_PixieParts_VariantMaterial}
  }
  MaterialExpressionMaterialFunctionCall_17 {
   class=MaterialExpressionMaterialFunctionCall
   inputs [
    Material In {from=MaterialExpressionNamedRerouteUsage_145}
    AO In {from=MaterialExpressionNamedRerouteUsage_160}
    AO Mix {from=MaterialExpressionConvert_36}
    "Albedo Exp (Pre-Tint)" {from=MaterialExpressionConvert_29}
    Albedo Tint {from=MaterialExpressionConvert_29}
    Metallic {from=MaterialExpressionConvert_36}
    Specular Exp {from=MaterialExpressionConvert_36}
    Roughness Exp {from=MaterialExpressionConvert_36}
   ]
   properties {materialFunction=&MF_PixieParts_VariantMaterial}
  }
  MaterialExpressionMaterialFunctionCall_18 {
   class=MaterialExpressionMaterialFunctionCall
   inputs [
    Material In {from=MaterialExpressionNamedRerouteUsage_146}
    AO In {from=MaterialExpressionNamedRerouteUsage_162}
    AO Mix {from=MaterialExpressionConvert_38}
    "Albedo Exp (Pre-Tint)" {from=MaterialExpressionConvert_30}
    Albedo Tint {from=MaterialExpressionConvert_30}
    Metallic {from=MaterialExpressionConvert_38}
    Specular Exp {from=MaterialExpressionConvert_38}
    Roughness Exp {from=MaterialExpressionConvert_38}
   ]
   properties {materialFunction=&MF_PixieParts_VariantMaterial}
  }
  MaterialExpressionMaterialFunctionCall_19 {
   class=MaterialExpressionMaterialFunctionCall
   inputs [
    Material In {from=MaterialExpressionNamedRerouteUsage_148}
    AO In {from=MaterialExpressionNamedRerouteUsage_163}
    AO Mix {from=MaterialExpressionConvert_39}
    "Albedo Exp (Pre-Tint)" {from=MaterialExpressionConvert_31}
    Albedo Tint {from=MaterialExpressionConvert_31}
    Metallic {from=MaterialExpressionConvert_39}
    Specular Exp {from=MaterialExpressionConvert_39}
    Roughness Exp {from=MaterialExpressionConvert_39}
   ]
   properties {materialFunction=&MF_PixieParts_VariantMaterial}
  }
  MaterialExpressionMaterialFunctionCall_2 {
   class=MaterialExpressionMaterialFunctionCall
   inputs [
    Emission In {from=MaterialExpressionNamedRerouteUsage_155}
    Emission Pre-Tint Exp {from=MaterialExpressionNamedRerouteUsage_83}
    Emission Color {from=MaterialExpressionNamedRerouteUsage_81}
    Emission Color 2 {from=MaterialExpressionNamedRerouteUsage_84}
    Emission Brightness {from=MaterialExpressionNamedRerouteUsage_85}
    Emission Saturation Offset {from=MaterialExpressionNamedRerouteUsage_90}
   ]
   properties {materialFunction=&/MF_PIxieParts_EmissionRemap}
  }
  MaterialExpressionMaterialFunctionCall_20 {
   class=MaterialExpressionMaterialFunctionCall
   inputs [
    Material In {from=MaterialExpressionNamedRerouteUsage_149}
    AO In {from=MaterialExpressionNamedRerouteUsage_164}
    AO Mix {from=MaterialExpressionConvert_40}
    "Albedo Exp (Pre-Tint)" {from=MaterialExpressionConvert_32}
    Albedo Tint {from=MaterialExpressionConvert_32}
    Metallic {from=MaterialExpressionConvert_40}
    Specular Exp {from=MaterialExpressionConvert_40}
    Roughness Exp {from=MaterialExpressionConvert_40}
   ]
   properties {materialFunction=&MF_PixieParts_VariantMaterial}
  }
  MaterialExpressionMaterialFunctionCall_21 {
   class=MaterialExpressionMaterialFunctionCall
   inputs [
    Material {from=MaterialExpressionNamedRerouteUsage_1}
    Enable Visor {from=MaterialExpressionNamedRerouteUsage_21}
    Tech Art Mask {from=""}
    Base Color Scale {from=MaterialExpressionOneMinus_0}
    Emission Pre-Tint Exp {from=MaterialExpressionNamedRerouteUsage_3}
    Emission Color {from=MaterialExpressionNamedRerouteUsage_151}
    Emission Color 2 {from=MaterialExpressionNamedRerouteUsage_11}
    Emission Brightness {from=MaterialExpressionNamedRerouteUsage_13}
    Emission Saturation Offset {from=MaterialExpressionNamedRerouteUsage_144}
   ]
   properties {materialFunction=&/MF_PixieParts_VisorMaterial}
  }
  MaterialExpressionMultiply_0 {
   class=MaterialExpressionMultiply
   inputs [
    A {from=MaterialExpressionNamedRerouteUsage_75}
    B {from=MaterialExpressionNamedRerouteUsage_147}
   ]
  }
  MaterialExpressionMultiply_1 {
   class=MaterialExpressionMultiply
   inputs [
    A {from=MaterialExpressionNamedRerouteUsage_154}
    B {from=MaterialExpressionMaterialFunctionCall_2, output="Emission Out (V3)"}
   ]
  }
  MaterialExpressionMultiply_16 {
   class=MaterialExpressionMultiply
   inputs [
    A {from=MaterialExpressionNamedRerouteUsage_343}
    B {from=MaterialExpressionOneMinus_6}
   ]
   properties {constB=0, desc="kill energy material \r\nif visor is enabled and\r\nmaterial is armor"}
  }
  MaterialExpressionMultiply_2 {
   class=MaterialExpressionMultiply
   inputs [
    A {from=MaterialExpressionNamedRerouteUsage_132}
    B {from=MaterialExpressionOneMinus_1}
   ]
  }
  MaterialExpressionMultiply_26 {
   class=MaterialExpressionMultiply
   inputs [
    A {from=MaterialExpressionNamedRerouteUsage_20}
    B {from=MaterialExpressionOneMinus_9}
   ]
  }
  MaterialExpressionNamedRerouteDeclaration_0 {
   class=MaterialExpressionNamedRerouteDeclaration
   inputs [None {from=MaterialExpressionSetMaterialAttributes_5}]
   properties {name="Material 5 (MA)", nodeColor=<rgba 1 0 0.0938 1>}
  }
  MaterialExpressionNamedRerouteDeclaration_1 {
   class=MaterialExpressionNamedRerouteDeclaration
   inputs [None {from=MaterialExpressionMaterialFunctionCall_21, output=Result}]
   properties {name="Material 5 Out (MA)", nodeColor=<rgba 1 0 0.0938 1>}
  }
  MaterialExpressionNamedRerouteDeclaration_10 {
   class=MaterialExpressionNamedRerouteDeclaration
   inputs [None {from=MaterialExpressionFunctionInput_2}]
   properties {name="Visor Override Mask (S)", nodeColor=<rgba 0 0.4819 1 1>}
  }
  MaterialExpressionNamedRerouteDeclaration_100 {
   class=MaterialExpressionNamedRerouteDeclaration
   inputs [None {from=MaterialExpressionBlendMaterialAttributes_12}]
   properties {name="Material 1 In (MA)", nodeColor=<rgba 0.4531 1 0 1>}
  }
  MaterialExpressionNamedRerouteDeclaration_101 {
   class=MaterialExpressionNamedRerouteDeclaration
   inputs [None {from=MaterialExpressionBlendMaterialAttributes_13}]
   properties {name="Material 2 In (MA)", nodeColor=<rgba 0.4531 1 0 1>}
  }
  MaterialExpressionNamedRerouteDeclaration_102 {
   class=MaterialExpressionNamedRerouteDeclaration
   inputs [None {from=MaterialExpressionVectorParameter_39, output=RGB}]
   properties {name="Override - Mat4 Metal LUT (V3)", nodeColor=<rgba 0 0.9062 1 1>}
  }
  MaterialExpressionNamedRerouteDeclaration_103 {
   class=MaterialExpressionNamedRerouteDeclaration
   inputs [None {from=MaterialExpressionBlendMaterialAttributes_14}]
   properties {name="Material 3 In (MA)", nodeColor=<rgba 0.4531 1 0 1>}
  }
  MaterialExpressionNamedRerouteDeclaration_104 {
   class=MaterialExpressionNamedRerouteDeclaration
   inputs [None {from=MaterialExpressionBlendMaterialAttributes_15}]
   properties {name="Material 4 In (MA)", nodeColor=<rgba 0.4531 1 0 1>}
  }
  MaterialExpressionNamedRerouteDeclaration_105 {
   class=MaterialExpressionNamedRerouteDeclaration
   inputs [None {from=MaterialExpressionBlendMaterialAttributes_16}]
   properties {name="Material 5 In (MA)", nodeColor=<rgba 0.4531 1 0 1>}
  }
  MaterialExpressionNamedRerouteDeclaration_107 {
   class=MaterialExpressionNamedRerouteDeclaration
   inputs [None {from=MaterialExpressionVectorParameter_40, output=RGB}]
   properties {name="Override - Mat4 Thin Film ##1 (V3)", nodeColor=<rgba 0 1 0.5043 1>}
  }
  MaterialExpressionNamedRerouteDeclaration_108 {
   class=MaterialExpressionNamedRerouteDeclaration
   inputs [None {from=MaterialExpressionVectorParameter_41, output=RGB}]
   properties {name="Override - Mat4 Thin Film ##2 (V3)", nodeColor=<rgba 0.2957 1 0 1>}
  }
  MaterialExpressionNamedRerouteDeclaration_109 {
   class=MaterialExpressionNamedRerouteDeclaration
   inputs [None {from=MaterialExpressionVectorParameter_42, output=RGBA}]
   properties {name="Override - Mat5 Clear Coat (V4)", nodeColor=<rgba 1 0 0.5859 1>}
  }
  MaterialExpressionNamedRerouteDeclaration_11 {
   class=MaterialExpressionNamedRerouteDeclaration
   inputs [None {from=MaterialExpressionVectorParameter_5, output=RGB}]
   properties {name="Body - Energy Color 2 (V3)", nodeColor=<rgba 0 0.0859 1 1>}
  }
  MaterialExpressionNamedRerouteDeclaration_110 {
   class=MaterialExpressionNamedRerouteDeclaration
   inputs [None {from=MaterialExpressionVectorParameter_43, output=RGB}]
   properties {name="Override - Mat5 Metal LUT (V3)", nodeColor=<rgba 0 0.9062 1 1>}
  }
  MaterialExpressionNamedRerouteDeclaration_111 {
   class=MaterialExpressionNamedRerouteDeclaration
   inputs [None {from=MaterialExpressionVectorParameter_44, output=RGB}]
   properties {name="Override - Mat5 Thin Film ##1 (V3)", nodeColor=<rgba 0 1 0.5043 1>}
  }
  MaterialExpressionNamedRerouteDeclaration_112 {
   class=MaterialExpressionNamedRerouteDeclaration
   inputs [None {from=MaterialExpressionFunctionInput_28}]
   properties {name="AO (S)", nodeColor=<rgba 0 0.5217 1 1>}
  }
  MaterialExpressionNamedRerouteDeclaration_113 {
   class=MaterialExpressionNamedRerouteDeclaration
   inputs [None {from=MaterialExpressionSaturate_1}]
   properties {name="Enable Variant Overrides (S)", nodeColor=<rgba 1 0 0 1>}
  }
  MaterialExpressionNamedRerouteDeclaration_114 {
   class=MaterialExpressionNamedRerouteDeclaration
   inputs [None {from=MaterialExpressionVectorParameter_45, output=RGB}]
   properties {name="Override - Mat5 Thin Film ##2 (V3)", nodeColor=<rgba 0.2957 1 0 1>}
  }
  MaterialExpressionNamedRerouteDeclaration_119 {
   class=MaterialExpressionNamedRerouteDeclaration
   inputs [None {from=MaterialExpressionConvert_1}]
   properties {name="Material 1 - Clear Coat ##1 (S)", nodeColor=<rgba 1 0 0.5859 1>}
  }
  MaterialExpressionNamedRerouteDeclaration_12 {
   class=MaterialExpressionNamedRerouteDeclaration
   inputs [None {from=MaterialExpressionMultiply_16}]
   properties {name="Mask Energy (S)", nodeColor=<rgba 0 0.4819 1 1>}
  }
  MaterialExpressionNamedRerouteDeclaration_120 {
   class=MaterialExpressionNamedRerouteDeclaration
   inputs [None {from=MaterialExpressionSwitch_1}]
   properties {name="Material 1 - Metal LUT (V3)", nodeColor=<rgba 0 0.9062 1 1>}
  }
  MaterialExpressionNamedRerouteDeclaration_121 {
   class=MaterialExpressionNamedRerouteDeclaration
   inputs [None {from=MaterialExpressionSwitch_3}]
   properties {name="Material 1 - Thin Film ##1 (V3)", nodeColor=<rgba 0 1 0.5043 1>}
  }
  MaterialExpressionNamedRerouteDeclaration_122 {
   class=MaterialExpressionNamedRerouteDeclaration
   inputs [None {from=MaterialExpressionSwitch_4}]
   properties {name="Material 1 - Thin Film ##2 (V3)", nodeColor=<rgba 0.2957 1 0 1>}
  }
  MaterialExpressionNamedRerouteDeclaration_123 {
   class=MaterialExpressionNamedRerouteDeclaration
   inputs [None {from=MaterialExpressionConvert_2}]
   properties {name="Material 2 - Clear Coat ##1 (S)", nodeColor=<rgba 1 0 0.5859 1>}
  }
  MaterialExpressionNamedRerouteDeclaration_124 {
   class=MaterialExpressionNamedRerouteDeclaration
   inputs [None {from=MaterialExpressionSwitch_5}]
   properties {name="Material 2 - Metal LUT (V3)", nodeColor=<rgba 0 0.9062 1 1>}
  }
  MaterialExpressionNamedRerouteDeclaration_125 {
   class=MaterialExpressionNamedRerouteDeclaration
   inputs [None {from=MaterialExpressionSwitch_7}]
   properties {name="Material 2 - Thin Film ##1 (V3)", nodeColor=<rgba 0 1 0.5043 1>}
  }
  MaterialExpressionNamedRerouteDeclaration_126 {
   class=MaterialExpressionNamedRerouteDeclaration
   inputs [None {from=MaterialExpressionSwitch_9}]
   properties {name="Material 2 - Thin Film ##2 (V3)", nodeColor=<rgba 0.2957 1 0 1>}
  }
  MaterialExpressionNamedRerouteDeclaration_127 {
   class=MaterialExpressionNamedRerouteDeclaration
   inputs [None {from=MaterialExpressionConvert_3}]
   properties {name="Material 3 - Clear Coat ##1 (S)", nodeColor=<rgba 1 0 0.5859 1>}
  }
  MaterialExpressionNamedRerouteDeclaration_128 {
   class=MaterialExpressionNamedRerouteDeclaration
   inputs [None {from=MaterialExpressionSwitch_10}]
   properties {name="Material 3 - Metal LUT (V3)", nodeColor=<rgba 0 0.9062 1 1>}
  }
  MaterialExpressionNamedRerouteDeclaration_129 {
   class=MaterialExpressionNamedRerouteDeclaration
   inputs [None {from=MaterialExpressionSwitch_13}]
   properties {name="Material 3 - Thin Film ##1 (V3)", nodeColor=<rgba 0 1 0.5043 1>}
  }
  MaterialExpressionNamedRerouteDeclaration_13 {
   class=MaterialExpressionNamedRerouteDeclaration
   inputs [None {from=MaterialExpressionVectorParameter_7, output=RGB}]
   properties {name="Override - Energy Color 2 (V3)", nodeColor=<rgba 0 0.0859 1 1>}
  }
  MaterialExpressionNamedRerouteDeclaration_130 {
   class=MaterialExpressionNamedRerouteDeclaration
   inputs [None {from=MaterialExpressionSwitch_14}]
   properties {name="Material 3 - Thin Film ##2 (V3)", nodeColor=<rgba 0.2957 1 0 1>}
  }
  MaterialExpressionNamedRerouteDeclaration_131 {
   class=MaterialExpressionNamedRerouteDeclaration
   inputs [None {from=MaterialExpressionConvert_4}]
   properties {name="Material 4 - Clear Coat ##1 (S)", nodeColor=<rgba 1 0 0.5859 1>}
  }
  MaterialExpressionNamedRerouteDeclaration_132 {
   class=MaterialExpressionNamedRerouteDeclaration
   inputs [None {from=MaterialExpressionSwitch_21}]
   properties {name="Material 4 - Metal LUT (V3)", nodeColor=<rgba 0 0.9062 1 1>}
  }
  MaterialExpressionNamedRerouteDeclaration_133 {
   class=MaterialExpressionNamedRerouteDeclaration
   inputs [None {from=MaterialExpressionSwitch_25}]
   properties {name="Material 4 - Thin Film ##1 (V3)", nodeColor=<rgba 0 1 0.5043 1>}
  }
  MaterialExpressionNamedRerouteDeclaration_134 {
   class=MaterialExpressionNamedRerouteDeclaration
   inputs [None {from=MaterialExpressionSwitch_27}]
   properties {name="Material 4 - Thin Film ##2 (V3)", nodeColor=<rgba 0.2957 1 0 1>}
  }
  MaterialExpressionNamedRerouteDeclaration_135 {
   class=MaterialExpressionNamedRerouteDeclaration
   inputs [None {from=MaterialExpressionConvert_5}]
   properties {name="Material 5 - Clear Coat ##1 (S)", nodeColor=<rgba 1 0 0.5859 1>}
  }
  MaterialExpressionNamedRerouteDeclaration_136 {
   class=MaterialExpressionNamedRerouteDeclaration
   inputs [None {from=MaterialExpressionSwitch_29}]
   properties {name="Material 5 - Metal LUT (V3)", nodeColor=<rgba 0 0.9062 1 1>}
  }
  MaterialExpressionNamedRerouteDeclaration_137 {
   class=MaterialExpressionNamedRerouteDeclaration
   inputs [None {from=MaterialExpressionSwitch_31}]
   properties {name="Material 5 - Thin Film ##1 (V3)", nodeColor=<rgba 0 1 0.5043 1>}
  }
  MaterialExpressionNamedRerouteDeclaration_138 {
   class=MaterialExpressionNamedRerouteDeclaration
   inputs [None {from=MaterialExpressionSwitch_32}]
   properties {name="Material 5 - Thin Film ##2 (V3)", nodeColor=<rgba 0.2957 1 0 1>}
  }
  MaterialExpressionNamedRerouteDeclaration_139 {
   class=MaterialExpressionNamedRerouteDeclaration
   inputs [None {from=MaterialExpressionConvert_1}]
   properties {name="Material 1 - Clear Coat ##2 (V3)", nodeColor=<rgba 1 0 0.5859 1>}
  }
  MaterialExpressionNamedRerouteDeclaration_14 {
   class=MaterialExpressionNamedRerouteDeclaration
   inputs [None {from=MaterialExpressionFunctionInput_3}]
   properties {name="Enable Visor (S)", nodeColor=<rgba 1 0 0 1>}
  }
  MaterialExpressionNamedRerouteDeclaration_140 {
   class=MaterialExpressionNamedRerouteDeclaration
   inputs [None {from=MaterialExpressionConvert_2}]
   properties {name="Material 2 - Clear Coat ##2 (V3)", nodeColor=<rgba 1 0 0.5859 1>}
  }
  MaterialExpressionNamedRerouteDeclaration_141 {
   class=MaterialExpressionNamedRerouteDeclaration
   inputs [None {from=MaterialExpressionConvert_3}]
   properties {name="Material 3 - Clear Coat ##2 (V3)", nodeColor=<rgba 1 0 0.5859 1>}
  }
  MaterialExpressionNamedRerouteDeclaration_142 {
   class=MaterialExpressionNamedRerouteDeclaration
   inputs [None {from=MaterialExpressionConvert_4}]
   properties {name="Material 4 - Clear Coat ##2 (V3)", nodeColor=<rgba 1 0 0.5859 1>}
  }
  MaterialExpressionNamedRerouteDeclaration_143 {
   class=MaterialExpressionNamedRerouteDeclaration
   inputs [None {from=MaterialExpressionConvert_5}]
   properties {name="Material 5 - Clear Coat ##2 (V3)", nodeColor=<rgba 1 0 0.5859 1>}
  }
  MaterialExpressionNamedRerouteDeclaration_144 {
   class=MaterialExpressionNamedRerouteDeclaration
   inputs [None {from=MaterialExpressionConvert_19}]
   properties {name="Energy Color 2 Mix (S)", nodeColor=<rgba 0 0.0859 1 1>}
  }
  MaterialExpressionNamedRerouteDeclaration_156 {
   class=MaterialExpressionNamedRerouteDeclaration
   inputs [None {from=MaterialExpressionVectorParameter_47, output=RGBA}]
   properties {name="Body - Energy Settings (V4)", nodeColor=<rgba 0 0.0859 1 1>}
  }
  MaterialExpressionNamedRerouteDeclaration_157 {
   class=MaterialExpressionNamedRerouteDeclaration
   inputs [None {from=MaterialExpressionVectorParameter_48, output=RGBA}]
   properties {name="Override - Energy Settings (V4)", nodeColor=<rgba 0 0.0859 1 1>}
  }
  MaterialExpressionNamedRerouteDeclaration_158 {
   class=MaterialExpressionNamedRerouteDeclaration
   inputs [None {from=MaterialExpressionConvert_19}]
   properties {name="Energy Saturation Offset (S)", nodeColor=<rgba 0 0.0859 1 1>}
  }
  MaterialExpressionNamedRerouteDeclaration_16 {
   class=MaterialExpressionNamedRerouteDeclaration
   inputs [None {from=MaterialExpressionFunctionInput_21}]
   properties {name="Armor Material In (MA)", nodeColor=<rgba 0.4531 1 0 1>}
  }
  MaterialExpressionNamedRerouteDeclaration_162 {
   class=MaterialExpressionNamedRerouteDeclaration
   inputs [None {from=MaterialExpressionScalarParameter_12}]
   properties {name="Energy - Color 2 Mix Mult (S)", nodeColor=<rgba 0 0.0859 1 1>}
  }
  MaterialExpressionNamedRerouteDeclaration_163 {
   class=MaterialExpressionNamedRerouteDeclaration
   inputs [None {from=MaterialExpressionScalarParameter_13}]
   properties {name="Energy - Visor OFF Brightness Gain (S)", nodeColor=<rgba 0 0.0859 1 1>}
  }
  MaterialExpressionNamedRerouteDeclaration_164 {
   class=MaterialExpressionNamedRerouteDeclaration
   inputs [None {from=MaterialExpressionGetMaterialAttributes_7, output=EmissiveColor}]
   properties {name="Emission In (V3)", nodeColor=<rgba 0.5938 0 1 1>}
  }
  MaterialExpressionNamedRerouteDeclaration_17 {
   class=MaterialExpressionNamedRerouteDeclaration
   inputs [None {from=MaterialExpressionFunctionInput_26}]
   properties {name="Stone Material In (MA)", nodeColor=<rgba 0.4531 1 0 1>}
  }
  MaterialExpressionNamedRerouteDeclaration_18 {
   class=MaterialExpressionNamedRerouteDeclaration
   inputs [None {from=MaterialExpressionFunctionInput_1}]
   properties {name="Base Material In (MA)", nodeColor=<rgba 0.4531 1 0 1>}
  }
  MaterialExpressionNamedRerouteDeclaration_189 {
   class=MaterialExpressionNamedRerouteDeclaration
   inputs [None {from=MaterialExpressionFunctionInput_0}]
   properties {name="Mask Energy In (S)", nodeColor=<rgba 0 0.4819 1 1>}
  }
  MaterialExpressionNamedRerouteDeclaration_19 {
   class=MaterialExpressionNamedRerouteDeclaration
   inputs [None {from=MaterialExpressionVectorParameter_0, output=RGBA}]
   properties {name="Body - Mat1 Clear Coat (V4)", nodeColor=<rgba 1 0 0.5859 1>}
  }
  MaterialExpressionNamedRerouteDeclaration_2 {
   class=MaterialExpressionNamedRerouteDeclaration
   inputs [None {from=MaterialExpressionConvert_19}]
   properties {name="Energy Pre-Tint Exp (S)", nodeColor=<rgba 0 0.0859 1 1>}
  }
  MaterialExpressionNamedRerouteDeclaration_20 {
   class=MaterialExpressionNamedRerouteDeclaration
   inputs [None {from=MaterialExpressionVectorParameter_1, output=RGB}]
   properties {name="Body - Mat1 Metal LUT (V3)", nodeColor=<rgba 0 0.9062 1 1>}
  }
  MaterialExpressionNamedRerouteDeclaration_21 {
   class=MaterialExpressionNamedRerouteDeclaration
   inputs [None {from=MaterialExpressionStep_0}]
   properties {name="Enable Clear Coat (S)", nodeColor=<rgba 0 0.0391 1 1>}
  }
  MaterialExpressionNamedRerouteDeclaration_22 {
   class=MaterialExpressionNamedRerouteDeclaration
   inputs [None {from=MaterialExpressionLinearInterpolate_1}]
   properties {name="Energy - Color 2 (V3)", nodeColor=<rgba 0 0.0859 1 1>}
  }
  MaterialExpressionNamedRerouteDeclaration_23 {
   class=MaterialExpressionNamedRerouteDeclaration
   inputs [None {from=MaterialExpressionVectorParameter_3, output=RGB}]
   properties {name="Body - Mat1 Thin Film ##1 (V3)", nodeColor=<rgba 0 1 0.5043 1>}
  }
  MaterialExpressionNamedRerouteDeclaration_24 {
   class=MaterialExpressionNamedRerouteDeclaration
   inputs [None {from=MaterialExpressionVectorParameter_4, output=RGB}]
   properties {name="Body - Mat1 Thin Film ##2 (V3)", nodeColor=<rgba 0.2957 1 0 1>}
  }
  MaterialExpressionNamedRerouteDeclaration_25 {
   class=MaterialExpressionNamedRerouteDeclaration
   inputs [None {from=MaterialExpressionConvert_19}]
   properties {name="Energy Brightness (S)", nodeColor=<rgba 0 0.0859 1 1>}
  }
  MaterialExpressionNamedRerouteDeclaration_29 {
   class=MaterialExpressionNamedRerouteDeclaration
   inputs [None {from=MaterialExpressionVectorParameter_10, output=RGBA}]
   properties {name="Body - Mat2 Clear Coat (V4)", nodeColor=<rgba 1 0 0.5859 1>}
  }
  MaterialExpressionNamedRerouteDeclaration_3 {
   class=MaterialExpressionNamedRerouteDeclaration
   inputs [None {from=MaterialExpressionConvert_8}]
   properties {name="Thin Film ##1 (V3)", nodeColor=<rgba 0 1 0.5043 1>}
  }
  MaterialExpressionNamedRerouteDeclaration_30 {
   class=MaterialExpressionNamedRerouteDeclaration
   inputs [None {from=MaterialExpressionVectorParameter_11, output=RGB}]
   properties {name="Body - Mat2 Metal LUT (V3)", nodeColor=<rgba 0 0.9062 1 1>}
  }
  MaterialExpressionNamedRerouteDeclaration_31 {
   class=MaterialExpressionNamedRerouteDeclaration
   inputs [None {from=MaterialExpressionVectorParameter_12, output=RGB}]
   properties {name="Body - Mat2 Thin Film ##1 (V3)", nodeColor=<rgba 0 1 0.5043 1>}
  }
  MaterialExpressionNamedRerouteDeclaration_32 {
   class=MaterialExpressionNamedRerouteDeclaration
   inputs [None {from=MaterialExpressionVectorParameter_13, output=RGB}]
   properties {name="Body - Mat2 Thin Film ##2 (V3)", nodeColor=<rgba 0.2957 1 0 1>}
  }
  MaterialExpressionNamedRerouteDeclaration_33 {
   class=MaterialExpressionNamedRerouteDeclaration
   inputs [None {from=MaterialExpressionVectorParameter_14, output=RGBA}]
   properties {name="Body - Mat3 Clear Coat (V4)", nodeColor=<rgba 1 0 0.5859 1>}
  }
  MaterialExpressionNamedRerouteDeclaration_34 {
   class=MaterialExpressionNamedRerouteDeclaration
   inputs [None {from=MaterialExpressionVectorParameter_15, output=RGB}]
   properties {name="Body - Mat3 Metal LUT (V3)", nodeColor=<rgba 0 0.9062 1 1>}
  }
  MaterialExpressionNamedRerouteDeclaration_35 {
   class=MaterialExpressionNamedRerouteDeclaration
   inputs [None {from=MaterialExpressionVectorParameter_16, output=RGB}]
   properties {name="Body - Mat3 Thin Film ##1 (V3)", nodeColor=<rgba 0 1 0.5043 1>}
  }
  MaterialExpressionNamedRerouteDeclaration_36 {
   class=MaterialExpressionNamedRerouteDeclaration
   inputs [None {from=MaterialExpressionVectorParameter_17, output=RGB}]
   properties {name="Body - Mat3 Thin Film ##2 (V3)", nodeColor=<rgba 0.2957 1 0 1>}
  }
  MaterialExpressionNamedRerouteDeclaration_39 {
   class=MaterialExpressionNamedRerouteDeclaration
   inputs [None {from=MaterialExpressionVectorParameter_18, output=RGBA}]
   properties {name="Body - Mat4 Clear Coat (V4)", nodeColor=<rgba 1 0 0.5859 1>}
  }
  MaterialExpressionNamedRerouteDeclaration_4 {
   class=MaterialExpressionNamedRerouteDeclaration
   inputs [None {from=MaterialExpressionConvert_9}]
   properties {name="Thin Film ##2 (V3)", nodeColor=<rgba 0.2957 1 0 1>}
  }
  MaterialExpressionNamedRerouteDeclaration_40 {
   class=MaterialExpressionNamedRerouteDeclaration
   inputs [None {from=MaterialExpressionVectorParameter_19, output=RGB}]
   properties {name="Body - Mat4 Metal LUT (V3)", nodeColor=<rgba 0 0.9062 1 1>}
  }
  MaterialExpressionNamedRerouteDeclaration_41 {
   class=MaterialExpressionNamedRerouteDeclaration
   inputs [None {from=MaterialExpressionVectorParameter_20, output=RGB}]
   properties {name="Body - Mat4 Thin Film ##1 (V3)", nodeColor=<rgba 0 1 0.5043 1>}
  }
  MaterialExpressionNamedRerouteDeclaration_42 {
   class=MaterialExpressionNamedRerouteDeclaration
   inputs [None {from=MaterialExpressionVectorParameter_21, output=RGB}]
   properties {name="Body - Mat4 Thin Film ##2 (V3)", nodeColor=<rgba 0.2957 1 0 1>}
  }
  MaterialExpressionNamedRerouteDeclaration_43 {
   class=MaterialExpressionNamedRerouteDeclaration
   inputs [None {from=MaterialExpressionVectorParameter_22, output=RGBA}]
   properties {name="Body - Mat5 Clear Coat (V4)", nodeColor=<rgba 1 0 0.5859 1>}
  }
  MaterialExpressionNamedRerouteDeclaration_44 {
   class=MaterialExpressionNamedRerouteDeclaration
   inputs [None {from=MaterialExpressionVectorParameter_23, output=RGB}]
   properties {name="Body - Mat5 Metal LUT (V3)", nodeColor=<rgba 0 0.9062 1 1>}
  }
  MaterialExpressionNamedRerouteDeclaration_45 {
   class=MaterialExpressionNamedRerouteDeclaration
   inputs [None {from=MaterialExpressionVectorParameter_24, output=RGB}]
   properties {name="Body - Mat5 Thin Film ##1 (V3)", nodeColor=<rgba 0 1 0.5043 1>}
  }
  MaterialExpressionNamedRerouteDeclaration_46 {
   class=MaterialExpressionNamedRerouteDeclaration
   inputs [None {from=MaterialExpressionVectorParameter_25, output=RGB}]
   properties {name="Body - Mat5 Thin Film ##2 (V3)", nodeColor=<rgba 0.2957 1 0 1>}
  }
  MaterialExpressionNamedRerouteDeclaration_47 {
   class=MaterialExpressionNamedRerouteDeclaration
   inputs [None {from=MaterialExpressionVectorParameter_26, output=RGBA}]
   properties {name="Override - Mat1 Clear Coat (V4)", nodeColor=<rgba 1 0 0.5859 1>}
  }
  MaterialExpressionNamedRerouteDeclaration_48 {
   class=MaterialExpressionNamedRerouteDeclaration
   inputs [None {from=MaterialExpressionVectorParameter_27, output=RGB}]
   properties {name="Override - Mat1 Metal LUT (V3)", nodeColor=<rgba 0 0.9062 1 1>}
  }
  MaterialExpressionNamedRerouteDeclaration_49 {
   class=MaterialExpressionNamedRerouteDeclaration
   inputs [None {from=MaterialExpressionVectorParameter_28, output=RGB}]
   properties {name="Override - Mat1 Thin Film ##1 (V3)", nodeColor=<rgba 0 1 0.5043 1>}
  }
  MaterialExpressionNamedRerouteDeclaration_5 {
   class=MaterialExpressionNamedRerouteDeclaration
   inputs [None {from=MaterialExpressionConvert_6}]
   properties {name="Clear Coat ##1 (S)", nodeColor=<rgba 1 0 0.5859 1>}
  }
  MaterialExpressionNamedRerouteDeclaration_50 {
   class=MaterialExpressionNamedRerouteDeclaration
   inputs [None {from=MaterialExpressionVectorParameter_29, output=RGB}]
   properties {name="Override - Mat1 Thin Film ##2 (V3)", nodeColor=<rgba 0.2957 1 0 1>}
  }
  MaterialExpressionNamedRerouteDeclaration_51 {
   class=MaterialExpressionNamedRerouteDeclaration
   inputs [None {from=MaterialExpressionVectorParameter_30, output=RGBA}]
   properties {name="Override - Mat2 Clear Coat (V4)", nodeColor=<rgba 1 0 0.5859 1>}
  }
  MaterialExpressionNamedRerouteDeclaration_52 {
   class=MaterialExpressionNamedRerouteDeclaration
   inputs [None {from=MaterialExpressionVectorParameter_31, output=RGB}]
   properties {name="Override - Mat2 Metal LUT (V3)", nodeColor=<rgba 0 0.9062 1 1>}
  }
  MaterialExpressionNamedRerouteDeclaration_53 {
   class=MaterialExpressionNamedRerouteDeclaration
   inputs [None {from=MaterialExpressionVectorParameter_32, output=RGB}]
   properties {name="Override - Mat2 Thin Film ##1 (V3)", nodeColor=<rgba 0 1 0.5043 1>}
  }
  MaterialExpressionNamedRerouteDeclaration_54 {
   class=MaterialExpressionNamedRerouteDeclaration
   inputs [None {from=MaterialExpressionVectorParameter_33, output=RGB}]
   properties {name="Override - Mat2 Thin Film ##2 (V3)", nodeColor=<rgba 0.2957 1 0 1>}
  }
  MaterialExpressionNamedRerouteDeclaration_55 {
   class=MaterialExpressionNamedRerouteDeclaration
   inputs [None {from=MaterialExpressionVectorParameter_34, output=RGBA}]
   properties {name="Override - Mat3 Clear Coat (V4)", nodeColor=<rgba 1 0 0.5859 1>}
  }
  MaterialExpressionNamedRerouteDeclaration_56 {
   class=MaterialExpressionNamedRerouteDeclaration
   inputs [None {from=MaterialExpressionVectorParameter_35, output=RGB}]
   properties {name="Override - Mat3 Metal LUT (V3)", nodeColor=<rgba 0 0.9062 1 1>}
  }
  MaterialExpressionNamedRerouteDeclaration_57 {
   class=MaterialExpressionNamedRerouteDeclaration
   inputs [None {from=MaterialExpressionVectorParameter_36, output=RGB}]
   properties {name="Override - Mat3 Thin Film ##1 (V3)", nodeColor=<rgba 0 1 0.5043 1>}
  }
  MaterialExpressionNamedRerouteDeclaration_58 {
   class=MaterialExpressionNamedRerouteDeclaration
   inputs [None {from=MaterialExpressionVectorParameter_37, output=RGB}]
   properties {name="Override - Mat3 Thin Film ##2 (V3)", nodeColor=<rgba 0.2957 1 0 1>}
  }
  MaterialExpressionNamedRerouteDeclaration_59 {
   class=MaterialExpressionNamedRerouteDeclaration
   inputs [None {from=MaterialExpressionVectorParameter_38, output=RGBA}]
   properties {name="Override - Mat4 Clear Coat (V4)", nodeColor=<rgba 1 0 0.5859 1>}
  }
  MaterialExpressionNamedRerouteDeclaration_6 {
   class=MaterialExpressionNamedRerouteDeclaration
   inputs [None {from=MaterialExpressionConvert_6}]
   properties {name="Clear Coat ##2 (V3)", nodeColor=<rgba 1 0 0.5859 1>}
  }
  MaterialExpressionNamedRerouteDeclaration_60 {
   class=MaterialExpressionNamedRerouteDeclaration
   inputs [None {from=MaterialExpressionVectorParameter_55, output=RGBA}]
   properties {name="Body - Mat1 Albedo (V4)", nodeColor=<rgba 0 0.0859 1 1>}
  }
  MaterialExpressionNamedRerouteDeclaration_61 {
   class=MaterialExpressionNamedRerouteDeclaration
   inputs [None {from=MaterialExpressionVectorParameter_56, output=RGBA}]
   properties {name="Body - Mat1 Remaps (V4)", nodeColor=<rgba 1 0.9739 0 1>}
  }
  MaterialExpressionNamedRerouteDeclaration_62 {
   class=MaterialExpressionNamedRerouteDeclaration
   inputs [None {from=MaterialExpressionVectorParameter_57, output=RGBA}]
   properties {name="Body - Mat2 Albedo (V4)", nodeColor=<rgba 0 0.0859 1 1>}
  }
  MaterialExpressionNamedRerouteDeclaration_63 {
   class=MaterialExpressionNamedRerouteDeclaration
   inputs [None {from=MaterialExpressionVectorParameter_68, output=RGBA}]
   properties {name="Body - Mat2 Remaps (V4)", nodeColor=<rgba 1 0.9739 0 1>}
  }
  MaterialExpressionNamedRerouteDeclaration_64 {
   class=MaterialExpressionNamedRerouteDeclaration
   inputs [None {from=MaterialExpressionVectorParameter_58, output=RGBA}]
   properties {name="Body - Mat3 Albedo (V4)", nodeColor=<rgba 0 0.0859 1 1>}
  }
  MaterialExpressionNamedRerouteDeclaration_65 {
   class=MaterialExpressionNamedRerouteDeclaration
   inputs [None {from=MaterialExpressionVectorParameter_69, output=RGBA}]
   properties {name="Body - Mat3 Remaps (V4)", nodeColor=<rgba 1 0.9739 0 1>}
  }
  MaterialExpressionNamedRerouteDeclaration_66 {
   class=MaterialExpressionNamedRerouteDeclaration
   inputs [None {from=MaterialExpressionVectorParameter_59, output=RGBA}]
   properties {name="Body - Mat4 Albedo (V4)", nodeColor=<rgba 0 0.0859 1 1>}
  }
  MaterialExpressionNamedRerouteDeclaration_67 {
   class=MaterialExpressionNamedRerouteDeclaration
   inputs [None {from=MaterialExpressionVectorParameter_70, output=RGBA}]
   properties {name="Body - Mat4 Remaps (V4)", nodeColor=<rgba 1 0.9739 0 1>}
  }
  MaterialExpressionNamedRerouteDeclaration_68 {
   class=MaterialExpressionNamedRerouteDeclaration
   inputs [None {from=MaterialExpressionVectorParameter_60, output=RGBA}]
   properties {name="Body - Mat5 Albedo (V4)", nodeColor=<rgba 0 0.0859 1 1>}
  }
  MaterialExpressionNamedRerouteDeclaration_69 {
   class=MaterialExpressionNamedRerouteDeclaration
   inputs [None {from=MaterialExpressionVectorParameter_71, output=RGBA}]
   properties {name="Body - Mat5 Remaps (V4)", nodeColor=<rgba 1 0.9739 0 1>}
  }
  MaterialExpressionNamedRerouteDeclaration_7 {
   class=MaterialExpressionNamedRerouteDeclaration
   inputs [None {from=MaterialExpressionMultiply_1}]
   properties {name="Energy Emission (Visor Off) (V3)", nodeColor=<rgba 0 0.0859 1 1>}
  }
  MaterialExpressionNamedRerouteDeclaration_70 {
   class=MaterialExpressionNamedRerouteDeclaration
   inputs [None {from=MaterialExpressionVectorParameter_61, output=RGBA}]
   properties {name="Override - Mat1 Albedo (V4)", nodeColor=<rgba 0 0.0859 1 1>}
  }
  MaterialExpressionNamedRerouteDeclaration_71 {
   class=MaterialExpressionNamedRerouteDeclaration
   inputs [None {from=MaterialExpressionVectorParameter_72, output=RGBA}]
   properties {name="Override - Mat1 Remaps (V4)", nodeColor=<rgba 1 0.9739 0 1>}
  }
  MaterialExpressionNamedRerouteDeclaration_72 {
   class=MaterialExpressionNamedRerouteDeclaration
   inputs [None {from=MaterialExpressionVectorParameter_62, output=RGBA}]
   properties {name="Override - Mat2 Albedo (V4)", nodeColor=<rgba 0 0.0859 1 1>}
  }
  MaterialExpressionNamedRerouteDeclaration_73 {
   class=MaterialExpressionNamedRerouteDeclaration
   inputs [None {from=MaterialExpressionVectorParameter_73, output=RGBA}]
   properties {name="Override - Mat2 Remaps (V4)", nodeColor=<rgba 1 0.9739 0 1>}
  }
  MaterialExpressionNamedRerouteDeclaration_74 {
   class=MaterialExpressionNamedRerouteDeclaration
   inputs [None {from=MaterialExpressionVectorParameter_63, output=RGBA}]
   properties {name="Override - Mat3 Albedo (V4)", nodeColor=<rgba 0 0.0859 1 1>}
  }
  MaterialExpressionNamedRerouteDeclaration_75 {
   class=MaterialExpressionNamedRerouteDeclaration
   inputs [None {from=MaterialExpressionVectorParameter_74, output=RGBA}]
   properties {name="Override - Mat3 Remaps (V4)", nodeColor=<rgba 1 0.9739 0 1>}
  }
  MaterialExpressionNamedRerouteDeclaration_76 {
   class=MaterialExpressionNamedRerouteDeclaration
   inputs [None {from=MaterialExpressionVectorParameter_64, output=RGBA}]
   properties {name="Override - Mat4 Albedo (V4)", nodeColor=<rgba 0 0.0859 1 1>}
  }
  MaterialExpressionNamedRerouteDeclaration_77 {
   class=MaterialExpressionNamedRerouteDeclaration
   inputs [None {from=MaterialExpressionVectorParameter_75, output=RGBA}]
   properties {name="Override - Mat4 Remaps (V4)", nodeColor=<rgba 1 0.9739 0 1>}
  }
  MaterialExpressionNamedRerouteDeclaration_78 {
   class=MaterialExpressionNamedRerouteDeclaration
   inputs [None {from=MaterialExpressionVectorParameter_65, output=RGBA}]
   properties {name="Override - Mat5 Albedo (V4)", nodeColor=<rgba 0 0.0859 1 1>}
  }
  MaterialExpressionNamedRerouteDeclaration_79 {
   class=MaterialExpressionNamedRerouteDeclaration
   inputs [None {from=MaterialExpressionVectorParameter_76, output=RGBA}]
   properties {name="Override - Mat5 Remaps (V4)", nodeColor=<rgba 1 0.9739 0 1>}
  }
  MaterialExpressionNamedRerouteDeclaration_8 {
   class=MaterialExpressionNamedRerouteDeclaration
   inputs [None {from=MaterialExpressionSetMaterialAttributes_13}]
   properties {name="Base Material (MA)", nodeColor=<rgba 0.4531 1 0 1>}
  }
  MaterialExpressionNamedRerouteDeclaration_80 {
   class=MaterialExpressionNamedRerouteDeclaration
   inputs [None {from=MaterialExpressionSaturate_12}]
   properties {name="Mat1 Variant ID (S)", nodeColor=<rgba 1 0 0 1>}
  }
  MaterialExpressionNamedRerouteDeclaration_81 {
   class=MaterialExpressionNamedRerouteDeclaration
   inputs [None {from=MaterialExpressionSwitch_26}]
   properties {name="Material 1 - Albedo (V4)", nodeColor=<rgba 0 0.0859 1 1>}
  }
  MaterialExpressionNamedRerouteDeclaration_82 {
   class=MaterialExpressionNamedRerouteDeclaration
   inputs [None {from=MaterialExpressionSwitch_28}]
   properties {name="Material 1 - Remaps (V3)", nodeColor=<rgba 1 0.9739 0 1>}
  }
  MaterialExpressionNamedRerouteDeclaration_83 {
   class=MaterialExpressionNamedRerouteDeclaration
   inputs [None {from=MaterialExpressionSwitch_8}]
   properties {name="Material 2 - Albedo (V4)", nodeColor=<rgba 0 0.0859 1 1>}
  }
  MaterialExpressionNamedRerouteDeclaration_84 {
   class=MaterialExpressionNamedRerouteDeclaration
   inputs [None {from=MaterialExpressionSwitch_15}]
   properties {name="Material 2 - Remaps (V4)", nodeColor=<rgba 1 0.9739 0 1>}
  }
  MaterialExpressionNamedRerouteDeclaration_85 {
   class=MaterialExpressionNamedRerouteDeclaration
   inputs [None {from=MaterialExpressionSwitch_16}]
   properties {name="Material 3 - Albedo (V4)", nodeColor=<rgba 0 0.0859 1 1>}
  }
  MaterialExpressionNamedRerouteDeclaration_86 {
   class=MaterialExpressionNamedRerouteDeclaration
   inputs [None {from=MaterialExpressionSwitch_17}]
   properties {name="Material 3 - Remaps (V3)", nodeColor=<rgba 1 0.9739 0 1>}
  }
  MaterialExpressionNamedRerouteDeclaration_87 {
   class=MaterialExpressionNamedRerouteDeclaration
   inputs [None {from=MaterialExpressionSwitch_18}]
   properties {name="Material 4 - Albedo (V4)", nodeColor=<rgba 0 0.0859 1 1>}
  }
  MaterialExpressionNamedRerouteDeclaration_88 {
   class=MaterialExpressionNamedRerouteDeclaration
   inputs [None {from=MaterialExpressionSwitch_19}]
   properties {name="Material 4 - Remaps (V3)", nodeColor=<rgba 1 0.9739 0 1>}
  }
  MaterialExpressionNamedRerouteDeclaration_89 {
   class=MaterialExpressionNamedRerouteDeclaration
   inputs [None {from=MaterialExpressionSwitch_20}]
   properties {name="Material 5 - Albedo (V4)", nodeColor=<rgba 0 0.0859 1 1>}
  }
  MaterialExpressionNamedRerouteDeclaration_9 {
   class=MaterialExpressionNamedRerouteDeclaration
   inputs [None {from=MaterialExpressionConvert_7}]
   properties {name="Metal LUT (V3)", nodeColor=<rgba 0 0.9062 1 1>}
  }
  MaterialExpressionNamedRerouteDeclaration_90 {
   class=MaterialExpressionNamedRerouteDeclaration
   inputs [None {from=MaterialExpressionSwitch_22}]
   properties {name="Material 5 - Remaps (V3)", nodeColor=<rgba 1 0.9739 0 1>}
  }
  MaterialExpressionNamedRerouteDeclaration_91 {
   class=MaterialExpressionNamedRerouteDeclaration
   inputs [None {from=MaterialExpressionSetMaterialAttributes_12}]
   properties {name="Armor Material (MA)", nodeColor=<rgba 0.4531 1 0 1>}
  }
  MaterialExpressionNamedRerouteDeclaration_92 {
   class=MaterialExpressionNamedRerouteDeclaration
   inputs [None {from=MaterialExpressionSaturate_11}]
   properties {name="Mat2 Variant ID (S)", nodeColor=<rgba 1 0 0 1>}
  }
  MaterialExpressionNamedRerouteDeclaration_93 {
   class=MaterialExpressionNamedRerouteDeclaration
   inputs [None {from=MaterialExpressionSaturate_10}]
   properties {name="Mat3 Variant ID (S)", nodeColor=<rgba 1 0 0 1>}
  }
  MaterialExpressionNamedRerouteDeclaration_94 {
   class=MaterialExpressionNamedRerouteDeclaration
   inputs [None {from=MaterialExpressionSaturate_9}]
   properties {name="Mat4 Variant ID (S)", nodeColor=<rgba 1 0 0 1>}
  }
  MaterialExpressionNamedRerouteDeclaration_95 {
   class=MaterialExpressionNamedRerouteDeclaration
   inputs [None {from=MaterialExpressionSaturate_8}]
   properties {name="Mat5 Variant ID (S)", nodeColor=<rgba 1 0 0 1>}
  }
  MaterialExpressionNamedRerouteDeclaration_96 {
   class=MaterialExpressionNamedRerouteDeclaration
   inputs [None {from=MaterialExpressionVectorParameter_66, output=RGB}]
   properties {name="Body - Energy Color (V3)", nodeColor=<rgba 0 0.0859 1 1>}
  }
  MaterialExpressionNamedRerouteDeclaration_97 {
   class=MaterialExpressionNamedRerouteDeclaration
   inputs [None {from=MaterialExpressionVectorParameter_67, output=RGB}]
   properties {name="Override - Energy Color (V3)", nodeColor=<rgba 0 0.0859 1 1>}
  }
  MaterialExpressionNamedRerouteDeclaration_98 {
   class=MaterialExpressionNamedRerouteDeclaration
   inputs [None {from=MaterialExpressionSwitch_24}]
   properties {name="Energy - Color (V3)", nodeColor=<rgba 0 0.0859 1 1>}
  }
  MaterialExpressionNamedRerouteDeclaration_99 {
   class=MaterialExpressionNamedRerouteDeclaration
   inputs [None {from=MaterialExpressionSetMaterialAttributes_11}]
   properties {name="Stone Material (MA)", nodeColor=<rgba 0.4531 1 0 1>}
  }
  MaterialExpressionNamedRerouteUsage_0 {class=MaterialExpressionNamedRerouteUsage}
  MaterialExpressionNamedRerouteUsage_1 {class=MaterialExpressionNamedRerouteUsage}
  MaterialExpressionNamedRerouteUsage_10 {class=MaterialExpressionNamedRerouteUsage}
  MaterialExpressionNamedRerouteUsage_100 {class=MaterialExpressionNamedRerouteUsage}
  MaterialExpressionNamedRerouteUsage_101 {class=MaterialExpressionNamedRerouteUsage}
  MaterialExpressionNamedRerouteUsage_102 {class=MaterialExpressionNamedRerouteUsage}
  MaterialExpressionNamedRerouteUsage_103 {class=MaterialExpressionNamedRerouteUsage}
  MaterialExpressionNamedRerouteUsage_104 {class=MaterialExpressionNamedRerouteUsage}
  MaterialExpressionNamedRerouteUsage_105 {class=MaterialExpressionNamedRerouteUsage}
  MaterialExpressionNamedRerouteUsage_106 {class=MaterialExpressionNamedRerouteUsage}
  MaterialExpressionNamedRerouteUsage_107 {class=MaterialExpressionNamedRerouteUsage}
  MaterialExpressionNamedRerouteUsage_108 {class=MaterialExpressionNamedRerouteUsage}
  MaterialExpressionNamedRerouteUsage_109 {class=MaterialExpressionNamedRerouteUsage}
  MaterialExpressionNamedRerouteUsage_11 {class=MaterialExpressionNamedRerouteUsage}
  MaterialExpressionNamedRerouteUsage_110 {class=MaterialExpressionNamedRerouteUsage}
  MaterialExpressionNamedRerouteUsage_111 {class=MaterialExpressionNamedRerouteUsage}
  MaterialExpressionNamedRerouteUsage_112 {class=MaterialExpressionNamedRerouteUsage}
  MaterialExpressionNamedRerouteUsage_113 {class=MaterialExpressionNamedRerouteUsage}
  MaterialExpressionNamedRerouteUsage_114 {class=MaterialExpressionNamedRerouteUsage}
  MaterialExpressionNamedRerouteUsage_115 {class=MaterialExpressionNamedRerouteUsage}
  MaterialExpressionNamedRerouteUsage_116 {class=MaterialExpressionNamedRerouteUsage}
  MaterialExpressionNamedRerouteUsage_117 {class=MaterialExpressionNamedRerouteUsage}
  MaterialExpressionNamedRerouteUsage_118 {class=MaterialExpressionNamedRerouteUsage}
  MaterialExpressionNamedRerouteUsage_119 {class=MaterialExpressionNamedRerouteUsage}
  MaterialExpressionNamedRerouteUsage_12 {class=MaterialExpressionNamedRerouteUsage}
  MaterialExpressionNamedRerouteUsage_120 {class=MaterialExpressionNamedRerouteUsage}
  MaterialExpressionNamedRerouteUsage_121 {class=MaterialExpressionNamedRerouteUsage}
  MaterialExpressionNamedRerouteUsage_122 {class=MaterialExpressionNamedRerouteUsage}
  MaterialExpressionNamedRerouteUsage_123 {class=MaterialExpressionNamedRerouteUsage}
  MaterialExpressionNamedRerouteUsage_124 {class=MaterialExpressionNamedRerouteUsage}
  MaterialExpressionNamedRerouteUsage_125 {class=MaterialExpressionNamedRerouteUsage}
  MaterialExpressionNamedRerouteUsage_126 {class=MaterialExpressionNamedRerouteUsage}
  MaterialExpressionNamedRerouteUsage_128 {class=MaterialExpressionNamedRerouteUsage}
  MaterialExpressionNamedRerouteUsage_129 {class=MaterialExpressionNamedRerouteUsage}
  MaterialExpressionNamedRerouteUsage_13 {class=MaterialExpressionNamedRerouteUsage}
  MaterialExpressionNamedRerouteUsage_130 {class=MaterialExpressionNamedRerouteUsage}
  MaterialExpressionNamedRerouteUsage_131 {class=MaterialExpressionNamedRerouteUsage}
  MaterialExpressionNamedRerouteUsage_132 {class=MaterialExpressionNamedRerouteUsage}
  MaterialExpressionNamedRerouteUsage_133 {class=MaterialExpressionNamedRerouteUsage}
  MaterialExpressionNamedRerouteUsage_134 {class=MaterialExpressionNamedRerouteUsage}
  MaterialExpressionNamedRerouteUsage_135 {class=MaterialExpressionNamedRerouteUsage}
  MaterialExpressionNamedRerouteUsage_136 {class=MaterialExpressionNamedRerouteUsage}
  MaterialExpressionNamedRerouteUsage_137 {class=MaterialExpressionNamedRerouteUsage}
  MaterialExpressionNamedRerouteUsage_138 {class=MaterialExpressionNamedRerouteUsage}
  MaterialExpressionNamedRerouteUsage_139 {class=MaterialExpressionNamedRerouteUsage}
  MaterialExpressionNamedRerouteUsage_14 {class=MaterialExpressionNamedRerouteUsage}
  MaterialExpressionNamedRerouteUsage_140 {class=MaterialExpressionNamedRerouteUsage}
  MaterialExpressionNamedRerouteUsage_141 {class=MaterialExpressionNamedRerouteUsage}
  MaterialExpressionNamedRerouteUsage_142 {class=MaterialExpressionNamedRerouteUsage}
  MaterialExpressionNamedRerouteUsage_143 {class=MaterialExpressionNamedRerouteUsage}
  MaterialExpressionNamedRerouteUsage_144 {class=MaterialExpressionNamedRerouteUsage}
  MaterialExpressionNamedRerouteUsage_145 {class=MaterialExpressionNamedRerouteUsage}
  MaterialExpressionNamedRerouteUsage_146 {class=MaterialExpressionNamedRerouteUsage}
  MaterialExpressionNamedRerouteUsage_147 {class=MaterialExpressionNamedRerouteUsage}
  MaterialExpressionNamedRerouteUsage_148 {class=MaterialExpressionNamedRerouteUsage}
  MaterialExpressionNamedRerouteUsage_149 {class=MaterialExpressionNamedRerouteUsage}
  MaterialExpressionNamedRerouteUsage_15 {class=MaterialExpressionNamedRerouteUsage}
  MaterialExpressionNamedRerouteUsage_151 {class=MaterialExpressionNamedRerouteUsage}
  MaterialExpressionNamedRerouteUsage_152 {class=MaterialExpressionNamedRerouteUsage}
  MaterialExpressionNamedRerouteUsage_154 {class=MaterialExpressionNamedRerouteUsage}
  MaterialExpressionNamedRerouteUsage_155 {class=MaterialExpressionNamedRerouteUsage}
  MaterialExpressionNamedRerouteUsage_158 {class=MaterialExpressionNamedRerouteUsage}
  MaterialExpressionNamedRerouteUsage_159 {class=MaterialExpressionNamedRerouteUsage}
  MaterialExpressionNamedRerouteUsage_16 {class=MaterialExpressionNamedRerouteUsage}
  MaterialExpressionNamedRerouteUsage_160 {class=MaterialExpressionNamedRerouteUsage}
  MaterialExpressionNamedRerouteUsage_161 {class=MaterialExpressionNamedRerouteUsage}
  MaterialExpressionNamedRerouteUsage_162 {class=MaterialExpressionNamedRerouteUsage}
  MaterialExpressionNamedRerouteUsage_163 {class=MaterialExpressionNamedRerouteUsage}
  MaterialExpressionNamedRerouteUsage_164 {class=MaterialExpressionNamedRerouteUsage}
  MaterialExpressionNamedRerouteUsage_165 {class=MaterialExpressionNamedRerouteUsage}
  MaterialExpressionNamedRerouteUsage_166 {class=MaterialExpressionNamedRerouteUsage}
  MaterialExpressionNamedRerouteUsage_167 {class=MaterialExpressionNamedRerouteUsage}
  MaterialExpressionNamedRerouteUsage_168 {class=MaterialExpressionNamedRerouteUsage}
  MaterialExpressionNamedRerouteUsage_169 {class=MaterialExpressionNamedRerouteUsage}
  MaterialExpressionNamedRerouteUsage_17 {class=MaterialExpressionNamedRerouteUsage}
  MaterialExpressionNamedRerouteUsage_170 {class=MaterialExpressionNamedRerouteUsage}
  MaterialExpressionNamedRerouteUsage_171 {class=MaterialExpressionNamedRerouteUsage}
  MaterialExpressionNamedRerouteUsage_173 {class=MaterialExpressionNamedRerouteUsage}
  MaterialExpressionNamedRerouteUsage_175 {class=MaterialExpressionNamedRerouteUsage}
  MaterialExpressionNamedRerouteUsage_177 {class=MaterialExpressionNamedRerouteUsage}
  MaterialExpressionNamedRerouteUsage_179 {class=MaterialExpressionNamedRerouteUsage}
  MaterialExpressionNamedRerouteUsage_18 {class=MaterialExpressionNamedRerouteUsage}
  MaterialExpressionNamedRerouteUsage_181 {class=MaterialExpressionNamedRerouteUsage}
  MaterialExpressionNamedRerouteUsage_182 {class=MaterialExpressionNamedRerouteUsage}
  MaterialExpressionNamedRerouteUsage_186 {class=MaterialExpressionNamedRerouteUsage}
  MaterialExpressionNamedRerouteUsage_187 {class=MaterialExpressionNamedRerouteUsage}
  MaterialExpressionNamedRerouteUsage_188 {class=MaterialExpressionNamedRerouteUsage}
  MaterialExpressionNamedRerouteUsage_189 {class=MaterialExpressionNamedRerouteUsage}
  MaterialExpressionNamedRerouteUsage_19 {class=MaterialExpressionNamedRerouteUsage}
  MaterialExpressionNamedRerouteUsage_191 {class=MaterialExpressionNamedRerouteUsage}
  MaterialExpressionNamedRerouteUsage_192 {class=MaterialExpressionNamedRerouteUsage}
  MaterialExpressionNamedRerouteUsage_193 {class=MaterialExpressionNamedRerouteUsage}
  MaterialExpressionNamedRerouteUsage_194 {class=MaterialExpressionNamedRerouteUsage}
  MaterialExpressionNamedRerouteUsage_195 {class=MaterialExpressionNamedRerouteUsage}
  MaterialExpressionNamedRerouteUsage_196 {class=MaterialExpressionNamedRerouteUsage}
  MaterialExpressionNamedRerouteUsage_198 {class=MaterialExpressionNamedRerouteUsage}
  MaterialExpressionNamedRerouteUsage_199 {class=MaterialExpressionNamedRerouteUsage}
  MaterialExpressionNamedRerouteUsage_2 {class=MaterialExpressionNamedRerouteUsage}
  MaterialExpressionNamedRerouteUsage_20 {class=MaterialExpressionNamedRerouteUsage}
  MaterialExpressionNamedRerouteUsage_200 {class=MaterialExpressionNamedRerouteUsage}
  MaterialExpressionNamedRerouteUsage_201 {class=MaterialExpressionNamedRerouteUsage}
  MaterialExpressionNamedRerouteUsage_202 {class=MaterialExpressionNamedRerouteUsage}
  MaterialExpressionNamedRerouteUsage_203 {class=MaterialExpressionNamedRerouteUsage}
  MaterialExpressionNamedRerouteUsage_204 {class=MaterialExpressionNamedRerouteUsage}
  MaterialExpressionNamedRerouteUsage_205 {class=MaterialExpressionNamedRerouteUsage}
  MaterialExpressionNamedRerouteUsage_206 {class=MaterialExpressionNamedRerouteUsage}
  MaterialExpressionNamedRerouteUsage_207 {class=MaterialExpressionNamedRerouteUsage}
  MaterialExpressionNamedRerouteUsage_208 {class=MaterialExpressionNamedRerouteUsage}
  MaterialExpressionNamedRerouteUsage_209 {class=MaterialExpressionNamedRerouteUsage}
  MaterialExpressionNamedRerouteUsage_21 {class=MaterialExpressionNamedRerouteUsage}
  MaterialExpressionNamedRerouteUsage_210 {class=MaterialExpressionNamedRerouteUsage}
  MaterialExpressionNamedRerouteUsage_211 {class=MaterialExpressionNamedRerouteUsage}
  MaterialExpressionNamedRerouteUsage_212 {class=MaterialExpressionNamedRerouteUsage}
  MaterialExpressionNamedRerouteUsage_213 {class=MaterialExpressionNamedRerouteUsage}
  MaterialExpressionNamedRerouteUsage_214 {class=MaterialExpressionNamedRerouteUsage}
  MaterialExpressionNamedRerouteUsage_215 {class=MaterialExpressionNamedRerouteUsage}
  MaterialExpressionNamedRerouteUsage_216 {class=MaterialExpressionNamedRerouteUsage}
  MaterialExpressionNamedRerouteUsage_217 {class=MaterialExpressionNamedRerouteUsage}
  MaterialExpressionNamedRerouteUsage_218 {class=MaterialExpressionNamedRerouteUsage}
  MaterialExpressionNamedRerouteUsage_219 {class=MaterialExpressionNamedRerouteUsage}
  MaterialExpressionNamedRerouteUsage_22 {class=MaterialExpressionNamedRerouteUsage}
  MaterialExpressionNamedRerouteUsage_220 {class=MaterialExpressionNamedRerouteUsage}
  MaterialExpressionNamedRerouteUsage_221 {class=MaterialExpressionNamedRerouteUsage}
  MaterialExpressionNamedRerouteUsage_222 {class=MaterialExpressionNamedRerouteUsage}
  MaterialExpressionNamedRerouteUsage_223 {class=MaterialExpressionNamedRerouteUsage}
  MaterialExpressionNamedRerouteUsage_224 {class=MaterialExpressionNamedRerouteUsage}
  MaterialExpressionNamedRerouteUsage_225 {class=MaterialExpressionNamedRerouteUsage}
  MaterialExpressionNamedRerouteUsage_226 {class=MaterialExpressionNamedRerouteUsage}
  MaterialExpressionNamedRerouteUsage_23 {class=MaterialExpressionNamedRerouteUsage}
  MaterialExpressionNamedRerouteUsage_24 {class=MaterialExpressionNamedRerouteUsage}
  MaterialExpressionNamedRerouteUsage_25 {class=MaterialExpressionNamedRerouteUsage}
  MaterialExpressionNamedRerouteUsage_26 {class=MaterialExpressionNamedRerouteUsage}
  MaterialExpressionNamedRerouteUsage_27 {class=MaterialExpressionNamedRerouteUsage}
  MaterialExpressionNamedRerouteUsage_28 {class=MaterialExpressionNamedRerouteUsage}
  MaterialExpressionNamedRerouteUsage_29 {class=MaterialExpressionNamedRerouteUsage}
  MaterialExpressionNamedRerouteUsage_3 {class=MaterialExpressionNamedRerouteUsage}
  MaterialExpressionNamedRerouteUsage_30 {class=MaterialExpressionNamedRerouteUsage}
  MaterialExpressionNamedRerouteUsage_31 {class=MaterialExpressionNamedRerouteUsage}
  MaterialExpressionNamedRerouteUsage_32 {class=MaterialExpressionNamedRerouteUsage}
  MaterialExpressionNamedRerouteUsage_33 {class=MaterialExpressionNamedRerouteUsage}
  MaterialExpressionNamedRerouteUsage_34 {class=MaterialExpressionNamedRerouteUsage}
  MaterialExpressionNamedRerouteUsage_341 {class=MaterialExpressionNamedRerouteUsage}
  MaterialExpressionNamedRerouteUsage_343 {class=MaterialExpressionNamedRerouteUsage}
  MaterialExpressionNamedRerouteUsage_35 {class=MaterialExpressionNamedRerouteUsage}
  MaterialExpressionNamedRerouteUsage_38 {class=MaterialExpressionNamedRerouteUsage}
  MaterialExpressionNamedRerouteUsage_39 {class=MaterialExpressionNamedRerouteUsage}
  MaterialExpressionNamedRerouteUsage_4 {class=MaterialExpressionNamedRerouteUsage}
  MaterialExpressionNamedRerouteUsage_40 {class=MaterialExpressionNamedRerouteUsage}
  MaterialExpressionNamedRerouteUsage_41 {class=MaterialExpressionNamedRerouteUsage}
  MaterialExpressionNamedRerouteUsage_42 {class=MaterialExpressionNamedRerouteUsage}
  MaterialExpressionNamedRerouteUsage_43 {class=MaterialExpressionNamedRerouteUsage}
  MaterialExpressionNamedRerouteUsage_44 {class=MaterialExpressionNamedRerouteUsage}
  MaterialExpressionNamedRerouteUsage_45 {class=MaterialExpressionNamedRerouteUsage}
  MaterialExpressionNamedRerouteUsage_46 {class=MaterialExpressionNamedRerouteUsage}
  MaterialExpressionNamedRerouteUsage_47 {class=MaterialExpressionNamedRerouteUsage}
  MaterialExpressionNamedRerouteUsage_48 {class=MaterialExpressionNamedRerouteUsage}
  MaterialExpressionNamedRerouteUsage_49 {class=MaterialExpressionNamedRerouteUsage}
  MaterialExpressionNamedRerouteUsage_5 {class=MaterialExpressionNamedRerouteUsage}
  MaterialExpressionNamedRerouteUsage_50 {class=MaterialExpressionNamedRerouteUsage}
  MaterialExpressionNamedRerouteUsage_51 {class=MaterialExpressionNamedRerouteUsage}
  MaterialExpressionNamedRerouteUsage_52 {class=MaterialExpressionNamedRerouteUsage}
  MaterialExpressionNamedRerouteUsage_53 {class=MaterialExpressionNamedRerouteUsage}
  MaterialExpressionNamedRerouteUsage_54 {class=MaterialExpressionNamedRerouteUsage}
  MaterialExpressionNamedRerouteUsage_56 {class=MaterialExpressionNamedRerouteUsage}
  MaterialExpressionNamedRerouteUsage_57 {class=MaterialExpressionNamedRerouteUsage}
  MaterialExpressionNamedRerouteUsage_58 {class=MaterialExpressionNamedRerouteUsage}
  MaterialExpressionNamedRerouteUsage_59 {class=MaterialExpressionNamedRerouteUsage}
  MaterialExpressionNamedRerouteUsage_6 {class=MaterialExpressionNamedRerouteUsage}
  MaterialExpressionNamedRerouteUsage_60 {class=MaterialExpressionNamedRerouteUsage}
  MaterialExpressionNamedRerouteUsage_61 {class=MaterialExpressionNamedRerouteUsage}
  MaterialExpressionNamedRerouteUsage_62 {class=MaterialExpressionNamedRerouteUsage}
  MaterialExpressionNamedRerouteUsage_63 {class=MaterialExpressionNamedRerouteUsage}
  MaterialExpressionNamedRerouteUsage_64 {class=MaterialExpressionNamedRerouteUsage}
  MaterialExpressionNamedRerouteUsage_65 {class=MaterialExpressionNamedRerouteUsage}
  MaterialExpressionNamedRerouteUsage_66 {class=MaterialExpressionNamedRerouteUsage}
  MaterialExpressionNamedRerouteUsage_67 {class=MaterialExpressionNamedRerouteUsage}
  MaterialExpressionNamedRerouteUsage_68 {class=MaterialExpressionNamedRerouteUsage}
  MaterialExpressionNamedRerouteUsage_69 {class=MaterialExpressionNamedRerouteUsage}
  MaterialExpressionNamedRerouteUsage_7 {class=MaterialExpressionNamedRerouteUsage}
  MaterialExpressionNamedRerouteUsage_70 {class=MaterialExpressionNamedRerouteUsage}
  MaterialExpressionNamedRerouteUsage_73 {class=MaterialExpressionNamedRerouteUsage}
  MaterialExpressionNamedRerouteUsage_75 {class=MaterialExpressionNamedRerouteUsage}
  MaterialExpressionNamedRerouteUsage_76 {class=MaterialExpressionNamedRerouteUsage}
  MaterialExpressionNamedRerouteUsage_79 {class=MaterialExpressionNamedRerouteUsage}
  MaterialExpressionNamedRerouteUsage_8 {class=MaterialExpressionNamedRerouteUsage}
  MaterialExpressionNamedRerouteUsage_80 {class=MaterialExpressionNamedRerouteUsage}
  MaterialExpressionNamedRerouteUsage_81 {class=MaterialExpressionNamedRerouteUsage}
  MaterialExpressionNamedRerouteUsage_82 {class=MaterialExpressionNamedRerouteUsage}
  MaterialExpressionNamedRerouteUsage_83 {class=MaterialExpressionNamedRerouteUsage}
  MaterialExpressionNamedRerouteUsage_84 {class=MaterialExpressionNamedRerouteUsage}
  MaterialExpressionNamedRerouteUsage_85 {class=MaterialExpressionNamedRerouteUsage}
  MaterialExpressionNamedRerouteUsage_86 {class=MaterialExpressionNamedRerouteUsage}
  MaterialExpressionNamedRerouteUsage_87 {class=MaterialExpressionNamedRerouteUsage}
  MaterialExpressionNamedRerouteUsage_88 {class=MaterialExpressionNamedRerouteUsage}
  MaterialExpressionNamedRerouteUsage_89 {class=MaterialExpressionNamedRerouteUsage}
  MaterialExpressionNamedRerouteUsage_9 {class=MaterialExpressionNamedRerouteUsage}
  MaterialExpressionNamedRerouteUsage_90 {class=MaterialExpressionNamedRerouteUsage}
  MaterialExpressionNamedRerouteUsage_91 {class=MaterialExpressionNamedRerouteUsage}
  MaterialExpressionNamedRerouteUsage_92 {class=MaterialExpressionNamedRerouteUsage}
  MaterialExpressionNamedRerouteUsage_93 {class=MaterialExpressionNamedRerouteUsage}
  MaterialExpressionNamedRerouteUsage_94 {class=MaterialExpressionNamedRerouteUsage}
  MaterialExpressionNamedRerouteUsage_95 {class=MaterialExpressionNamedRerouteUsage}
  MaterialExpressionNamedRerouteUsage_96 {class=MaterialExpressionNamedRerouteUsage}
  MaterialExpressionNamedRerouteUsage_97 {class=MaterialExpressionNamedRerouteUsage}
  MaterialExpressionNamedRerouteUsage_98 {class=MaterialExpressionNamedRerouteUsage}
  MaterialExpressionNamedRerouteUsage_99 {class=MaterialExpressionNamedRerouteUsage}
  MaterialExpressionOneMinus_0 {class=MaterialExpressionOneMinus, inputs [None {from=MaterialExpressionNamedRerouteUsage_6}]}
  MaterialExpressionOneMinus_1 {class=MaterialExpressionOneMinus, inputs [None {from=MaterialExpressionNamedRerouteUsage_16}]}
  MaterialExpressionOneMinus_6 {class=MaterialExpressionOneMinus, inputs [None {from=MaterialExpressionMultiply_26}]}
  MaterialExpressionOneMinus_9 {class=MaterialExpressionOneMinus, inputs [None {from=MaterialExpressionSaturate_21}]}
  MaterialExpressionSaturate_1 {class=MaterialExpressionSaturate, inputs [None {from=MaterialExpressionFunctionInput_29}]}
  MaterialExpressionSaturate_10 {class=MaterialExpressionSaturate, inputs [None {from=MaterialExpressionFunctionInput_23}]}
  MaterialExpressionSaturate_11 {class=MaterialExpressionSaturate, inputs [None {from=MaterialExpressionFunctionInput_22}]}
  MaterialExpressionSaturate_12 {class=MaterialExpressionSaturate, inputs [None {from=MaterialExpressionFunctionInput_20}]}
  MaterialExpressionSaturate_2 {class=MaterialExpressionSaturate, inputs [None {from=MaterialExpressionMultiply_0}]}
  MaterialExpressionSaturate_21 {class=MaterialExpressionSaturate, inputs [None {from=MaterialExpressionNamedRerouteUsage_341}]}
  MaterialExpressionSaturate_8 {class=MaterialExpressionSaturate, inputs [None {from=MaterialExpressionFunctionInput_25}]}
  MaterialExpressionSaturate_9 {class=MaterialExpressionSaturate, inputs [None {from=MaterialExpressionFunctionInput_24}]}
  MaterialExpressionScalarParameter_12 {
   class=MaterialExpressionScalarParameter
   properties {defaultValue=1, group=PixieParts | Materials | Energy, parameterName=Emission - Energy Color 2 Mix Mult, sortPriority=100}
  }
  MaterialExpressionScalarParameter_13 {
   class=MaterialExpressionScalarParameter
   properties {defaultValue=1, group=PixieParts | Materials | Energy, parameterName=Emission - Visor OFF Brightness Gain, sortPriority=110}
  }
  MaterialExpressionSetMaterialAttributes_1 {
   class=MaterialExpressionSetMaterialAttributes
   inputs [
    MaterialAttributes {from=MaterialExpressionMaterialFunctionCall_16, output="Material Out (MA)"}
    Tangent {from=MaterialExpressionNamedRerouteUsage_202, output=""}
    Displacement {from=MaterialExpressionNamedRerouteUsage_203, output=""}
    SubsurfaceColor {from=MaterialExpressionNamedRerouteUsage_222, output=""}
    BentNormal {from=MaterialExpressionNamedRerouteUsage_204, output=""}
    TransmittanceColor {from=MaterialExpressionNamedRerouteUsage_205, output=""}
   ]
   properties {attributeSetTypes [
     #2
     #3
     #4
     #5
     #6
    ]}
  }
  MaterialExpressionSetMaterialAttributes_11 {
   class=MaterialExpressionSetMaterialAttributes
   inputs [
    MaterialAttributes {from=MaterialExpressionNamedRerouteUsage_191, output=""}
    Refraction {from=MaterialExpressionNamedRerouteUsage_196, output=""}
   ]
   properties {attributeSetTypes [#7]}
  }
  MaterialExpressionSetMaterialAttributes_12 {
   class=MaterialExpressionSetMaterialAttributes
   inputs [
    MaterialAttributes {from=MaterialExpressionNamedRerouteUsage_192, output=""}
    Refraction {from=MaterialExpressionNamedRerouteUsage_195, output=""}
   ]
   properties {attributeSetTypes [#7]}
  }
  MaterialExpressionSetMaterialAttributes_13 {
   class=MaterialExpressionSetMaterialAttributes
   inputs [
    MaterialAttributes {from=MaterialExpressionNamedRerouteUsage_193, output=""}
    Refraction {from=MaterialExpressionNamedRerouteUsage_194, output=""}
   ]
   properties {attributeSetTypes [#7]}
  }
  MaterialExpressionSetMaterialAttributes_2 {
   class=MaterialExpressionSetMaterialAttributes
   inputs [
    MaterialAttributes {from=MaterialExpressionMaterialFunctionCall_17, output="Material Out (MA)"}
    Tangent {from=MaterialExpressionNamedRerouteUsage_206, output=""}
    Displacement {from=MaterialExpressionNamedRerouteUsage_207, output=""}
    SubsurfaceColor {from=MaterialExpressionNamedRerouteUsage_223, output=""}
    BentNormal {from=MaterialExpressionNamedRerouteUsage_208, output=""}
    TransmittanceColor {from=MaterialExpressionNamedRerouteUsage_209, output=""}
   ]
   properties {attributeSetTypes [
     #2
     #3
     #4
     #5
     #6
    ]}
  }
  MaterialExpressionSetMaterialAttributes_3 {
   class=MaterialExpressionSetMaterialAttributes
   inputs [
    MaterialAttributes {from=MaterialExpressionMaterialFunctionCall_18, output="Material Out (MA)"}
    Tangent {from=MaterialExpressionNamedRerouteUsage_210, output=""}
    Displacement {from=MaterialExpressionNamedRerouteUsage_211, output=""}
    SubsurfaceColor {from=MaterialExpressionNamedRerouteUsage_224, output=""}
    BentNormal {from=MaterialExpressionNamedRerouteUsage_212, output=""}
    TransmittanceColor {from=MaterialExpressionNamedRerouteUsage_213, output=""}
   ]
   properties {attributeSetTypes [
     #2
     #3
     #4
     #5
     #6
    ]}
  }
  MaterialExpressionSetMaterialAttributes_4 {
   class=MaterialExpressionSetMaterialAttributes
   inputs [
    MaterialAttributes {from=MaterialExpressionMaterialFunctionCall_19, output="Material Out (MA)"}
    Tangent {from=MaterialExpressionNamedRerouteUsage_214, output=""}
    Displacement {from=MaterialExpressionNamedRerouteUsage_215, output=""}
    SubsurfaceColor {from=MaterialExpressionNamedRerouteUsage_225, output=""}
    BentNormal {from=MaterialExpressionNamedRerouteUsage_216, output=""}
    TransmittanceColor {from=MaterialExpressionNamedRerouteUsage_217, output=""}
   ]
   properties {attributeSetTypes [
     #2
     #3
     #4
     #5
     #6
    ]}
  }
  MaterialExpressionSetMaterialAttributes_5 {
   class=MaterialExpressionSetMaterialAttributes
   inputs [
    MaterialAttributes {from=MaterialExpressionMaterialFunctionCall_20, output="Material Out (MA)"}
    Tangent {from=MaterialExpressionNamedRerouteUsage_218, output=""}
    Displacement {from=MaterialExpressionNamedRerouteUsage_219, output=""}
    SubsurfaceColor {from=MaterialExpressionNamedRerouteUsage_226, output=""}
    BentNormal {from=MaterialExpressionNamedRerouteUsage_220, output=""}
    TransmittanceColor {from=MaterialExpressionNamedRerouteUsage_221, output=""}
   ]
   properties {attributeSetTypes [
     #2
     #3
     #4
     #5
     #6
    ]}
  }
  MaterialExpressionSetMaterialAttributes_8 {
   class=MaterialExpressionSetMaterialAttributes
   inputs [
    MaterialAttributes {from=MaterialExpressionNamedRerouteUsage_86, output=""}
    BaseColor {from=MaterialExpressionConstant_0, output=""}
    Metallic {from=MaterialExpressionConstant_0, output=""}
    Specular {from=MaterialExpressionConstant_0, output=""}
    Roughness {from=MaterialExpressionConstant_0, output=""}
    EmissiveColor {from=MaterialExpressionConstant_0, output=""}
    Tangent {from=MaterialExpressionNamedRerouteDeclaration_9, output=""}
    Displacement {from=MaterialExpressionNamedRerouteDeclaration_5, output=""}
    SubsurfaceColor {from=MaterialExpressionNamedRerouteDeclaration_6, output=""}
    BentNormal {from=MaterialExpressionNamedRerouteDeclaration_3, output=""}
    TransmittanceColor {from=MaterialExpressionNamedRerouteDeclaration_4, output=""}
   ]
   properties {attributeSetTypes [
     #8
     #9
     #10
     #11
     #1
     #2
     #3
     #4
     #5
     #6
    ]}
  }
  MaterialExpressionStep_0 {
   class=MaterialExpressionStep
   inputs [
    Y {from="", output=""}
    X {from=MaterialExpressionAdd_0, output=""}
   ]
   properties {constY=0.01}
  }
  MaterialExpressionSwitch_0 {
   class=MaterialExpressionSwitch
   inputs [
    SwitchValue {from=MaterialExpressionNamedRerouteUsage_14, output=""}
    Default {from="", output=""}
    Body {from=MaterialExpressionNamedRerouteUsage_7, output=""}
    Override {from=MaterialExpressionNamedRerouteUsage_9, output=""}
   ]
   properties {inputs [
     {
      inputName=Body
      input {
       expression=&self:MaterialExpressionNamedRerouteUsage_7
       inputName=None
       mask=0
       maskA=0
       maskB=0
       maskG=0
       maskR=0
       outputIndex=0
      }
     }
     {
      inputName=Override
      input {
       expression=&self:MaterialExpressionNamedRerouteUsage_9
       inputName=None
       mask=0
       maskA=0
       maskB=0
       maskG=0
       maskR=0
       outputIndex=0
      }
     }
    ]}
  }
  MaterialExpressionSwitch_1 {
   class=MaterialExpressionSwitch
   inputs [
    SwitchValue {from=MaterialExpressionNamedRerouteUsage_73, output=""}
    Default {from="", output=""}
    Body {from=MaterialExpressionNamedRerouteUsage_27, output=""}
    Override {from=MaterialExpressionNamedRerouteUsage_50, output=""}
   ]
   properties {inputs [
     {
      inputName=Body
      input {
       expression=&self:MaterialExpressionNamedRerouteUsage_27
       inputName=None
       mask=0
       maskA=0
       maskB=0
       maskG=0
       maskR=0
       outputIndex=0
      }
     }
     {
      inputName=Override
      input {
       expression=&self:MaterialExpressionNamedRerouteUsage_50
       inputName=None
       mask=0
       maskA=0
       maskB=0
       maskG=0
       maskR=0
       outputIndex=0
      }
     }
    ]}
  }
  MaterialExpressionSwitch_10 {
   class=MaterialExpressionSwitch
   inputs [
    SwitchValue {from=MaterialExpressionNamedRerouteUsage_158, output=""}
    Default {from="", output=""}
    Body {from=MaterialExpressionNamedRerouteUsage_35, output=""}
    Override {from=MaterialExpressionNamedRerouteUsage_59, output=""}
   ]
   properties {inputs [
     {
      inputName=Body
      input {
       expression=&self:MaterialExpressionNamedRerouteUsage_35
       inputName=None
       mask=0
       maskA=0
       maskB=0
       maskG=0
       maskR=0
       outputIndex=0
      }
     }
     {
      inputName=Override
      input {
       expression=&self:MaterialExpressionNamedRerouteUsage_59
       inputName=None
       mask=0
       maskA=0
       maskB=0
       maskG=0
       maskR=0
       outputIndex=0
      }
     }
    ]}
  }
  MaterialExpressionSwitch_12 {
   class=MaterialExpressionSwitch
   inputs [
    SwitchValue {from=MaterialExpressionNamedRerouteUsage_159, output=""}
    Default {from="", output=""}
    Body {from=MaterialExpressionNamedRerouteUsage_39, output=""}
    Override {from=MaterialExpressionNamedRerouteUsage_60, output=""}
   ]
   properties {inputs [
     {
      inputName=Body
      input {
       expression=&self:MaterialExpressionNamedRerouteUsage_39
       inputName=None
       mask=0
       maskA=0
       maskB=0
       maskG=0
       maskR=0
       outputIndex=0
      }
     }
     {
      inputName=Override
      input {
       expression=&self:MaterialExpressionNamedRerouteUsage_60
       inputName=None
       mask=0
       maskA=0
       maskB=0
       maskG=0
       maskR=0
       outputIndex=0
      }
     }
    ]}
  }
  MaterialExpressionSwitch_13 {
   class=MaterialExpressionSwitch
   inputs [
    SwitchValue {from=MaterialExpressionNamedRerouteUsage_165, output=""}
    Default {from="", output=""}
    Body {from=MaterialExpressionNamedRerouteUsage_40, output=""}
    Override {from=MaterialExpressionNamedRerouteUsage_61, output=""}
   ]
   properties {inputs [
     {
      inputName=Body
      input {
       expression=&self:MaterialExpressionNamedRerouteUsage_40
       inputName=None
       mask=0
       maskA=0
       maskB=0
       maskG=0
       maskR=0
       outputIndex=0
      }
     }
     {
      inputName=Override
      input {
       expression=&self:MaterialExpressionNamedRerouteUsage_61
       inputName=None
       mask=0
       maskA=0
       maskB=0
       maskG=0
       maskR=0
       outputIndex=0
      }
     }
    ]}
  }
  MaterialExpressionSwitch_14 {
   class=MaterialExpressionSwitch
   inputs [
    SwitchValue {from=MaterialExpressionNamedRerouteUsage_170, output=""}
    Default {from="", output=""}
    Body {from=MaterialExpressionNamedRerouteUsage_41, output=""}
    Override {from=MaterialExpressionNamedRerouteUsage_62, output=""}
   ]
   properties {inputs [
     {
      inputName=Body
      input {
       expression=&self:MaterialExpressionNamedRerouteUsage_41
       inputName=None
       mask=0
       maskA=0
       maskB=0
       maskG=0
       maskR=0
       outputIndex=0
      }
     }
     {
      inputName=Override
      input {
       expression=&self:MaterialExpressionNamedRerouteUsage_62
       inputName=None
       mask=0
       maskA=0
       maskB=0
       maskG=0
       maskR=0
       outputIndex=0
      }
     }
    ]}
  }
  MaterialExpressionSwitch_15 {
   class=MaterialExpressionSwitch
   inputs [
    SwitchValue {from=MaterialExpressionNamedRerouteUsage_169, output=""}
    Default {from="", output=""}
    Body {from=MaterialExpressionNamedRerouteUsage_98, output=""}
    Override {from=MaterialExpressionNamedRerouteUsage_108, output=""}
   ]
   properties {inputs [
     {
      inputName=Body
      input {
       expression=&self:MaterialExpressionNamedRerouteUsage_98
       inputName=None
       mask=0
       maskA=0
       maskB=0
       maskG=0
       maskR=0
       outputIndex=0
      }
     }
     {
      inputName=Override
      input {
       expression=&self:MaterialExpressionNamedRerouteUsage_108
       inputName=None
       mask=0
       maskA=0
       maskB=0
       maskG=0
       maskR=0
       outputIndex=0
      }
     }
    ]}
  }
  MaterialExpressionSwitch_16 {
   class=MaterialExpressionSwitch
   inputs [
    SwitchValue {from=MaterialExpressionNamedRerouteUsage_171, output=""}
    Default {from="", output=""}
    Body {from=MaterialExpressionNamedRerouteUsage_99, output=""}
    Override {from=MaterialExpressionNamedRerouteUsage_109, output=""}
   ]
   properties {inputs [
     {
      inputName=Body
      input {
       expression=&self:MaterialExpressionNamedRerouteUsage_99
       inputName=None
       mask=0
       maskA=0
       maskB=0
       maskG=0
       maskR=0
       outputIndex=0
      }
     }
     {
      inputName=Override
      input {
       expression=&self:MaterialExpressionNamedRerouteUsage_109
       inputName=None
       mask=0
       maskA=0
       maskB=0
       maskG=0
       maskR=0
       outputIndex=0
      }
     }
    ]}
  }
  MaterialExpressionSwitch_17 {
   class=MaterialExpressionSwitch
   inputs [
    SwitchValue {from=MaterialExpressionNamedRerouteUsage_173, output=""}
    Default {from="", output=""}
    Body {from=MaterialExpressionNamedRerouteUsage_100, output=""}
    Override {from=MaterialExpressionNamedRerouteUsage_110, output=""}
   ]
   properties {inputs [
     {
      inputName=Body
      input {
       expression=&self:MaterialExpressionNamedRerouteUsage_100
       inputName=None
       mask=0
       maskA=0
       maskB=0
       maskG=0
       maskR=0
       outputIndex=0
      }
     }
     {
      inputName=Override
      input {
       expression=&self:MaterialExpressionNamedRerouteUsage_110
       inputName=None
       mask=0
       maskA=0
       maskB=0
       maskG=0
       maskR=0
       outputIndex=0
      }
     }
    ]}
  }
  MaterialExpressionSwitch_18 {
   class=MaterialExpressionSwitch
   inputs [
    SwitchValue {from=MaterialExpressionNamedRerouteUsage_175, output=""}
    Default {from="", output=""}
    Body {from=MaterialExpressionNamedRerouteUsage_101, output=""}
    Override {from=MaterialExpressionNamedRerouteUsage_111, output=""}
   ]
   properties {inputs [
     {
      inputName=Body
      input {
       expression=&self:MaterialExpressionNamedRerouteUsage_101
       inputName=None
       mask=0
       maskA=0
       maskB=0
       maskG=0
       maskR=0
       outputIndex=0
      }
     }
     {
      inputName=Override
      input {
       expression=&self:MaterialExpressionNamedRerouteUsage_111
       inputName=None
       mask=0
       maskA=0
       maskB=0
       maskG=0
       maskR=0
       outputIndex=0
      }
     }
    ]}
  }
  MaterialExpressionSwitch_19 {
   class=MaterialExpressionSwitch
   inputs [
    SwitchValue {from=MaterialExpressionNamedRerouteUsage_177, output=""}
    Default {from="", output=""}
    Body {from=MaterialExpressionNamedRerouteUsage_102, output=""}
    Override {from=MaterialExpressionNamedRerouteUsage_112, output=""}
   ]
   properties {inputs [
     {
      inputName=Body
      input {
       expression=&self:MaterialExpressionNamedRerouteUsage_102
       inputName=None
       mask=0
       maskA=0
       maskB=0
       maskG=0
       maskR=0
       outputIndex=0
      }
     }
     {
      inputName=Override
      input {
       expression=&self:MaterialExpressionNamedRerouteUsage_112
       inputName=None
       mask=0
       maskA=0
       maskB=0
       maskG=0
       maskR=0
       outputIndex=0
      }
     }
    ]}
  }
  MaterialExpressionSwitch_2 {
   class=MaterialExpressionSwitch
   inputs [
    SwitchValue {from=MaterialExpressionNamedRerouteUsage_76, output=""}
    Default {from="", output=""}
    Body {from=MaterialExpressionNamedRerouteUsage_28, output=""}
    Override {from=MaterialExpressionNamedRerouteUsage_51, output=""}
   ]
   properties {inputs [
     {
      inputName=Body
      input {
       expression=&self:MaterialExpressionNamedRerouteUsage_28
       inputName=None
       mask=0
       maskA=0
       maskB=0
       maskG=0
       maskR=0
       outputIndex=0
      }
     }
     {
      inputName=Override
      input {
       expression=&self:MaterialExpressionNamedRerouteUsage_51
       inputName=None
       mask=0
       maskA=0
       maskB=0
       maskG=0
       maskR=0
       outputIndex=0
      }
     }
    ]}
  }
  MaterialExpressionSwitch_20 {
   class=MaterialExpressionSwitch
   inputs [
    SwitchValue {from=MaterialExpressionNamedRerouteUsage_179, output=""}
    Default {from="", output=""}
    Body {from=MaterialExpressionNamedRerouteUsage_103, output=""}
    Override {from=MaterialExpressionLinearInterpolate_2, output=""}
   ]
   properties {inputs [
     {
      inputName=Body
      input {
       expression=&self:MaterialExpressionNamedRerouteUsage_103
       inputName=None
       mask=0
       maskA=0
       maskB=0
       maskG=0
       maskR=0
       outputIndex=0
      }
     }
     {
      inputName=Override
      input {
       expression=&self:MaterialExpressionLinearInterpolate_2
       inputName=None
       mask=0
       maskA=0
       maskB=0
       maskG=0
       maskR=0
       outputIndex=0
      }
     }
    ]}
  }
  MaterialExpressionSwitch_21 {
   class=MaterialExpressionSwitch
   inputs [
    SwitchValue {from=MaterialExpressionNamedRerouteUsage_186, output=""}
    Default {from="", output=""}
    Body {from=MaterialExpressionNamedRerouteUsage_42, output=""}
    Override {from=MaterialExpressionNamedRerouteUsage_63, output=""}
   ]
   properties {inputs [
     {
      inputName=Body
      input {
       expression=&self:MaterialExpressionNamedRerouteUsage_42
       inputName=None
       mask=0
       maskA=0
       maskB=0
       maskG=0
       maskR=0
       outputIndex=0
      }
     }
     {
      inputName=Override
      input {
       expression=&self:MaterialExpressionNamedRerouteUsage_63
       inputName=None
       mask=0
       maskA=0
       maskB=0
       maskG=0
       maskR=0
       outputIndex=0
      }
     }
    ]}
  }
  MaterialExpressionSwitch_22 {
   class=MaterialExpressionSwitch
   inputs [
    SwitchValue {from=MaterialExpressionNamedRerouteUsage_181, output=""}
    Default {from="", output=""}
    Body {from=MaterialExpressionNamedRerouteUsage_104, output=""}
    Override {from=MaterialExpressionLinearInterpolate_3, output=""}
   ]
   properties {inputs [
     {
      inputName=Body
      input {
       expression=&self:MaterialExpressionNamedRerouteUsage_104
       inputName=None
       mask=0
       maskA=0
       maskB=0
       maskG=0
       maskR=0
       outputIndex=0
      }
     }
     {
      inputName=Override
      input {
       expression=&self:MaterialExpressionLinearInterpolate_3
       inputName=None
       mask=0
       maskA=0
       maskB=0
       maskG=0
       maskR=0
       outputIndex=0
      }
     }
    ]}
  }
  MaterialExpressionSwitch_23 {
   class=MaterialExpressionSwitch
   inputs [
    SwitchValue {from=MaterialExpressionNamedRerouteUsage_187, output=""}
    Default {from="", output=""}
    Body {from=MaterialExpressionNamedRerouteUsage_43, output=""}
    Override {from=MaterialExpressionNamedRerouteUsage_64, output=""}
   ]
   properties {inputs [
     {
      inputName=Body
      input {
       expression=&self:MaterialExpressionNamedRerouteUsage_43
       inputName=None
       mask=0
       maskA=0
       maskB=0
       maskG=0
       maskR=0
       outputIndex=0
      }
     }
     {
      inputName=Override
      input {
       expression=&self:MaterialExpressionNamedRerouteUsage_64
       inputName=None
       mask=0
       maskA=0
       maskB=0
       maskG=0
       maskR=0
       outputIndex=0
      }
     }
    ]}
  }
  MaterialExpressionSwitch_24 {
   class=MaterialExpressionSwitch
   inputs [
    SwitchValue {from=MaterialExpressionNamedRerouteUsage_182, output=""}
    Default {from="", output=""}
    Body {from=MaterialExpressionNamedRerouteUsage_125, output=""}
    Override {from=MaterialExpressionNamedRerouteUsage_126, output=""}
   ]
   properties {inputs [
     {
      inputName=Body
      input {
       expression=&self:MaterialExpressionNamedRerouteUsage_125
       inputName=None
       mask=0
       maskA=0
       maskB=0
       maskG=0
       maskR=0
       outputIndex=0
      }
     }
     {
      inputName=Override
      input {
       expression=&self:MaterialExpressionNamedRerouteUsage_126
       inputName=None
       mask=0
       maskA=0
       maskB=0
       maskG=0
       maskR=0
       outputIndex=0
      }
     }
    ]}
  }
  MaterialExpressionSwitch_25 {
   class=MaterialExpressionSwitch
   inputs [
    SwitchValue {from=MaterialExpressionNamedRerouteUsage_188, output=""}
    Default {from="", output=""}
    Body {from=MaterialExpressionNamedRerouteUsage_44, output=""}
    Override {from=MaterialExpressionNamedRerouteUsage_65, output=""}
   ]
   properties {inputs [
     {
      inputName=Body
      input {
       expression=&self:MaterialExpressionNamedRerouteUsage_44
       inputName=None
       mask=0
       maskA=0
       maskB=0
       maskG=0
       maskR=0
       outputIndex=0
      }
     }
     {
      inputName=Override
      input {
       expression=&self:MaterialExpressionNamedRerouteUsage_65
       inputName=None
       mask=0
       maskA=0
       maskB=0
       maskG=0
       maskR=0
       outputIndex=0
      }
     }
    ]}
  }
  MaterialExpressionSwitch_26 {
   class=MaterialExpressionSwitch
   inputs [
    SwitchValue {from=MaterialExpressionNamedRerouteUsage_166, output=""}
    Default {from="", output=""}
    Body {from=MaterialExpressionNamedRerouteUsage_95, output=""}
    Override {from=MaterialExpressionNamedRerouteUsage_105, output=""}
   ]
   properties {inputs [
     {
      inputName=Body
      input {
       expression=&self:MaterialExpressionNamedRerouteUsage_95
       inputName=None
       mask=0
       maskA=0
       maskB=0
       maskG=0
       maskR=0
       outputIndex=0
      }
     }
     {
      inputName=Override
      input {
       expression=&self:MaterialExpressionNamedRerouteUsage_105
       inputName=None
       mask=0
       maskA=0
       maskB=0
       maskG=0
       maskR=0
       outputIndex=0
      }
     }
    ]}
  }
  MaterialExpressionSwitch_27 {
   class=MaterialExpressionSwitch
   inputs [
    SwitchValue {from=MaterialExpressionNamedRerouteUsage_189, output=""}
    Default {from="", output=""}
    Body {from=MaterialExpressionNamedRerouteUsage_45, output=""}
    Override {from=MaterialExpressionNamedRerouteUsage_66, output=""}
   ]
   properties {inputs [
     {
      inputName=Body
      input {
       expression=&self:MaterialExpressionNamedRerouteUsage_45
       inputName=None
       mask=0
       maskA=0
       maskB=0
       maskG=0
       maskR=0
       outputIndex=0
      }
     }
     {
      inputName=Override
      input {
       expression=&self:MaterialExpressionNamedRerouteUsage_66
       inputName=None
       mask=0
       maskA=0
       maskB=0
       maskG=0
       maskR=0
       outputIndex=0
      }
     }
    ]}
  }
  MaterialExpressionSwitch_28 {
   class=MaterialExpressionSwitch
   inputs [
    SwitchValue {from=MaterialExpressionNamedRerouteUsage_167, output=""}
    Default {from="", output=""}
    Body {from=MaterialExpressionNamedRerouteUsage_96, output=""}
    Override {from=MaterialExpressionNamedRerouteUsage_106, output=""}
   ]
   properties {inputs [
     {
      inputName=Body
      input {
       expression=&self:MaterialExpressionNamedRerouteUsage_96
       inputName=None
       mask=0
       maskA=0
       maskB=0
       maskG=0
       maskR=0
       outputIndex=0
      }
     }
     {
      inputName=Override
      input {
       expression=&self:MaterialExpressionNamedRerouteUsage_106
       inputName=None
       mask=0
       maskA=0
       maskB=0
       maskG=0
       maskR=0
       outputIndex=0
      }
     }
    ]}
  }
  MaterialExpressionSwitch_29 {
   class=MaterialExpressionSwitch
   inputs [
    SwitchValue {from=MaterialExpressionNamedRerouteUsage_198, output=""}
    Default {from="", output=""}
    Body {from=MaterialExpressionNamedRerouteUsage_46, output=""}
    Override {from=MaterialExpressionLinearInterpolate_4, output=""}
   ]
   properties {inputs [
     {
      inputName=Body
      input {
       expression=&self:MaterialExpressionNamedRerouteUsage_46
       inputName=None
       mask=0
       maskA=0
       maskB=0
       maskG=0
       maskR=0
       outputIndex=0
      }
     }
     {
      inputName=Override
      input {
       expression=&self:MaterialExpressionLinearInterpolate_4
       inputName=None
       mask=0
       maskA=0
       maskB=0
       maskG=0
       maskR=0
       outputIndex=0
      }
     }
    ]}
  }
  MaterialExpressionSwitch_3 {
   class=MaterialExpressionSwitch
   inputs [
    SwitchValue {from=MaterialExpressionNamedRerouteUsage_79, output=""}
    Default {from="", output=""}
    Body {from=MaterialExpressionNamedRerouteUsage_29, output=""}
    Override {from=MaterialExpressionNamedRerouteUsage_52, output=""}
   ]
   properties {inputs [
     {
      inputName=Body
      input {
       expression=&self:MaterialExpressionNamedRerouteUsage_29
       inputName=None
       mask=0
       maskA=0
       maskB=0
       maskG=0
       maskR=0
       outputIndex=0
      }
     }
     {
      inputName=Override
      input {
       expression=&self:MaterialExpressionNamedRerouteUsage_52
       inputName=None
       mask=0
       maskA=0
       maskB=0
       maskG=0
       maskR=0
       outputIndex=0
      }
     }
    ]}
  }
  MaterialExpressionSwitch_30 {
   class=MaterialExpressionSwitch
   inputs [
    SwitchValue {from=MaterialExpressionNamedRerouteUsage_199, output=""}
    Default {from="", output=""}
    Body {from=MaterialExpressionNamedRerouteUsage_47, output=""}
    Override {from=MaterialExpressionLinearInterpolate_5, output=""}
   ]
   properties {inputs [
     {
      inputName=Body
      input {
       expression=&self:MaterialExpressionNamedRerouteUsage_47
       inputName=None
       mask=0
       maskA=0
       maskB=0
       maskG=0
       maskR=0
       outputIndex=0
      }
     }
     {
      inputName=Override
      input {
       expression=&self:MaterialExpressionLinearInterpolate_5
       inputName=None
       mask=0
       maskA=0
       maskB=0
       maskG=0
       maskR=0
       outputIndex=0
      }
     }
    ]}
  }
  MaterialExpressionSwitch_31 {
   class=MaterialExpressionSwitch
   inputs [
    SwitchValue {from=MaterialExpressionNamedRerouteUsage_200, output=""}
    Default {from="", output=""}
    Body {from=MaterialExpressionNamedRerouteUsage_48, output=""}
    Override {from=MaterialExpressionLinearInterpolate_6, output=""}
   ]
   properties {inputs [
     {
      inputName=Body
      input {
       expression=&self:MaterialExpressionNamedRerouteUsage_48
       inputName=None
       mask=0
       maskA=0
       maskB=0
       maskG=0
       maskR=0
       outputIndex=0
      }
     }
     {
      inputName=Override
      input {
       expression=&self:MaterialExpressionLinearInterpolate_6
       inputName=None
       mask=0
       maskA=0
       maskB=0
       maskG=0
       maskR=0
       outputIndex=0
      }
     }
    ]}
  }
  MaterialExpressionSwitch_32 {
   class=MaterialExpressionSwitch
   inputs [
    SwitchValue {from=MaterialExpressionNamedRerouteUsage_201, output=""}
    Default {from="", output=""}
    Body {from=MaterialExpressionNamedRerouteUsage_49, output=""}
    Override {from=MaterialExpressionLinearInterpolate_7, output=""}
   ]
   properties {inputs [
     {
      inputName=Body
      input {
       expression=&self:MaterialExpressionNamedRerouteUsage_49
       inputName=None
       mask=0
       maskA=0
       maskB=0
       maskG=0
       maskR=0
       outputIndex=0
      }
     }
     {
      inputName=Override
      input {
       expression=&self:MaterialExpressionLinearInterpolate_7
       inputName=None
       mask=0
       maskA=0
       maskB=0
       maskG=0
       maskR=0
       outputIndex=0
      }
     }
    ]}
  }
  MaterialExpressionSwitch_34 {
   class=MaterialExpressionSwitch
   inputs [
    SwitchValue {from=MaterialExpressionNamedRerouteUsage_89, output=""}
    Default {from="", output=""}
    Body {from=MaterialExpressionNamedRerouteUsage_87, output=""}
    Override {from=MaterialExpressionNamedRerouteUsage_88, output=""}
   ]
   properties {inputs [
     {
      inputName=Body
      input {
       expression=&self:MaterialExpressionNamedRerouteUsage_87
       inputName=None
       mask=0
       maskA=0
       maskB=0
       maskG=0
       maskR=0
       outputIndex=0
      }
     }
     {
      inputName=Override
      input {
       expression=&self:MaterialExpressionNamedRerouteUsage_88
       inputName=None
       mask=0
       maskA=0
       maskB=0
       maskG=0
       maskR=0
       outputIndex=0
      }
     }
    ]}
  }
  MaterialExpressionSwitch_4 {
   class=MaterialExpressionSwitch
   inputs [
    SwitchValue {from=MaterialExpressionNamedRerouteUsage_82, output=""}
    Default {from="", output=""}
    Body {from=MaterialExpressionNamedRerouteUsage_30, output=""}
    Override {from=MaterialExpressionNamedRerouteUsage_53, output=""}
   ]
   properties {inputs [
     {
      inputName=Body
      input {
       expression=&self:MaterialExpressionNamedRerouteUsage_30
       inputName=None
       mask=0
       maskA=0
       maskB=0
       maskG=0
       maskR=0
       outputIndex=0
      }
     }
     {
      inputName=Override
      input {
       expression=&self:MaterialExpressionNamedRerouteUsage_53
       inputName=None
       mask=0
       maskA=0
       maskB=0
       maskG=0
       maskR=0
       outputIndex=0
      }
     }
    ]}
  }
  MaterialExpressionSwitch_5 {
   class=MaterialExpressionSwitch
   inputs [
    SwitchValue {from=MaterialExpressionNamedRerouteUsage_91, output=""}
    Default {from="", output=""}
    Body {from=MaterialExpressionNamedRerouteUsage_31, output=""}
    Override {from=MaterialExpressionNamedRerouteUsage_54, output=""}
   ]
   properties {inputs [
     {
      inputName=Body
      input {
       expression=&self:MaterialExpressionNamedRerouteUsage_31
       inputName=None
       mask=0
       maskA=0
       maskB=0
       maskG=0
       maskR=0
       outputIndex=0
      }
     }
     {
      inputName=Override
      input {
       expression=&self:MaterialExpressionNamedRerouteUsage_54
       inputName=None
       mask=0
       maskA=0
       maskB=0
       maskG=0
       maskR=0
       outputIndex=0
      }
     }
    ]}
  }
  MaterialExpressionSwitch_6 {
   class=MaterialExpressionSwitch
   inputs [
    SwitchValue {from=MaterialExpressionNamedRerouteUsage_92, output=""}
    Default {from="", output=""}
    Body {from=MaterialExpressionNamedRerouteUsage_32, output=""}
    Override {from=MaterialExpressionNamedRerouteUsage_56, output=""}
   ]
   properties {inputs [
     {
      inputName=Body
      input {
       expression=&self:MaterialExpressionNamedRerouteUsage_32
       inputName=None
       mask=0
       maskA=0
       maskB=0
       maskG=0
       maskR=0
       outputIndex=0
      }
     }
     {
      inputName=Override
      input {
       expression=&self:MaterialExpressionNamedRerouteUsage_56
       inputName=None
       mask=0
       maskA=0
       maskB=0
       maskG=0
       maskR=0
       outputIndex=0
      }
     }
    ]}
  }
  MaterialExpressionSwitch_7 {
   class=MaterialExpressionSwitch
   inputs [
    SwitchValue {from=MaterialExpressionNamedRerouteUsage_93, output=""}
    Default {from="", output=""}
    Body {from=MaterialExpressionNamedRerouteUsage_33, output=""}
    Override {from=MaterialExpressionNamedRerouteUsage_57, output=""}
   ]
   properties {inputs [
     {
      inputName=Body
      input {
       expression=&self:MaterialExpressionNamedRerouteUsage_33
       inputName=None
       mask=0
       maskA=0
       maskB=0
       maskG=0
       maskR=0
       outputIndex=0
      }
     }
     {
      inputName=Override
      input {
       expression=&self:MaterialExpressionNamedRerouteUsage_57
       inputName=None
       mask=0
       maskA=0
       maskB=0
       maskG=0
       maskR=0
       outputIndex=0
      }
     }
    ]}
  }
  MaterialExpressionSwitch_8 {
   class=MaterialExpressionSwitch
   inputs [
    SwitchValue {from=MaterialExpressionNamedRerouteUsage_168, output=""}
    Default {from="", output=""}
    Body {from=MaterialExpressionNamedRerouteUsage_97, output=""}
    Override {from=MaterialExpressionNamedRerouteUsage_107, output=""}
   ]
   properties {inputs [
     {
      inputName=Body
      input {
       expression=&self:MaterialExpressionNamedRerouteUsage_97
       inputName=None
       mask=0
       maskA=0
       maskB=0
       maskG=0
       maskR=0
       outputIndex=0
      }
     }
     {
      inputName=Override
      input {
       expression=&self:MaterialExpressionNamedRerouteUsage_107
       inputName=None
       mask=0
       maskA=0
       maskB=0
       maskG=0
       maskR=0
       outputIndex=0
      }
     }
    ]}
  }
  MaterialExpressionSwitch_9 {
   class=MaterialExpressionSwitch
   inputs [
    SwitchValue {from=MaterialExpressionNamedRerouteUsage_94, output=""}
    Default {from="", output=""}
    Body {from=MaterialExpressionNamedRerouteUsage_34, output=""}
    Override {from=MaterialExpressionNamedRerouteUsage_58, output=""}
   ]
   properties {inputs [
     {
      inputName=Body
      input {
       expression=&self:MaterialExpressionNamedRerouteUsage_34
       inputName=None
       mask=0
       maskA=0
       maskB=0
       maskG=0
       maskR=0
       outputIndex=0
      }
     }
     {
      inputName=Override
      input {
       expression=&self:MaterialExpressionNamedRerouteUsage_58
       inputName=None
       mask=0
       maskA=0
       maskB=0
       maskG=0
       maskR=0
       outputIndex=0
      }
     }
    ]}
  }
  MaterialExpressionVectorParameter_0 {
   class=MaterialExpressionVectorParameter
   properties {
    group=PixieParts | Variants | Body | Material 1
    parameterName=Body - Mat1 Clear Coat
    sortPriority=110
    channelNames {a=Top Coat Material Roughness Mix, b=Under Coat Roughness, g=Under Coat Metallic Mult, r=Clear Coat Intensity}
    defaultValue=<rgba 0 1 0.1 0>
   }
  }
  MaterialExpressionVectorParameter_1 {
   class=MaterialExpressionVectorParameter
   properties {
    group=PixieParts | Variants | Body | Material 1
    parameterName=Body - Mat1 Metal LUT
    sortPriority=100
    channelNames {a=n/a, b=n/a, g=LUT Channel, r=LUT Intensity}
   }
  }
  MaterialExpressionVectorParameter_10 {
   class=MaterialExpressionVectorParameter
   properties {
    group=PixieParts | Variants | Body | Material 2
    parameterName=Body - Mat2 Clear Coat
    sortPriority=110
    channelNames {a=Top Coat Material Roughness Mix, b=Under Coat Roughness, g=Under Coat Metallic Mult, r=Clear Coat Intensity}
    defaultValue=<rgba 0 1 0.1 0>
   }
  }
  MaterialExpressionVectorParameter_11 {
   class=MaterialExpressionVectorParameter
   properties {
    group=PixieParts | Variants | Body | Material 2
    parameterName=Body - Mat2 Metal LUT
    sortPriority=100
    channelNames {a=n/a, b=n/a, g=LUT Channel, r=LUT Intensity}
   }
  }
  MaterialExpressionVectorParameter_12 {
   class=MaterialExpressionVectorParameter
   properties {
    group=PixieParts | Variants | Body | Material 2
    parameterName="Body - Mat2 Thin Film ##1"
    sortPriority=120
    channelNames {a=n/a, b=Roughness Influence, g=Thin Film Exponent, r=Thin Film Intensity}
    defaultValue=<rgba 0 2 0 0>
   }
  }
  MaterialExpressionVectorParameter_13 {
   class=MaterialExpressionVectorParameter
   properties {
    group=PixieParts | Variants | Body | Material 2
    parameterName="Body - Mat2 Thin Film ##2"
    sortPriority=130
    channelNames {a=n/a, b=Thin Film Offset, g=Thin Film Scale, r=Thin Film Wrap}
    defaultValue=<rgba 0.5 1 0 0>
   }
  }
  MaterialExpressionVectorParameter_14 {
   class=MaterialExpressionVectorParameter
   properties {
    group=PixieParts | Variants | Body | Material 3
    parameterName=Body - Mat3 Clear Coat
    sortPriority=110
    channelNames {a=Top Coat Material Roughness Mix, b=Under Coat Roughness, g=Under Coat Metallic Mult, r=Clear Coat Intensity}
    defaultValue=<rgba 0 1 0.1 0>
   }
  }
  MaterialExpressionVectorParameter_15 {
   class=MaterialExpressionVectorParameter
   properties {
    group=PixieParts | Variants | Body | Material 3
    parameterName=Body - Mat3 Metal LUT
    sortPriority=100
    channelNames {a=n/a, b=n/a, g=LUT Channel, r=LUT Intensity}
   }
  }
  MaterialExpressionVectorParameter_16 {
   class=MaterialExpressionVectorParameter
   properties {
    group=PixieParts | Variants | Body | Material 3
    parameterName="Body - Mat3 Thin Film ##1"
    sortPriority=120
    channelNames {a=n/a, b=Roughness Influence, g=Thin Film Exponent, r=Thin Film Intensity}
    defaultValue=<rgba 0 2 0 0>
   }
  }
  MaterialExpressionVectorParameter_17 {
   class=MaterialExpressionVectorParameter
   properties {
    group=PixieParts | Variants | Body | Material 3
    parameterName="Body - Mat3 Thin Film ##2"
    sortPriority=130
    channelNames {a=n/a, b=Thin Film Offset, g=Thin Film Scale, r=Thin Film Wrap}
    defaultValue=<rgba 0.5 1 0 0>
   }
  }
  MaterialExpressionVectorParameter_18 {
   class=MaterialExpressionVectorParameter
   properties {
    group=PixieParts | Variants | Body | Material 4
    parameterName=Body - Mat4 Clear Coat
    sortPriority=110
    channelNames {a=Top Coat Material Roughness Mix, b=Under Coat Roughness, g=Under Coat Metallic Mult, r=Clear Coat Intensity}
    defaultValue=<rgba 0 1 0.1 0>
   }
  }
  MaterialExpressionVectorParameter_19 {
   class=MaterialExpressionVectorParameter
   properties {
    group=PixieParts | Variants | Body | Material 4
    parameterName=Body - Mat4 Metal LUT
    sortPriority=100
    channelNames {a=n/a, b=n/a, g=LUT Channel, r=LUT Intensity}
   }
  }
  MaterialExpressionVectorParameter_20 {
   class=MaterialExpressionVectorParameter
   properties {
    group=PixieParts | Variants | Body | Material 4
    parameterName="Body - Mat4 Thin Film ##1"
    sortPriority=120
    channelNames {a=n/a, b=Roughness Influence, g=Thin Film Exponent, r=Thin Film Intensity}
    defaultValue=<rgba 0 2 0 0>
   }
  }
  MaterialExpressionVectorParameter_21 {
   class=MaterialExpressionVectorParameter
   properties {
    group=PixieParts | Variants | Body | Material 4
    parameterName="Body - Mat4 Thin Film ##2"
    sortPriority=130
    channelNames {a=n/a, b=Thin Film Offset, g=Thin Film Scale, r=Thin Film Wrap}
    defaultValue=<rgba 0.5 1 0 0>
   }
  }
  MaterialExpressionVectorParameter_22 {
   class=MaterialExpressionVectorParameter
   properties {
    group=PixieParts | Variants | Body | Material 5
    parameterName=Body - Mat5 Clear Coat
    sortPriority=110
    channelNames {a=Top Coat Material Roughness Mix, b=Under Coat Roughness, g=Under Coat Metallic Mult, r=Clear Coat Intensity}
    defaultValue=<rgba 0 1 0.1 0>
   }
  }
  MaterialExpressionVectorParameter_23 {
   class=MaterialExpressionVectorParameter
   properties {
    group=PixieParts | Variants | Body | Material 5
    parameterName=Body - Mat5 Metal LUT
    sortPriority=100
    channelNames {a=n/a, b=n/a, g=LUT Channel, r=LUT Intensity}
   }
  }
  MaterialExpressionVectorParameter_24 {
   class=MaterialExpressionVectorParameter
   properties {
    group=PixieParts | Variants | Body | Material 5
    parameterName="Body - Mat5 Thin Film ##1"
    sortPriority=120
    channelNames {a=n/a, b=Roughness Influence, g=Thin Film Exponent, r=Thin Film Intensity}
    defaultValue=<rgba 0 2 0 0>
   }
  }
  MaterialExpressionVectorParameter_25 {
   class=MaterialExpressionVectorParameter
   properties {
    group=PixieParts | Variants | Body | Material 5
    parameterName="Body - Mat5 Thin Film ##2"
    sortPriority=130
    channelNames {a=n/a, b=Thin Film Offset, g=Thin Film Scale, r=Thin Film Wrap}
    defaultValue=<rgba 0.5 1 0 0>
   }
  }
  MaterialExpressionVectorParameter_26 {
   class=MaterialExpressionVectorParameter
   properties {
    group=PixieParts | Variants | Override | Material 1
    parameterName=Override - Mat1 Clear Coat
    sortPriority=110
    channelNames {a=Top Coat Material Roughness Mix, b=Under Coat Roughness, g=Under Coat Metallic Mult, r=Clear Coat Intensity}
    defaultValue=<rgba 0 1 0.1 0>
   }
  }
  MaterialExpressionVectorParameter_27 {
   class=MaterialExpressionVectorParameter
   properties {
    group=PixieParts | Variants | Override | Material 1
    parameterName=Override - Mat1 Metal LUT
    sortPriority=100
    channelNames {a=n/a, b=n/a, g=LUT Channel, r=LUT Intensity}
   }
  }
  MaterialExpressionVectorParameter_28 {
   class=MaterialExpressionVectorParameter
   properties {
    group=PixieParts | Variants | Override | Material 1
    parameterName="Override - Mat1 Thin Film ##1"
    sortPriority=120
    channelNames {a=n/a, b=Roughness Influence, g=Thin Film Exponent, r=Thin Film Intensity}
    defaultValue=<rgba 0 2 0 0>
   }
  }
  MaterialExpressionVectorParameter_29 {
   class=MaterialExpressionVectorParameter
   properties {
    group=PixieParts | Variants | Override | Material 1
    parameterName="Override - Mat1 Thin Film ##2"
    sortPriority=130
    channelNames {a=n/a, b=Thin Film Offset, g=Thin Film Scale, r=Thin Film Wrap}
    defaultValue=<rgba 0.5 1 0 0>
   }
  }
  MaterialExpressionVectorParameter_3 {
   class=MaterialExpressionVectorParameter
   properties {
    group=PixieParts | Variants | Body | Material 1
    parameterName="Body - Mat1 Thin Film ##1"
    sortPriority=120
    channelNames {a=n/a, b=Roughness Influence, g=Thin Film Exponent, r=Thin Film Intensity}
    defaultValue=<rgba 0 2 0 0>
   }
  }
  MaterialExpressionVectorParameter_30 {
   class=MaterialExpressionVectorParameter
   properties {
    group=PixieParts | Variants | Override | Material 2
    parameterName=Override - Mat2 Clear Coat
    sortPriority=110
    channelNames {a=Top Coat Material Roughness Mix, b=Under Coat Roughness, g=Under Coat Metallic Mult, r=Clear Coat Intensity}
    defaultValue=<rgba 0 1 0.1 0>
   }
  }
  MaterialExpressionVectorParameter_31 {
   class=MaterialExpressionVectorParameter
   properties {
    group=PixieParts | Variants | Override | Material 2
    parameterName=Override - Mat2 Metal LUT
    sortPriority=100
    channelNames {a=n/a, b=n/a, g=LUT Channel, r=LUT Intensity}
   }
  }
  MaterialExpressionVectorParameter_32 {
   class=MaterialExpressionVectorParameter
   properties {
    group=PixieParts | Variants | Override | Material 2
    parameterName="Override - Mat2 Thin Film ##1"
    sortPriority=120
    channelNames {a=n/a, b=Roughness Influence, g=Thin Film Exponent, r=Thin Film Intensity}
    defaultValue=<rgba 0 2 0 0>
   }
  }
  MaterialExpressionVectorParameter_33 {
   class=MaterialExpressionVectorParameter
   properties {
    group=PixieParts | Variants | Override | Material 2
    parameterName="Override - Mat2 Thin Film ##2"
    sortPriority=130
    channelNames {a=n/a, b=Thin Film Offset, g=Thin Film Scale, r=Thin Film Wrap}
    defaultValue=<rgba 0.5 1 0 0>
   }
  }
  MaterialExpressionVectorParameter_34 {
   class=MaterialExpressionVectorParameter
   properties {
    group=PixieParts | Variants | Override | Material 3
    parameterName=Override - Mat3 Clear Coat
    sortPriority=110
    channelNames {a=Top Coat Material Roughness Mix, b=Under Coat Roughness, g=Under Coat Metallic Mult, r=Clear Coat Intensity}
    defaultValue=<rgba 0 1 0.1 0>
   }
  }
  MaterialExpressionVectorParameter_35 {
   class=MaterialExpressionVectorParameter
   properties {
    group=PixieParts | Variants | Override | Material 3
    parameterName=Override - Mat3 Metal LUT
    sortPriority=100
    channelNames {a=n/a, b=n/a, g=LUT Channel, r=LUT Intensity}
   }
  }
  MaterialExpressionVectorParameter_36 {
   class=MaterialExpressionVectorParameter
   properties {
    group=PixieParts | Variants | Override | Material 3
    parameterName="Override - Mat3 Thin Film ##1"
    sortPriority=120
    channelNames {a=n/a, b=Roughness Influence, g=Thin Film Exponent, r=Thin Film Intensity}
    defaultValue=<rgba 0 2 0 0>
   }
  }
  MaterialExpressionVectorParameter_37 {
   class=MaterialExpressionVectorParameter
   properties {
    group=PixieParts | Variants | Override | Material 3
    parameterName="Override - Mat3 Thin Film ##2"
    sortPriority=130
    channelNames {a=n/a, b=Thin Film Offset, g=Thin Film Scale, r=Thin Film Wrap}
    defaultValue=<rgba 0.5 1 0 0>
   }
  }
  MaterialExpressionVectorParameter_38 {
   class=MaterialExpressionVectorParameter
   properties {
    group=PixieParts | Variants | Override | Material 4
    parameterName=Override - Mat4 Clear Coat
    sortPriority=110
    channelNames {a=Top Coat Material Roughness Mix, b=Under Coat Roughness, g=Under Coat Metallic Mult, r=Clear Coat Intensity}
    defaultValue=<rgba 0 1 0.1 0>
   }
  }
  MaterialExpressionVectorParameter_39 {
   class=MaterialExpressionVectorParameter
   properties {
    group=PixieParts | Variants | Override | Material 4
    parameterName=Override - Mat4 Metal LUT
    sortPriority=100
    channelNames {a=n/a, b=n/a, g=LUT Channel, r=LUT Intensity}
   }
  }
  MaterialExpressionVectorParameter_4 {
   class=MaterialExpressionVectorParameter
   properties {
    group=PixieParts | Variants | Body | Material 1
    parameterName="Body - Mat1 Thin Film ##2"
    sortPriority=130
    channelNames {a=n/a, b=Thin Film Offset, g=Thin Film Scale, r=Thin Film Wrap}
    defaultValue=<rgba 0.5 1 0 0>
   }
  }
  MaterialExpressionVectorParameter_40 {
   class=MaterialExpressionVectorParameter
   properties {
    group=PixieParts | Variants | Override | Material 4
    parameterName="Override - Mat4 Thin Film ##1"
    sortPriority=120
    channelNames {a=n/a, b=Roughness Influence, g=Thin Film Exponent, r=Thin Film Intensity}
    defaultValue=<rgba 0 2 0 0>
   }
  }
  MaterialExpressionVectorParameter_41 {
   class=MaterialExpressionVectorParameter
   properties {
    group=PixieParts | Variants | Override | Material 4
    parameterName="Override - Mat4 Thin Film ##2"
    sortPriority=130
    channelNames {a=n/a, b=Thin Film Offset, g=Thin Film Scale, r=Thin Film Wrap}
    defaultValue=<rgba 0.5 1 0 0>
   }
  }
  MaterialExpressionVectorParameter_42 {
   class=MaterialExpressionVectorParameter
   properties {
    group=PixieParts | Variants | Override | Material 5
    parameterName=Override - Mat5 Clear Coat
    sortPriority=110
    channelNames {a=Top Coat Material Roughness Mix, b=Under Coat Roughness, g=Under Coat Metallic Mult, r=Clear Coat Intensity}
    defaultValue=<rgba 0 1 0.1 0>
   }
  }
  MaterialExpressionVectorParameter_43 {
   class=MaterialExpressionVectorParameter
   properties {
    group=PixieParts | Variants | Override | Material 5
    parameterName=Override - Mat5 Metal LUT
    sortPriority=100
    channelNames {a=n/a, b=n/a, g=LUT Channel, r=LUT Intensity}
   }
  }
  MaterialExpressionVectorParameter_44 {
   class=MaterialExpressionVectorParameter
   properties {
    group=PixieParts | Variants | Override | Material 5
    parameterName="Override - Mat5 Thin Film ##1"
    sortPriority=120
    channelNames {a=n/a, b=Roughness Influence, g=Thin Film Exponent, r=Thin Film Intensity}
    defaultValue=<rgba 0 2 0 0>
   }
  }
  MaterialExpressionVectorParameter_45 {
   class=MaterialExpressionVectorParameter
   properties {
    group=PixieParts | Variants | Override | Material 5
    parameterName="Override - Mat5 Thin Film ##2"
    sortPriority=130
    channelNames {a=n/a, b=Thin Film Offset, g=Thin Film Scale, r=Thin Film Wrap}
    defaultValue=<rgba 0.5 1 0 0>
   }
  }
  MaterialExpressionVectorParameter_47 {
   class=MaterialExpressionVectorParameter
   properties {
    group=PixieParts | Variants | Body | Energy
    parameterName=Body - Energy Settings
    sortPriority=10
    channelNames {a=Saturation Offset, b=Brightness, g=Color 2 Mix, r=Pre-Tint Exp}
    defaultValue=<rgba 1 1 1 0>
   }
  }
  MaterialExpressionVectorParameter_48 {
   class=MaterialExpressionVectorParameter
   properties {
    group=PixieParts | Variants | Override | Energy
    parameterName=Override - Energy Settings
    sortPriority=10
    channelNames {a=Saturation Offset, b=Brightness, g=Color 2 Mix, r=Pre-Tint Exp}
    defaultValue=<rgba 1 1 1 0>
   }
  }
  MaterialExpressionVectorParameter_5 {
   class=MaterialExpressionVectorParameter
   properties {
    group=PixieParts | Variants | Body | Energy
    parameterName=Body - Energy Color 2
    sortPriority=30
    channelNames {a=n/a, b=Tint B, g=Tint G, r=Tint R}
    defaultValue=<rgba 1 1 1 0>
   }
  }
  MaterialExpressionVectorParameter_55 {
   class=MaterialExpressionVectorParameter
   properties {
    group=PixieParts | Variants | Body | Material 1
    parameterName=Body - Mat1 Albedo
    sortPriority=10
    channelNames {a=Pre-Tint Exp, b=Tint B, g=Tint G, r=Tint R}
    defaultValue=<rgba 1 1 1 1>
   }
  }
  MaterialExpressionVectorParameter_56 {
   class=MaterialExpressionVectorParameter
   properties {
    group=PixieParts | Variants | Body | Material 1
    parameterName=Body - Mat1 Remaps
    sortPriority=50
    channelNames {a=AO Mix, b=Roughness Exp, g=Specular Exp, r=Metallic}
    defaultValue=<rgba 1 1 1 0>
   }
  }
  MaterialExpressionVectorParameter_57 {
   class=MaterialExpressionVectorParameter
   properties {
    group=PixieParts | Variants | Body | Material 2
    parameterName=Body - Mat2 Albedo
    sortPriority=10
    channelNames {a=Pre-Tint Exp, b=Tint B, g=Tint G, r=Tint R}
    defaultValue=<rgba 1 1 1 1>
   }
  }
  MaterialExpressionVectorParameter_58 {
   class=MaterialExpressionVectorParameter
   properties {
    group=PixieParts | Variants | Body | Material 3
    parameterName=Body - Mat3 Albedo
    sortPriority=10
    channelNames {a=Pre-Tint Exp, b=Tint B, g=Tint G, r=Tint R}
    defaultValue=<rgba 1 1 1 1>
   }
  }
  MaterialExpressionVectorParameter_59 {
   class=MaterialExpressionVectorParameter
   properties {
    group=PixieParts | Variants | Body | Material 4
    parameterName=Body - Mat4 Albedo
    sortPriority=10
    channelNames {a=Pre-Tint Exp, b=Tint B, g=Tint G, r=Tint R}
    defaultValue=<rgba 1 1 1 1>
   }
  }
  MaterialExpressionVectorParameter_60 {
   class=MaterialExpressionVectorParameter
   properties {
    group=PixieParts | Variants | Body | Material 5
    parameterName=Body - Mat5 Albedo
    sortPriority=10
    channelNames {a=Pre-Tint Exp, b=Tint B, g=Tint G, r=Tint R}
    defaultValue=<rgba 1 1 1 1>
   }
  }
  MaterialExpressionVectorParameter_61 {
   class=MaterialExpressionVectorParameter
   properties {
    group=PixieParts | Variants | Override | Material 1
    parameterName=Override - Mat1 Albedo
    sortPriority=10
    channelNames {a=Pre-Tint Exp, b=Tint B, g=Tint G, r=Tint R}
    defaultValue=<rgba 1 1 1 1>
   }
  }
  MaterialExpressionVectorParameter_62 {
   class=MaterialExpressionVectorParameter
   properties {
    group=PixieParts | Variants | Override | Material 2
    parameterName=Override - Mat2 Albedo
    sortPriority=10
    channelNames {a=Pre-Tint Exp, b=Tint B, g=Tint G, r=Tint R}
    defaultValue=<rgba 1 1 1 1>
   }
  }
  MaterialExpressionVectorParameter_63 {
   class=MaterialExpressionVectorParameter
   properties {
    group=PixieParts | Variants | Override | Material 3
    parameterName=Override - Mat3 Albedo
    sortPriority=10
    channelNames {a=Pre-Tint Exp, b=Tint B, g=Tint G, r=Tint R}
    defaultValue=<rgba 1 1 1 1>
   }
  }
  MaterialExpressionVectorParameter_64 {
   class=MaterialExpressionVectorParameter
   properties {
    group=PixieParts | Variants | Override | Material 4
    parameterName=Override - Mat4 Albedo
    sortPriority=10
    channelNames {a=Pre-Tint Exp, b=Tint B, g=Tint G, r=Tint R}
    defaultValue=<rgba 1 1 1 1>
   }
  }
  MaterialExpressionVectorParameter_65 {
   class=MaterialExpressionVectorParameter
   properties {
    group=PixieParts | Variants | Override | Material 5
    parameterName=Override - Mat5 Albedo
    sortPriority=10
    channelNames {a=Pre-Tint Exp, b=Tint B, g=Tint G, r=Tint R}
    defaultValue=<rgba 1 1 1 1>
   }
  }
  MaterialExpressionVectorParameter_66 {
   class=MaterialExpressionVectorParameter
   properties {
    group=PixieParts | Variants | Body | Energy
    parameterName=Body - Energy Color
    sortPriority=20
    channelNames {a=n/a, b=Tint B, g=Tint G, r=Tint R}
    defaultValue=<rgba 1 1 1 0>
   }
  }
  MaterialExpressionVectorParameter_67 {
   class=MaterialExpressionVectorParameter
   properties {
    group=PixieParts | Variants | Override | Energy
    parameterName=Override - Energy Color
    sortPriority=20
    channelNames {a=n/a, b=Tint B, g=Tint G, r=Tint R}
    defaultValue=<rgba 1 1 1 0>
   }
  }
  MaterialExpressionVectorParameter_68 {
   class=MaterialExpressionVectorParameter
   properties {
    group=PixieParts | Variants | Body | Material 2
    parameterName=Body - Mat2 Remaps
    sortPriority=50
    channelNames {a=AO Mix, b=Roughness Exp, g=Specular Exp, r=Metallic}
    defaultValue=<rgba 1 1 1 0>
   }
  }
  MaterialExpressionVectorParameter_69 {
   class=MaterialExpressionVectorParameter
   properties {
    group=PixieParts | Variants | Body | Material 3
    parameterName=Body - Mat3 Remaps
    sortPriority=50
    channelNames {a=AO Mix, b=Roughness Exp, g=Specular Exp, r=Metallic}
    defaultValue=<rgba 1 1 1 0>
   }
  }
  MaterialExpressionVectorParameter_7 {
   class=MaterialExpressionVectorParameter
   properties {
    group=PixieParts | Variants | Override | Energy
    parameterName=Override - Energy Color 2
    sortPriority=30
    channelNames {a=n/a, b=Tint B, g=Tint G, r=Tint R}
    defaultValue=<rgba 1 1 1 0>
   }
  }
  MaterialExpressionVectorParameter_70 {
   class=MaterialExpressionVectorParameter
   properties {
    group=PixieParts | Variants | Body | Material 4
    parameterName=Body - Mat4 Remaps
    sortPriority=50
    channelNames {a=AO Mix, b=Roughness Exp, g=Specular Exp, r=Metallic}
    defaultValue=<rgba 1 1 1 0>
   }
  }
  MaterialExpressionVectorParameter_71 {
   class=MaterialExpressionVectorParameter
   properties {
    group=PixieParts | Variants | Body | Material 5
    parameterName=Body - Mat5 Remaps
    sortPriority=50
    channelNames {a=AO Mix, b=Roughness Exp, g=Specular Exp, r=Metallic}
    defaultValue=<rgba 1 1 1 0>
   }
  }
  MaterialExpressionVectorParameter_72 {
   class=MaterialExpressionVectorParameter
   properties {
    group=PixieParts | Variants | Override | Material 1
    parameterName=Override - Mat1 Remaps
    sortPriority=50
    channelNames {a=AO Mix, b=Roughness Exp, g=Specular Exp, r=Metallic}
    defaultValue=<rgba 1 1 1 0>
   }
  }
  MaterialExpressionVectorParameter_73 {
   class=MaterialExpressionVectorParameter
   properties {
    group=PixieParts | Variants | Override | Material 2
    parameterName=Override - Mat2 Remaps
    sortPriority=50
    channelNames {a=AO Mix, b=Roughness Exp, g=Specular Exp, r=Metallic}
    defaultValue=<rgba 1 1 1 0>
   }
  }
  MaterialExpressionVectorParameter_74 {
   class=MaterialExpressionVectorParameter
   properties {
    group=PixieParts | Variants | Override | Material 3
    parameterName=Override - Mat3 Remaps
    sortPriority=50
    channelNames {a=AO Mix, b=Roughness Exp, g=Specular Exp, r=Metallic}
    defaultValue=<rgba 1 1 1 0>
   }
  }
  MaterialExpressionVectorParameter_75 {
   class=MaterialExpressionVectorParameter
   properties {
    group=PixieParts | Variants | Override | Material 4
    parameterName=Override - Mat4 Remaps
    sortPriority=50
    channelNames {a=AO Mix, b=Roughness Exp, g=Specular Exp, r=Metallic}
    defaultValue=<rgba 1 1 1 0>
   }
  }
  MaterialExpressionVectorParameter_76 {
   class=MaterialExpressionVectorParameter
   properties {
    group=PixieParts | Variants | Override | Material 5
    parameterName=Override - Mat5 Remaps
    sortPriority=50
    channelNames {a=AO Mix, b=Roughness Exp, g=Specular Exp, r=Metallic}
    defaultValue=<rgba 1 1 1 0>
   }
  }
 }}
