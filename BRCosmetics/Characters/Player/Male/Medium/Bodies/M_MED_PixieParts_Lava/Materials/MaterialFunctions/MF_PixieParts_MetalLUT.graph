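# Material function that applies a view-dependent metal LUT tint to the PixieParts Lava character body material, toggled by the "Use Metal LUT" static switch.
# The saturated camera-to-pixel-normal dot product samples a reflectance curve atlas row, and a selected RGBA channel multiplies BaseColor.
# The incoming Tangent attribute is repurposed to carry LUT intensity and channel index.
# Intensity blends between material attribute sets for the output.
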MaterialFunction /BRCosmetics/Characters/Player/Male/Medium/Bodies/M_MED_PixieParts_Lava/Materials/MaterialFunctions/MF_PixieParts_MetalLUT
graph {
 properties {libraryCategoriesText []}
 nodes [
  n1 = NamedRerouteUsage()
  n2 = NamedRerouteUsage()
  n3 = Saturate(None=NamedRerouteUsage())
  FunctionOutput(None=StaticSwitchParameter(True=BlendMaterialAttributes(A=@n1, B=@n2, Alpha=@n3) {vertexAttributeBlendType=UseA}, False=NamedRerouteUsage()) {group=PixieParts | Globals | Metal LUT, parameterName=Use Metal LUT, sortPriority=0}) {outputName="Material Out (MA)"}
  MaterialExpressionGetMaterialAttributes_2 = GetMaterialAttributes(None=NamedRerouteDeclaration(None=FunctionInput() {inputName=Material In, inputType=FunctionInput_MaterialAttributes}) {name="Material In (MA)", nodeColor=<rgba 0.4531 1 0 1>}) {attributeGetTypes [#1, #2]}
  n4 = Convert(None=@MaterialExpressionGetMaterialAttributes_2.Tangent) {
   convertInputs [{
     type=Vector3
     expressionInput {
      expression=&self:MaterialExpressionGetMaterialAttributes_2
      inputName=None
      mask=0
      maskA=0
      maskB=0
      maskG=0
      maskR=0
      outputIndex=2
     }
    }]
   convertOutputs [{type=Scalar}, {type=Scalar}]
  }
  NamedRerouteDeclaration(None=@n4) {name="Metal LUT Intensity (S)", nodeColor=<rgba 0 0.9062 1 1>}
  NamedRerouteDeclaration(None=@MaterialExpressionGetMaterialAttributes_2.BaseColor) {name="BaseColor In (V3)", nodeColor=<rgba 1 0 0.9913 1>}
  NamedRerouteDeclaration(None=@n4) {name="Metal LUT Channel (S)", nodeColor=<rgba 0 0.9062 1 1>}
  n5 = NamedRerouteUsage()
  n6 = NamedRerouteUsage()
  n7 = CameraVectorWS()
  n8 = PixelNormalWS()
  MaterialExpressionCurveAtlasRowParameter_1 = CurveAtlasRowParameter(CurveTime=Saturate(None=DotProduct(A=@n7, B=@n8))) {
   atlas=/BRCosmetics/Characters/Player/Male/Medium/Bodies/M_MED_PixieParts_Lava/Materials/CRV_PixieParts_Atlas
   curve=/BRCosmetics/Characters/Player/Male/Medium/Bodies/M_MED_PixieParts_Lava/Materials/CRV_PixieParts_Reflectance
   group=PixieParts | Globals | Metal LUT
   parameterName=Metal LUT Curve
   sortPriority=10
  }
  n9 = Switch!(
   SwitchValue=Round(None=Clamp(None=@n6) {maxDefault=3})
   Default
   R=@MaterialExpressionCurveAtlasRowParameter_1.R
   G=@MaterialExpressionCurveAtlasRowParameter_1.R
   B=@MaterialExpressionCurveAtlasRowParameter_1.R
   A=@MaterialExpressionCurveAtlasRowParameter_1.R
  ) {inputs [
    {
     inputName=R
     input {
      expression=&self:MaterialExpressionCurveAtlasRowParameter_1
      inputName=None
      mask=1
      maskA=0
      maskB=0
      maskG=0
      maskR=1
      outputIndex=1
     }
    }
    {
     inputName=G
     input {
      expression=&self:MaterialExpressionCurveAtlasRowParameter_1
      inputName=None
      mask=1
      maskA=0
      maskB=0
      maskG=1
      maskR=0
      outputIndex=2
     }
    }
    {
     inputName=B
     input {
      expression=&self:MaterialExpressionCurveAtlasRowParameter_1
      inputName=None
      mask=1
      maskA=0
      maskB=1
      maskG=0
      maskR=0
      outputIndex=3
     }
    }
    {
     inputName=A
     input {
      expression=&self:MaterialExpressionCurveAtlasRowParameter_1
      inputName=None
      mask=1
      maskA=1
      maskB=0
      maskG=0
      maskR=0
      outputIndex=4
     }
    }
   ]}
  NamedRerouteDeclaration(None=SetMaterialAttributes!(MaterialAttributes=NamedRerouteUsage(), BaseColor=Multiply(A=@n5, B=@n9)) {attributeSetTypes [#1]}) {name="Material Out (MA)", nodeColor=<rgba 0 1 0.0859 1>}
 ]
}
