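# Material function that builds the tiled moss layer for the PixieParts Lava cosmetic body, reusing the Clay variant's moss textures.
# Two packed textures supply two-channel color with derived blue, roughness, specular and a normal blended over the incoming material's normal.
# A quality switch drops normal and specular detail on low tiers, then SimpleMaterial applies parameterized remaps.
# Unused attribute slots carry disabled clear-coat, metal and thin-film defaults; the exponent-adjusted moss mask outputs separately.
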MaterialFunction /BRCosmetics/Characters/Player/Male/Medium/Bodies/M_MED_PixieParts_Lava/Materials/MaterialFunctions/MF_PixieParts_MossMaterial
graph {expressions {
  MaterialExpressionAppendVector_0 {
   class=MaterialExpressionAppendVector
   inputs [
    A {from=MaterialExpressionTextureSample_1, output=R}
    B {from=MaterialExpressionTextureSample_1, output=R}
   ]
  }
  MaterialExpressionConstant_2 {class=MaterialExpressionConstant}
  MaterialExpressionConstant_3 {class=MaterialExpressionConstant, properties {r=1}}
  MaterialExpressionConstant_4 {class=MaterialExpressionConstant, properties {r=0.5}}
  MaterialExpressionConvert_0 {
   class=MaterialExpressionConvert
   inputs [
    None {from=MaterialExpressionTextureSample_0, output=R}
    None {from=MaterialExpressionTextureSample_0, output=R}
    None {from=MaterialExpressionMultiply_0}
   ]
   properties {
    nodeName=Make Vector3
    convertInputs [
     {
      type=Scalar
      expressionInput {
       expression=&self:MaterialExpressionTextureSample_0
       inputName=None
       mask=1
       maskA=0
       maskB=0
       maskG=0
       maskR=1
       outputIndex=1
      }
     }
     {
      type=Scalar
      expressionInput {
       expression=&self:MaterialExpressionTextureSample_0
       inputName=None
       mask=1
       maskA=0
       maskB=0
       maskG=1
       maskR=0
       outputIndex=2
      }
     }
     {
      type=Scalar
      expressionInput {
       expression=&self:MaterialExpressionMultiply_0
       inputName=None
       mask=0
       maskA=0
       maskB=0
       maskG=0
       maskR=0
       outputIndex=0
      }
     }
    ]
    convertOutputs [{type=Vector3}]
   }
  }
  MaterialExpressionConvert_1 {class=MaterialExpressionConvert, inputs [], properties {convertOutputs [{type=Vector3, defaultValue=<rgba 0 0 1 1>}]}}
  MaterialExpressionConvert_2 {
   class=MaterialExpressionConvert
   inputs []
   properties {desc=disabled, convertOutputs [{type=Scalar}, {type=Vector3, defaultValue=<rgba 1 0.1 0 1>}]}
  }
  MaterialExpressionConvert_3 {class=MaterialExpressionConvert, inputs [], properties {desc=disabled, convertOutputs [{type=Vector3}]}}
  MaterialExpressionConvert_4 {class=MaterialExpressionConvert, inputs [], properties {desc=disabled, convertOutputs [{type=Vector3, defaultValue=<rgba 0 2 0 1>}]}}
  MaterialExpressionConvert_5 {class=MaterialExpressionConvert, inputs [], properties {desc=disabled, convertOutputs [{type=Vector3, defaultValue=<rgba 0.5 1 0 1>}]}}
  MaterialExpressionDeriveNormalZ_0 {class=MaterialExpressionDeriveNormalZ, inputs [InXY {from=MaterialExpressionAppendVector_0, output=""}], properties {}}
  MaterialExpressionDivide_0 {
   class=MaterialExpressionDivide
   inputs [
    A {from=MaterialExpressionTextureCoordinate_0}
    B {from=MaterialExpressionNamedRerouteUsage_8}
   ]
  }
  MaterialExpressionFunctionInput_0 {
   class=MaterialExpressionFunctionInput
   inputs [Preview {from="", output=""}]
   properties {bUsePreviewValueAsDefault=true, inputName=Moss Mask In, inputType=FunctionInput_Scalar, sortPriority=10, previewValue=<1 0 0 1>}
  }
  MaterialExpressionFunctionInput_5 {
   class=MaterialExpressionFunctionInput
   inputs [Preview {from="", output=""}]
   properties {inputName=Material In, inputType=FunctionInput_MaterialAttributes}
  }
  MaterialExpressionFunctionOutput_0 {
   class=MaterialExpressionFunctionOutput
   inputs [None {from=MaterialExpressionSetMaterialAttributes_3, output=""}]
   properties {outputName="Material Out (MA)"}
  }
  MaterialExpressionFunctionOutput_1 {
   class=MaterialExpressionFunctionOutput
   inputs [None {from=MaterialExpressionNamedRerouteUsage_10, output=""}]
   properties {outputName="Moss Mask Out (MA)", sortPriority=10}
  }
  MaterialExpressionGetMaterialAttributes_0 {
   class=MaterialExpressionGetMaterialAttributes
   inputs [None {from=MaterialExpressionNamedRerouteUsage_11, output=""}]
   properties {attributeGetTypes [#1]}
  }
  MaterialExpressionMaterialFunctionCall_0 {
   class=MaterialExpressionMaterialFunctionCall
   inputs [
    Material In {from=MaterialExpressionMaterialFunctionCall_3, output=Result}
    Albedo Remaps {from=MaterialExpressionNamedRerouteUsage_6}
    Tint {from=""}
    Metallic {from=""}
    Specular Remaps {from=MaterialExpressionNamedRerouteUsage_4}
    Roughness Remaps {from=MaterialExpressionNamedRerouteUsage_5}
   ]
   properties {materialFunction=&/MF_PixieParts_SimpleMaterial}
  }
  MaterialExpressionMaterialFunctionCall_1 {
   class=MaterialExpressionMaterialFunctionCall
   inputs [
    Normal {from=MaterialExpressionDeriveNormalZ_0}
    Intensity {from=MaterialExpressionNamedRerouteUsage_7}
   ]
   properties {materialFunction=/Game/Global/MaterialFunctions/MF_NormalIntensity}
  }
  MaterialExpressionMaterialFunctionCall_2 {
   class=MaterialExpressionMaterialFunctionCall
   inputs [
    BaseNormal {from=MaterialExpressionGetMaterialAttributes_0, output=Normal}
    AdditionalNormal {from=MaterialExpressionMaterialFunctionCall_1, output="Result (V3)"}
   ]
   properties {materialFunction=&BlendAngleCorrectedNormals}
  }
  MaterialExpressionMaterialFunctionCall_3 {
   class=MaterialExpressionMaterialFunctionCall
   inputs [
    Low _ PC Low and Med {from=MaterialExpressionReroute_0}
    "High _ XB1, PS4" {from=MaterialExpressionSetMaterialAttributes_2}
    "Epic _ PC High, XSX, PS5" {from=MaterialExpressionSetMaterialAttributes_2}
    -------- {from=""}
    "Mobile Low _ PC Perf, Android" {from=MaterialExpressionReroute_0}
    Mobile Med _ Switch {from=MaterialExpressionSetMaterialAttributes_2}
   ]
   properties {materialFunction=/Game/Global/MaterialFunctions/TechArt/MF_Fortnite_QualitySwitch_MaterialAttributes}
  }
  MaterialExpressionMultiply_0 {
   class=MaterialExpressionMultiply
   inputs [
    A {from=MaterialExpressionTextureSample_0, output=R}
    B {from=""}
   ]
   properties {constB=0.3}
  }
  MaterialExpressionNamedRerouteDeclaration_0 {
   class=MaterialExpressionNamedRerouteDeclaration
   inputs [None {from=MaterialExpressionTextureSample_1, output=B}]
   properties {name="Specular (S)", nodeColor=<rgba 0.2188 0 1 1>}
  }
  MaterialExpressionNamedRerouteDeclaration_1 {
   class=MaterialExpressionNamedRerouteDeclaration
   inputs [None {from=MaterialExpressionMaterialFunctionCall_2, output=Result}]
   properties {name="Normal TS (V3)", nodeColor=<rgba 0.1016 1 0 1>}
  }
  MaterialExpressionNamedRerouteDeclaration_119 {
   class=MaterialExpressionNamedRerouteDeclaration
   inputs [None {from=MaterialExpressionConvert_2}]
   properties {name="Clear Coat ##1 (S)", nodeColor=<rgba 1 0 0.5859 1>}
  }
  MaterialExpressionNamedRerouteDeclaration_139 {
   class=MaterialExpressionNamedRerouteDeclaration
   inputs [None {from=MaterialExpressionConvert_2}]
   properties {name="Clear Coat ##2 (V3)", nodeColor=<rgba 1 0 0.5859 1>}
  }
  MaterialExpressionNamedRerouteDeclaration_190 {
   class=MaterialExpressionNamedRerouteDeclaration
   inputs [None {from=MaterialExpressionFunctionInput_5}]
   properties {name="Material In (MA)", nodeColor=<rgba 0.4531 1 0 1>}
  }
  MaterialExpressionNamedRerouteDeclaration_2 {
   class=MaterialExpressionNamedRerouteDeclaration
   inputs [None {from=MaterialExpressionTextureSample_0, output=B}]
   properties {name="Roughness (S)", desc=Encoded as sRGB when packed, nodeColor=<rgba 1 0 0.4219 1>}
  }
  MaterialExpressionNamedRerouteDeclaration_20 {
   class=MaterialExpressionNamedRerouteDeclaration
   inputs [None {from=MaterialExpressionConvert_3}]
   properties {name="Metal LUT (V3)", nodeColor=<rgba 0 0.9062 1 1>}
  }
  MaterialExpressionNamedRerouteDeclaration_23 {
   class=MaterialExpressionNamedRerouteDeclaration
   inputs [None {from=MaterialExpressionConvert_4}]
   properties {name="Thin Film ##1 (V3)", nodeColor=<rgba 0 1 0.5043 1>}
  }
  MaterialExpressionNamedRerouteDeclaration_24 {
   class=MaterialExpressionNamedRerouteDeclaration
   inputs [None {from=MaterialExpressionConvert_5}]
   properties {name="Thin Film ##2 (V3)", nodeColor=<rgba 0.2957 1 0 1>}
  }
  MaterialExpressionNamedRerouteDeclaration_3 {
   class=MaterialExpressionNamedRerouteDeclaration
   inputs [None {from=MaterialExpressionConvert_0}]
   properties {name="Base Color (V3)", nodeColor=<rgba 0 0.7422 1 1>}
  }
  MaterialExpressionNamedRerouteDeclaration_4 {
   class=MaterialExpressionNamedRerouteDeclaration
   inputs [None {from=MaterialExpressionVectorParameter_0, output=Tiles UV Scale}]
   properties {name="UV Scale (S)", nodeColor=<rgba 0.7813 0 1 1>}
  }
  MaterialExpressionNamedRerouteDeclaration_5 {
   class=MaterialExpressionNamedRerouteDeclaration
   inputs [None {from=MaterialExpressionVectorParameter_0, output=Normal Intensity}]
   properties {name="Normal Intensity (S)", nodeColor=<rgba 0 1 0.4609 1>}
  }
  MaterialExpressionNamedRerouteDeclaration_6 {
   class=MaterialExpressionNamedRerouteDeclaration
   inputs [None {from=MaterialExpressionVectorParameter_0, output=Moss Mask Exp}]
   properties {name="Moss Mask Exp (S)", nodeColor=<rgba 1 0.1406 0 1>}
  }
  MaterialExpressionNamedRerouteDeclaration_67 {
   class=MaterialExpressionNamedRerouteDeclaration
   inputs [None {from=MaterialExpressionVectorParameter_30, output=RGBA}]
   properties {name="Albedo Remap (V4)", nodeColor=<rgba 0.2573 1 0.0816 1>}
  }
  MaterialExpressionNamedRerouteDeclaration_68 {
   class=MaterialExpressionNamedRerouteDeclaration
   inputs [None {from=MaterialExpressionVectorParameter_28, output=RGB}]
   properties {name="Roughness Remap (V3)", nodeColor=<rgba 0.2573 1 0.0816 1>}
  }
  MaterialExpressionNamedRerouteDeclaration_69 {
   class=MaterialExpressionNamedRerouteDeclaration
   inputs [None {from=MaterialExpressionVectorParameter_29, output=RGB}]
   properties {name="Specular Remap (V3)", nodeColor=<rgba 0.2573 1 0.0816 1>}
  }
  MaterialExpressionNamedRerouteDeclaration_7 {
   class=MaterialExpressionNamedRerouteDeclaration
   inputs [None {from=MaterialExpressionPower_0}]
   properties {name="Moss Mask (S)", nodeColor=<rgba 0 0.1797 1 1>}
  }
  MaterialExpressionNamedRerouteUsage_10 {class=MaterialExpressionNamedRerouteUsage}
  MaterialExpressionNamedRerouteUsage_11 {class=MaterialExpressionNamedRerouteUsage}
  MaterialExpressionNamedRerouteUsage_12 {class=MaterialExpressionNamedRerouteUsage}
  MaterialExpressionNamedRerouteUsage_13 {class=MaterialExpressionNamedRerouteUsage}
  MaterialExpressionNamedRerouteUsage_14 {class=MaterialExpressionNamedRerouteUsage}
  MaterialExpressionNamedRerouteUsage_15 {class=MaterialExpressionNamedRerouteUsage}
  MaterialExpressionNamedRerouteUsage_4 {class=MaterialExpressionNamedRerouteUsage}
  MaterialExpressionNamedRerouteUsage_5 {class=MaterialExpressionNamedRerouteUsage}
  MaterialExpressionNamedRerouteUsage_6 {class=MaterialExpressionNamedRerouteUsage}
  MaterialExpressionNamedRerouteUsage_7 {class=MaterialExpressionNamedRerouteUsage}
  MaterialExpressionNamedRerouteUsage_8 {class=MaterialExpressionNamedRerouteUsage}
  MaterialExpressionNamedRerouteUsage_9 {class=MaterialExpressionNamedRerouteUsage}
  MaterialExpressionPower_0 {
   class=MaterialExpressionPower
   inputs [
    Base {from=MaterialExpressionFunctionInput_0}
    Exp {from=MaterialExpressionNamedRerouteUsage_9}
   ]
  }
  MaterialExpressionReroute_0 {class=MaterialExpressionReroute, inputs [None {from=MaterialExpressionSetMaterialAttributes_1}]}
  MaterialExpressionSetMaterialAttributes_1 {
   class=MaterialExpressionSetMaterialAttributes
   inputs [
    MaterialAttributes {from="", output=""}
    BaseColor {from=MaterialExpressionNamedRerouteUsage_12, output=""}
    Metallic {from=MaterialExpressionConstant_2, output=""}
    Specular {from=MaterialExpressionConstant_4, output=""}
    Roughness {from=MaterialExpressionNamedRerouteUsage_15, output=""}
    EmissiveColor {from=MaterialExpressionConstant_2, output=""}
    Opacity {from=MaterialExpressionConstant_3, output=""}
    OpacityMask {from=MaterialExpressionConstant_3, output=""}
    Normal {from=MaterialExpressionConvert_1, output=""}
    ShadingModel {from=MaterialExpressionShadingModel_1, output=""}
   ]
   properties {attributeSetTypes [
     #2
     #3
     #4
     #5
     #6
     #7
     #8
     #1
     #9
    ]}
  }
  MaterialExpressionSetMaterialAttributes_2 {
   class=MaterialExpressionSetMaterialAttributes
   inputs [
    MaterialAttributes {from=MaterialExpressionSetMaterialAttributes_1, output=""}
    Specular {from=MaterialExpressionNamedRerouteUsage_13, output=""}
    Normal {from=MaterialExpressionNamedRerouteUsage_14, output=""}
   ]
   properties {attributeSetTypes [#4, #1]}
  }
  MaterialExpressionSetMaterialAttributes_3 {
   class=MaterialExpressionSetMaterialAttributes
   inputs [
    MaterialAttributes {from=MaterialExpressionMaterialFunctionCall_0, output="Material Out (MA)"}
    Tangent {from=MaterialExpressionNamedRerouteDeclaration_20, output=""}
    Displacement {from=MaterialExpressionNamedRerouteDeclaration_119, output=""}
    SubsurfaceColor {from=MaterialExpressionNamedRerouteDeclaration_139, output=""}
    BentNormal {from=MaterialExpressionNamedRerouteDeclaration_23, output=""}
    TransmittanceColor {from=MaterialExpressionNamedRerouteDeclaration_24, output=""}
   ]
   properties {attributeSetTypes [
     #10
     #11
     #12
     #13
     #14
    ]}
  }
  MaterialExpressionShadingModel_1 {class=MaterialExpressionShadingModel, inputs [], properties {}}
  MaterialExpressionTextureCoordinate_0 {class=MaterialExpressionTextureCoordinate}
  MaterialExpressionTextureSample_0 {
   class=MaterialExpressionTextureSample
   inputs [
    UVs {from=MaterialExpressionDivide_0}
    Tex {from=""}
    Apply View MipBias {from=""}
   ]
   properties {
    samplerSource=SSM_Wrap_WorldGroupSettings
    texture=/BRCosmetics/Characters/Player/Male/Medium/Bodies/M_MED_PixieParts_Clay/Textures/T_PixieParts_Clay_Moss_DS
   }
  }
  MaterialExpressionTextureSample_1 {
   class=MaterialExpressionTextureSample
   inputs [
    UVs {from=MaterialExpressionDivide_0}
    Tex {from=""}
    Apply View MipBias {from=""}
   ]
   properties {
    samplerSource=SSM_Wrap_WorldGroupSettings
    samplerType=SAMPLERTYPE_LinearColor
    texture=/BRCosmetics/Characters/Player/Male/Medium/Bodies/M_MED_PixieParts_Clay/Textures/T_PixieParts_Clay_Moss_NR
   }
  }
  MaterialExpressionVectorParameter_0 {
   class=MaterialExpressionVectorParameter
   properties {
    group=PixieParts | Materials | Moss
    parameterName=Moss Map Controls
    sortPriority=100
    channelNames {a=n/a, b=Normal Intensity, g=Tiles UV Scale, r=Moss Mask Exp}
    defaultValue=<rgba 1 1 1 0>
   }
  }
  MaterialExpressionVectorParameter_28 {
   class=MaterialExpressionVectorParameter
   properties {
    group=PixieParts | Materials | Moss
    parameterName=Moss Roughness Remap
    sortPriority=130
    channelNames {a=n/a, b=Mult, g=S-Curve Exp, r=Offset}
    defaultValue=<rgba 0 1 1 0>
   }
  }
  MaterialExpressionVectorParameter_29 {
   class=MaterialExpressionVectorParameter
   properties {
    group=PixieParts | Materials | Moss
    parameterName=Moss Specular Remap
    sortPriority=120
    channelNames {a=n/a, b=Mult, g=S-Curve Exp, r=Offset}
    defaultValue=<rgba 0 1 1 0>
   }
  }
  MaterialExpressionVectorParameter_30 {
   class=MaterialExpressionVectorParameter
   properties {
    group=PixieParts | Materials | Moss
    parameterName=Moss Albedo Remap
    sortPriority=110
    channelNames {a=Gain, b=Exp, g=S-Curve Exp, r=Desat}
    defaultValue=<rgba 0 1 1 1>
   }
  }
 }}
