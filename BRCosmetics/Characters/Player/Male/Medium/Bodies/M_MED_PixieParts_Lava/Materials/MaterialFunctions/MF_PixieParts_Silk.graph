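# Material function for PixieParts pearl materials on a character cosmetic body, adding a silk look to incoming material attributes.
# A parameterized Fresnel term drives metallic, edge-tinted base color and anisotropy-scaled roughness, faking a silk sheen.
# The result also sets subsurface color, custom data and shading model.
# The "Use Silk" static switch selects an intensity-weighted attribute blend, with the Alpha input presumably as intensity, or bypasses the effect.
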MaterialFunction /BRCosmetics/Characters/Player/Male/Medium/Bodies/M_MED_PixieParts_Lava/Materials/MaterialFunctions/MF_PixieParts_Silk
graph {nodes [
  n1 = NamedRerouteUsage()
  n2 = NamedRerouteUsage()
  n3 = NamedRerouteUsage()
  FunctionOutput(None=StaticSwitchParameter(True=BlendMaterialAttributes(A=@n1, B=@n2, Alpha=@n3) {vertexAttributeBlendType=UseA}, False=NamedRerouteUsage()) {group=PixieParts | Materials | Pearl | Silk, parameterName=Use Silk, sortPriority=0}) {outputName="Material Out (MA)"}
  NamedRerouteDeclaration(None=FunctionInput() {bUsePreviewValueAsDefault=true, inputName=Alpha, inputType=FunctionInput_Scalar, sortPriority=10}) {name="Silk Intensity (S)", nodeColor=<rgba 1 0 0.5859 1>}
  n4 = GetMaterialAttributes(None=NamedRerouteDeclaration(None=FunctionInput() {inputName=Material In, inputType=FunctionInput_MaterialAttributes}) {name="Material In (MA)", nodeColor=<rgba 0.4531 1 0 1>}) {attributeGetTypes [#1, #2]}
  NamedRerouteDeclaration(None=@n4.Roughness) {name="Roughness In (S)", nodeColor=<rgba 0 0.3906 1 1>}
  NamedRerouteDeclaration(None=@n4.BaseColor) {name="BaseColor In (V3)", nodeColor=<rgba 1 0 0.9913 1>}
  NamedRerouteDeclaration(None=Fresnel(ExponentIn=ScalarParameter() {defaultValue=2, group=PixieParts | Materials | Pearl | Silk, parameterName=Silk - Fresnel Exp, sortPriority=10}) {baseReflectFraction=0}) {name="Silk Fresnel (S)", nodeColor=<rgba 0.5938 0 1 1>}
  n5 = SetMaterialAttributes!(
   MaterialAttributes=NamedRerouteUsage()
   BaseColor=NamedRerouteUsage()
   Metallic=NamedRerouteUsage()
   Roughness=NamedRerouteUsage()
   SubsurfaceColor=NamedRerouteUsage()
   CustomData0=Constant() {r=1}
   ShadingModel=ShadingModel()
  ) {attributeSetTypes [
    #1
    #3
    #2
    #4
    #5
    #6
   ]}
  NamedRerouteDeclaration(None=@n5) {name="Material Out (MA)", nodeColor=<rgba 0 1 0.0859 1>}
  n6 = ScalarParameter() {defaultValue=1, group=PixieParts | Materials | Pearl | Silk, parameterName=Silk - Fresnel Max, sortPriority=30}
  n7 = ScalarParameter() {group=PixieParts | Materials | Pearl | Silk, parameterName=Silk - Fresnel Min, sortPriority=20}
  n8 = NamedRerouteUsage()
  NamedRerouteDeclaration(None=Saturate(None=LinearInterpolate(A=@n6, B=@n7, Alpha=@n8))) {name="Silk Metallic (S)", nodeColor=<rgba 0 1 0.7826 1>}
  n9 = Multiply(
   A=NamedRerouteUsage()
   B=ScalarParameter() {defaultValue=1, group=PixieParts | Materials | Pearl | Silk, parameterName=Silk - Base Color Brightness, sortPriority=40}
  )
  n10 = VectorParameter() {group=PixieParts | Materials | Pearl | Silk, parameterName=Silk - Edge Tint, sortPriority=50, defaultValue=<rgba 0.1094 0.1094 0.1094 1>}
  n11 = NamedRerouteUsage()
  NamedRerouteDeclaration(None=Saturate(None=LinearInterpolate(A=@n9, B=@n10.RGB, Alpha=@n11))) {name="Silk Base Color (V3)", nodeColor=<rgba 1 0 0.9913 1>}
  n12 = NamedRerouteUsage()
  n13 = ScalarParameter() {defaultValue=0.5, group=PixieParts | Materials | Pearl | Silk, parameterName=Silk - Edge Anisotropy, sortPriority=60}
  NamedRerouteDeclaration(None=LinearInterpolate(A=@n12, B=Multiply(A=@n12, B=@n13), Alpha=NamedRerouteUsage())) {name="Silk Roughness (S)", nodeColor=<rgba 0 0.3906 1 1>}
 ]}
