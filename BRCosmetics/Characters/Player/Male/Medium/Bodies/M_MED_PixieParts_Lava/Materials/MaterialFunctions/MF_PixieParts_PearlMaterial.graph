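# Material function that builds the pearl material layer for the PixieParts character body, with parameters grouped under Pearl.
# It chains the SimpleMaterial and Silk functions, with a static switch for Subsurface Profile shading.
# Clear coat and other values are packed into repurposed attributes such as Displacement, SubsurfaceColor, Tangent and BentNormal.
# It then desaturates base color and blends toward a color-multiplied, fixed-roughness, zero-opacity variant, apparently for decals.
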MaterialFunction /BRCosmetics/Characters/Player/Male/Medium/Bodies/M_MED_PixieParts_Lava/Materials/MaterialFunctions/MF_PixieParts_PearlMaterial
graph {nodes [
  n1 = MaterialFunctionCall!(
   Material In=NamedRerouteUsage()
   Albedo Remaps=NamedRerouteUsage()
   Tint=NamedRerouteUsage()
   Metallic=NamedRerouteUsage()
   Specular Remaps=NamedRerouteUsage()
   Roughness Remaps=NamedRerouteUsage()
  ) {materialFunction=&/MF_PixieParts_SimpleMaterial}
  n2 = MaterialFunctionCall!(Material In=@n1."Material Out (MA)", Alpha=NamedRerouteUsage()) {materialFunction=&/MF_PixieParts_Silk}
  n3 = NamedRerouteUsage()
  n4 = ShadingModel() {shadingModel=MSM_SubsurfaceProfile}
  MaterialExpressionNamedRerouteUsage_10 = NamedRerouteUsage()
  n5 = Convert(None=@MaterialExpressionNamedRerouteUsage_10) {
   convertInputs [{
     type=Vector4
     defaultValue=<rgba 0 0 0 1>
     expressionInput {
      expression=&self:MaterialExpressionNamedRerouteUsage_10
      inputName=None
      mask=0
      maskA=0
      maskB=0
      maskG=0
      maskR=0
      outputIndex=0
     }
    }]
   convertOutputs [{type=Scalar}, {type=Vector3}]
  }
  n6 = SetMaterialAttributes!(
   MaterialAttributes=StaticSwitchParameter(
    True=SetMaterialAttributes!(MaterialAttributes=@n2."Material Out (MA)", Opacity=@n3, ShadingModel=@n4) {attributeSetTypes [#1, #2]}
    False=@n2."Material Out (MA)"
   ) {group=PixieParts | Materials | Pearl, parameterName=Peal Use Subsurface, sortPriority=0}
   Tangent=NamedRerouteUsage()
   Displacement=NamedRerouteDeclaration(None=@n5) {name="Clear Coat ##1 (S)", nodeColor=<rgba 1 0 0.5859 1>}
   SubsurfaceColor=NamedRerouteDeclaration(None=@n5) {name="Clear Coat ##2 (V3)", nodeColor=<rgba 1 0 0.5859 1>}
   BentNormal=NamedRerouteUsage()
   TransmittanceColor=NamedRerouteUsage()
  ) {attributeSetTypes [
    #3
    #4
    #5
    #6
    #7
   ]}
  n7 = GetMaterialAttributes(None=@n6) {attributeGetTypes [#8]}
  n8 = SetMaterialAttributes!(MaterialAttributes=@n7.MaterialAttributes, BaseColor=Desaturation(None=@n7.BaseColor, Fraction=NamedRerouteUsage())) {attributeSetTypes [#8]}
  n9 = GetMaterialAttributes(None=@n8) {attributeGetTypes [#8]}
  n10 = Multiply(A=@n9.BaseColor, B=NamedRerouteUsage())
  n11 = Constant() {r=0.2}
  n12 = Constant()
  FunctionOutput(None=BlendMaterialAttributes(
    A=@n8
    B=SetMaterialAttributes!(MaterialAttributes=@n9.MaterialAttributes, BaseColor=@n10, Roughness=@n11, Opacity=@n12) {attributeSetTypes [#8, #9, #1]}
    Alpha=NamedRerouteUsage()
   ) {vertexAttributeBlendType=UseA}) {outputName="Material Out (MA)"}
  NamedRerouteDeclaration(None=ScalarParameter() {group=PixieParts | Materials | Pearl, parameterName=Pearl - Metallic, sliderMax=1, sortPriority=50}) {name="Metallic (S)", nodeColor=<rgba 0.1679 1 0.4718 1>}
  NamedRerouteDeclaration(None=ScalarParameter() {group=PixieParts | Materials | Pearl, parameterName=Pearl - Desaturate, sliderMax=1, sliderMin=-1, sortPriority=200}) {name="Desaturate (S)", nodeColor=<rgba 0 1 0.8696 1>}
  n13 = ScalarParameter() {defaultValue=0.65, group=PixieParts | Materials | Pearl, parameterName=Pearl - Decal Gain, sliderMax=1, sortPriority=210}
  NamedRerouteDeclaration(None=@n13) {name="Decal Gain (S)", nodeColor=<rgba 1 0.375 0 1>}
  n14 = VectorParameter() {
   group=PixieParts | Materials | Pearl
   parameterName=Pearl - Clear Coat
   sortPriority=110
   channelNames {a=Top Coat Material Roughness Mix, b=Under Coat Roughness, g=Under Coat Metallic Mult, r=Clear Coat Intensity}
   defaultValue=<rgba 0 1 0.1 0>
  }
  NamedRerouteDeclaration(None=@n14.RGBA) {name="Clear Coat (V4)", nodeColor=<rgba 1 0 0.5859 1>}
  n15 = FunctionInput() {bUsePreviewValueAsDefault=true, inputName=Decals Mask In, inputType=FunctionInput_Scalar, sortPriority=20, previewValue=<0 0 0 0>}
  NamedRerouteDeclaration(None=@n15) {name="Decals Mask In (S)", nodeColor=<rgba 0 1 0.5078 1>}
  n16 = VectorParameter() {
   group=PixieParts | Materials | Pearl
   parameterName=Pearl - Metal LUT
   sortPriority=100
   channelNames {a=n/a, b=n/a, g=LUT Channel, r=LUT Intensity}
  }
  NamedRerouteDeclaration(None=@n16.RGB) {name="Metal LUT (V3)", nodeColor=<rgba 0 0.9062 1 1>}
  n17 = FunctionInput() {bUsePreviewValueAsDefault=true, inputName=Fresnel In, inputType=FunctionInput_Scalar, sortPriority=30, previewValue=<0 1 0 1>}
  NamedRerouteDeclaration(None=@n17) {name="Fresnel In (S)", nodeColor=<rgba 0 1 0.0156 1>}
  n18 = VectorParameter() {
   group=PixieParts | Materials | Pearl
   parameterName="Pearl - Thin Film ##1"
   sortPriority=120
   channelNames {a=n/a, b=Roughness Influence, g=Thin Film Exponent, r=Thin Film Intensity}
   defaultValue=<rgba 0 2 0 0>
  }
  NamedRerouteDeclaration(None=@n18.RGB) {name="Thin Film ##1 (V3)", nodeColor=<rgba 0 1 0.5043 1>}
  n19 = VectorParameter() {
   group=PixieParts | Materials | Pearl
   parameterName="Pearl - Thin Film ##2"
   sortPriority=130
   channelNames {a=n/a, b=Thin Film Offset, g=Thin Film Scale, r=Thin Film Wrap}
   defaultValue=<rgba 0.5 1 0 0>
  }
  NamedRerouteDeclaration(None=@n19.RGB) {name="Thin Film ##2 (V3)", nodeColor=<rgba 0.2957 1 0 1>}
  NamedRerouteDeclaration(None=ScalarParameter() {group=PixieParts | Materials | Pearl, parameterName=Pearl - Silk Intensity, sliderMax=1, sortPriority=70}) {name="Silk Intensity (S)", nodeColor=<rgba 1 0 0.2344 1>}
  NamedRerouteDeclaration(None=FunctionInput() {inputName=Material In, inputType=FunctionInput_MaterialAttributes}) {name="Material In (MA)", nodeColor=<rgba 0.4531 1 0 1>}
  n20 = VectorParameter() {
   group=PixieParts | Materials | Pearl
   parameterName=Pearl - Albedo Remap
   sortPriority=10
   channelNames {a=Gain, b=Exp, g=S-Curve Exp, r=Desat}
   defaultValue=<rgba 1 1 1 1>
  }
  NamedRerouteDeclaration(None=@n20.RGBA) {name="Albedo Remap (V4)", nodeColor=<rgba 0.1679 1 0.4718 1>}
  n21 = VectorParameter() {
   group=PixieParts | Materials | Pearl
   parameterName=Pearl - Roughness Remap
   sortPriority=30
   channelNames {a=n/a, b=Mult, g=S-Curve Exp, r=Offset}
   defaultValue=<rgba 0 1 1 0>
  }
  NamedRerouteDeclaration(None=@n21.RGB) {name="Roughness Remap (V3)", nodeColor=<rgba 0.1679 1 0.4718 1>}
  n22 = VectorParameter() {
   group=PixieParts | Materials | Pearl
   parameterName=Pearl - Specular Remap
   sortPriority=20
   channelNames {a=n/a, b=Mult, g=S-Curve Exp, r=Offset}
   defaultValue=<rgba 0 1 1 0>
  }
  NamedRerouteDeclaration(None=@n22.RGB) {name="Specular Remap (V3)", nodeColor=<rgba 0.1679 1 0.4718 1>}
  n23 = FunctionInput() {bUsePreviewValueAsDefault=true, inputName=GMap In, inputType=FunctionInput_Scalar, sortPriority=10, previewValue=<1 1 0 1>}
  NamedRerouteDeclaration(None=@n23) {name="GMap In (S)", nodeColor=<rgba 1 0.1875 0 1>}
  NamedRerouteDeclaration(None=ScalarParameter() {group=PixieParts | Materials | Pearl, parameterName=Pearl - Subsurface Intensity, sliderMax=1, sortPriority=60}) {name="Subsurface Intensity (S)", nodeColor=<rgba 0.1679 1 0.4718 1>}
  n24 = VectorParameter() {
   group=PixieParts | Materials | Pearl
   parameterName=Pearl - Base Color
   sortPriority=40
   channelNames {a=n/a, b="", g="", r=""}
   defaultValue=<rgba 1 1 1 1>
  }
  NamedRerouteDeclaration(None=@n24.RGB) {name="Base Color (V3)", nodeColor=<rgba 0.1679 1 0.4718 1>}
 ]}
