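# Material function for the Pixie Parts Lava body cosmetic that selects one of five input material attribute sets by a scalar Material ID.
# A Switch turns the ID into a Vector4, all zeros for Material 1 and one-hot for Materials 2 to 5, then splits it into four masks.
# Four chained BlendMaterialAttributes nodes layer the materials by mask, with vertex attributes always taken from the A side.
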MaterialFunction /BRCosmetics/Characters/Player/Male/Medium/Bodies/M_MED_PixieParts_Lava/Materials/MaterialFunctions/MF_PixieParts_SelectMaterial
graph {nodes [
  n1 = BlendMaterialAttributes(
   A=BlendMaterialAttributes(A=NamedRerouteUsage(), B=NamedRerouteUsage(), Alpha=NamedRerouteUsage()) {vertexAttributeBlendType=UseA}
   B=NamedRerouteUsage()
   Alpha=NamedRerouteUsage()
  ) {vertexAttributeBlendType=UseA}
  n2 = NamedRerouteUsage()
  n3 = NamedRerouteUsage()
  FunctionOutput(None=BlendMaterialAttributes(A=BlendMaterialAttributes(A=@n1, B=@n2, Alpha=@n3) {vertexAttributeBlendType=UseA}, B=NamedRerouteUsage(), Alpha=NamedRerouteUsage()) {vertexAttributeBlendType=UseA}) {outputName="Material Out (MA)"}
  MaterialExpressionConvert_0 = Convert!() {convertOutputs [{type=Vector4, defaultValue=<rgba 0 0 0 0>}]}
  MaterialExpressionConvert_1 = Convert() {convertOutputs [{type=Vector4, defaultValue=<rgba 1 0 0 0>}]}
  MaterialExpressionConvert_2 = Convert() {convertOutputs [{type=Vector4, defaultValue=<rgba 0 1 0 0>}]}
  MaterialExpressionConvert_3 = Convert() {convertOutputs [{type=Vector4, defaultValue=<rgba 0 0 1 0>}]}
  MaterialExpressionConvert_4 = Convert() {convertOutputs [{type=Vector4, defaultValue=<rgba 0 0 0 1>}]}
  MaterialExpressionSwitch_2 = Switch!(
   SwitchValue=NamedRerouteUsage()
   Default
   Material1=@MaterialExpressionConvert_0
   Material2=@MaterialExpressionConvert_1
   Material3=@MaterialExpressionConvert_2
   Material4=@MaterialExpressionConvert_3
   Material5=@MaterialExpressionConvert_4
  ) {inputs [
    {
     inputName=Material1
     input {
      expression=&self:MaterialExpressionConvert_0
      inputName=None
      mask=0
      maskA=0
      maskB=0
      maskG=0
      maskR=0
      outputIndex=0
     }
    }
    {
     inputName=Material2
     input {
      expression=&self:MaterialExpressionConvert_1
      inputName=None
      mask=0
      maskA=0
      maskB=0
      maskG=0
      maskR=0
      outputIndex=0
     }
    }
    {
     inputName=Material3
     input {
      expression=&self:MaterialExpressionConvert_2
      inputName=None
      mask=0
      maskA=0
      maskB=0
      maskG=0
      maskR=0
      outputIndex=0
     }
    }
    {
     inputName=Material4
     input {
      expression=&self:MaterialExpressionConvert_3
      inputName=None
      mask=0
      maskA=0
      maskB=0
      maskG=0
      maskR=0
      outputIndex=0
     }
    }
    {
     inputName=Material5
     input {
      expression=&self:MaterialExpressionConvert_4
      inputName=None
      mask=0
      maskA=0
      maskB=0
      maskG=0
      maskR=0
      outputIndex=0
     }
    }
   ]}
  n4 = Convert!(None=@MaterialExpressionSwitch_2) {
   nodeName=Break Float4
   convertInputs [{
     type=Vector4
     defaultValue=<rgba 0 0 0 1>
     expressionInput {
      expression=&self:MaterialExpressionSwitch_2
      inputName=None
      mask=0
      maskA=0
      maskB=0
      maskG=0
      maskR=0
      outputIndex=0
     }
    }]
   convertOutputs [{type=Scalar}, {type=Scalar}, {type=Scalar}, {type=Scalar}]
  }
  NamedRerouteDeclaration(None=@n4) {name="Material 2 Mask (S)", nodeColor=<rgba 0.0078 1 0 1>}
  NamedRerouteDeclaration(None=@n4) {name="Material 3 Mask (S)", nodeColor=<rgba 0.0078 1 0 1>}
  n5 = SetMaterialAttributes!(MaterialAttributes)
  NamedRerouteDeclaration(None=FunctionInput(Preview=@n5) {inputName=Material 3 In, inputType=FunctionInput_MaterialAttributes, sortPriority=20}) {name="Material 3 In (MA)", nodeColor=<rgba 1 1 1 1>}
  NamedRerouteDeclaration(None=FunctionInput(Preview=@n5) {inputName=Material 5 In, inputType=FunctionInput_MaterialAttributes, sortPriority=40}) {name="Material 5 In (MA)", nodeColor=<rgba 1 1 1 1>}
  NamedRerouteDeclaration(None=@n4) {name="Material 4 Mask (S)", nodeColor=<rgba 0.0078 1 0 1>}
  NamedRerouteDeclaration(None=@n4) {name="Material 5 Mask (S)", nodeColor=<rgba 0.0078 1 0 1>}
  NamedRerouteDeclaration(None=FunctionInput() {inputName=Material ID, inputType=FunctionInput_Scalar, sortPriority=50}) {name="Material ID (S)", nodeColor=<rgba 0.0078 1 0 1>}
  NamedRerouteDeclaration(None=FunctionInput(Preview=@n5) {inputName=Material 1 In, inputType=FunctionInput_MaterialAttributes}) {name="Material 1 In (MA)", nodeColor=<rgba 1 1 1 1>}
  NamedRerouteDeclaration(None=FunctionInput(Preview=@n5) {inputName=Material 2 In, inputType=FunctionInput_MaterialAttributes, sortPriority=10}) {name="Material 2 In (MA)", nodeColor=<rgba 1 1 1 1>}
  NamedRerouteDeclaration(None=FunctionInput(Preview=@n5) {inputName=Material 4 In, inputType=FunctionInput_MaterialAttributes, sortPriority=30}) {name="Material 4 In (MA)", nodeColor=<rgba 1 1 1 1>}
 ]}
