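# Material function for the PixieParts Lava player body that swizzles materials, producing five Material Attributes outputs from five regular and four special material inputs.
# Five MF_PixieParts_SelectMaterial calls each choose among five candidates by a scalar ID clamped to at most 4, one call per output slot.
# The five ID inputs default to consecutive indices from zero.
# Named reroutes carry the inputs, IDs and results between nodes.
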MaterialFunction /BRCosmetics/Characters/Player/Male/Medium/Bodies/M_MED_PixieParts_Lava/Materials/MaterialFunctions/MF_PixieParts_SwizzleSpecialMaterials
graph {nodes [
  FunctionOutput(None=NamedRerouteUsage()) {outputName="Material 1 Out (MA)"}
  FunctionOutput(None=NamedRerouteUsage()) {outputName="Material 2 Out (MA)", sortPriority=10}
  FunctionOutput(None=NamedRerouteUsage()) {outputName="Material 3 Out (MA)", sortPriority=20}
  FunctionOutput(None=NamedRerouteUsage()) {outputName="Material 4 Out (MA)", sortPriority=30}
  FunctionOutput(None=NamedRerouteUsage()) {outputName="Material 5 Out (MA)", sortPriority=40}
  n1 = FunctionInput() {bUsePreviewValueAsDefault=true, inputName=Material 2 ID, inputType=FunctionInput_Scalar, sortPriority=210, previewValue=<1 0 0 1>}
  NamedRerouteDeclaration(None=@n1) {name="Material 2 ID (S)", nodeColor=<rgba 0.0174 1 0 1>}
  n2 = FunctionInput() {bUsePreviewValueAsDefault=true, inputName=Material 3 ID, inputType=FunctionInput_Scalar, sortPriority=220, previewValue=<2 0 0 1>}
  NamedRerouteDeclaration(None=@n2) {name="Material 3 ID (S)", nodeColor=<rgba 0.0174 1 0 1>}
  n3 = SetMaterialAttributes!(MaterialAttributes)
  NamedRerouteDeclaration(None=FunctionInput(Preview=@n3) {inputName=Special Material 2 In, inputType=FunctionInput_MaterialAttributes, sortPriority=110}) {name="Special Material 2 In (MA)", nodeColor=<rgba 1 1 1 1>}
  NamedRerouteDeclaration(None=FunctionInput(Preview=@n3) {inputName=Special Material 3 In, inputType=FunctionInput_MaterialAttributes, sortPriority=120}) {name="Special Material 3 In (MA)", nodeColor=<rgba 1 1 1 1>}
  NamedRerouteDeclaration(None=FunctionInput(Preview=@n3) {inputName=Material 3 In, inputType=FunctionInput_MaterialAttributes, sortPriority=20}) {name="Material 3 In (MA)", nodeColor=<rgba 1 1 1 1>}
  NamedRerouteDeclaration(None=FunctionInput(Preview=@n3) {inputName=Material 5 In, inputType=FunctionInput_MaterialAttributes, sortPriority=40}) {name="Material 5 In (MA)", nodeColor=<rgba 1 1 1 1>}
  NamedRerouteDeclaration(None=FunctionInput(Preview=@n3) {inputName=Special Material 4 In, inputType=FunctionInput_MaterialAttributes, sortPriority=130}) {name="Special Material 4 In (MA)", nodeColor=<rgba 1 1 1 1>}
  n4 = FunctionInput() {bUsePreviewValueAsDefault=true, inputName=Material 4 ID, inputType=FunctionInput_Scalar, sortPriority=230, previewValue=<3 0 0 1>}
  NamedRerouteDeclaration(None=@n4) {name="Material 4 ID (S)", nodeColor=<rgba 0.0174 1 0 1>}
  n5 = FunctionInput() {bUsePreviewValueAsDefault=true, inputName=Material 5 ID, inputType=FunctionInput_Scalar, sortPriority=240, previewValue=<4 0 0 1>}
  NamedRerouteDeclaration(None=@n5) {name="Material 5 ID (S)", nodeColor=<rgba 0.0174 1 0 1>}
  n6 = MaterialFunctionCall!(
   Material 1 In=NamedRerouteUsage()
   Material 2 In=NamedRerouteUsage()
   Material 3 In=NamedRerouteUsage()
   Material 4 In=NamedRerouteUsage()
   Material 5 In=NamedRerouteUsage()
   Material ID=Clamp(None=NamedRerouteUsage()) {maxDefault=4}
  ) {materialFunction=&/MF_PixieParts_SelectMaterial}
  NamedRerouteDeclaration(None=@n6."Material Out (MA)") {name="Material 1 Out (MA)", nodeColor=<rgba 1 1 1 1>}
  NamedRerouteDeclaration(None=FunctionInput() {bUsePreviewValueAsDefault=true, inputName=Material 1 ID, inputType=FunctionInput_Scalar, sortPriority=200}) {name="Material 1 ID (S)", nodeColor=<rgba 0.0174 1 0 1>}
  n7 = MaterialFunctionCall(
   Material 1 In=NamedRerouteUsage()
   Material 2 In=NamedRerouteUsage()
   Material 3 In=NamedRerouteUsage()
   Material 4 In=NamedRerouteUsage()
   Material 5 In=NamedRerouteUsage()
   Material ID=Clamp(None=NamedRerouteUsage()) {maxDefault=4}
  ) {materialFunction=&MF_PixieParts_SelectMaterial}
  NamedRerouteDeclaration(None=@n7."Material Out (MA)") {name="Material 2 Out (MA)", nodeColor=<rgba 1 1 1 1>}
  n8 = MaterialFunctionCall(
   Material 1 In=NamedRerouteUsage()
   Material 2 In=NamedRerouteUsage()
   Material 3 In=NamedRerouteUsage()
   Material 4 In=NamedRerouteUsage()
   Material 5 In=NamedRerouteUsage()
   Material ID=Clamp(None=NamedRerouteUsage()) {maxDefault=4}
  ) {materialFunction=&MF_PixieParts_SelectMaterial}
  NamedRerouteDeclaration(None=@n8."Material Out (MA)") {name="Material 3 Out (MA)", nodeColor=<rgba 1 1 1 1>}
  n9 = MaterialFunctionCall(
   Material 1 In=NamedRerouteUsage()
   Material 2 In=NamedRerouteUsage()
   Material 3 In=NamedRerouteUsage()
   Material 4 In=NamedRerouteUsage()
   Material 5 In=NamedRerouteUsage()
   Material ID=Clamp(None=NamedRerouteUsage()) {maxDefault=4}
  ) {materialFunction=&MF_PixieParts_SelectMaterial}
  NamedRerouteDeclaration(None=@n9."Material Out (MA)") {name="Material 5 Out (MA)", nodeColor=<rgba 1 1 1 1>}
  NamedRerouteDeclaration(None=FunctionInput(Preview=@n3) {inputName=Material 1 In, inputType=FunctionInput_MaterialAttributes}) {name="Material 1 In (MA)", nodeColor=<rgba 1 1 1 1>}
  NamedRerouteDeclaration(None=FunctionInput(Preview=@n3) {inputName=Material 2 In, inputType=FunctionInput_MaterialAttributes, sortPriority=10}) {name="Material 2 In (MA)", nodeColor=<rgba 1 1 1 1>}
  NamedRerouteDeclaration(None=FunctionInput(Preview=@n3) {inputName=Material 4 In, inputType=FunctionInput_MaterialAttributes, sortPriority=30}) {name="Material 4 In (MA)", nodeColor=<rgba 1 1 1 1>}
  n10 = MaterialFunctionCall(
   Material 1 In=NamedRerouteUsage()
   Material 2 In=NamedRerouteUsage()
   Material 3 In=NamedRerouteUsage()
   Material 4 In=NamedRerouteUsage()
   Material 5 In=NamedRerouteUsage()
   Material ID=Clamp(None=NamedRerouteUsage()) {maxDefault=4}
  ) {materialFunction=&MF_PixieParts_SelectMaterial}
  NamedRerouteDeclaration(None=@n10."Material Out (MA)") {name="Material 4 Out (MA)", nodeColor=<rgba 1 1 1 1>}
  NamedRerouteDeclaration(None=FunctionInput(Preview=@n3) {inputName=Special Material 1 In, inputType=FunctionInput_MaterialAttributes, sortPriority=100}) {name="Special Material 1 In (MA)", nodeColor=<rgba 1 1 1 1>}
 ]}
