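# Material function for PixieParts character materials that layers an optional thin-film iridescence effect onto BaseColor.
# Film controls (intensity, exponent, roughness influence, wrap, scale, offset) are unpacked from the incoming BentNormal and TransmittanceColor attributes.
# A power, scale and offset remap builds the film texture UV, optionally distorted by a noise texture channel.
# The result blends over BaseColor through a power-based, roughness-attenuated mask, with a Fresnel input supplied by the caller.
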MaterialFunction /BRCosmetics/Characters/Player/Male/Medium/Bodies/M_MED_PixieParts_Lava/Materials/MaterialFunctions/MF_PixieParts_ThinFilm
graph {nodes [
  FunctionOutput(None=StaticSwitchParameter(True=NamedRerouteUsage(), False=NamedRerouteUsage()) {group=PixieParts | Globals | Thin Film, parameterName=Use Thin Film, sortPriority=0}) {outputName="Material Out (MA)"}
  MaterialExpressionGetMaterialAttributes_2 = GetMaterialAttributes(None=NamedRerouteDeclaration(None=FunctionInput() {inputName=Material In, inputType=FunctionInput_MaterialAttributes}) {name="Material In (MA)", nodeColor=<rgba 0.4531 1 0 1>}) {attributeGetTypes [
    #1
    #2
    #3
    #4
   ]}
  NamedRerouteDeclaration(None=@MaterialExpressionGetMaterialAttributes_2.Roughness) {name="Roughness In (S)", nodeColor=<rgba 0 0.3906 1 1>}
  n1 = TextureSampleParameter2D(UVs=Multiply(
    A=TextureCoordinate()
    B=ScalarParameter() {defaultValue=2, group=PixieParts | Globals | Thin Film, parameterName=UvFilmOffsetTiling, sortPriority=50}
   )) {
   group=PixieParts | Globals | Thin Film
   parameterName=ThinFilmUVOffsetTexture
   samplerSource=SSM_Wrap_WorldGroupSettings
   sortPriority=30
   texture=/Game/Global/Textures/Default/Noise/3xPerlin_M
  }
  n2 = ChannelMaskParameter(None=@n1.RGB) {
   group=PixieParts | Globals | Thin Film
   maskChannel=Green
   parameterName=ThinFilmUVOffsetChannel
   sortPriority=40
   defaultValue=<rgba 0 1 0 0>
  }
  NamedRerouteDeclaration(None=Multiply(
    A=Multiply(A=@n2.R) {constB=0.1}
    B=ScalarParameter() {defaultValue=0.2, group=PixieParts | Globals | Thin Film, parameterName=ThinFilmUVOffsetIntensity, sortPriority=60}
   )) {name="Thin Film UV Offset (V2)", nodeColor=<rgba 1 0.7266 0 1>}
  n3 = Power(Base=NamedRerouteUsage(), Exp=NamedRerouteUsage())
  n4 = NamedRerouteUsage()
  NamedRerouteDeclaration(None=Add(A=Multiply(A=@n3, B=@n4), B=NamedRerouteUsage())) {name="Thin Film UV (V2)", nodeColor=<rgba 0.4062 0 1 1>}
  n5 = NamedRerouteUsage()
  n6 = NamedRerouteUsage()
  n7 = NamedRerouteUsage()
  n8 = TextureSampleParameter2D(UVs=StaticSwitchParameter(True=Add(A=@n6, B=@n7), False=@n7) {group=PixieParts | Globals | Thin Film, parameterName=Use Thin Film UV Offset, sortPriority=20}) {
   group=PixieParts | Globals | Thin Film
   parameterName=Thin Film Texture
   samplerSource=SSM_Wrap_WorldGroupSettings
   sortPriority=10
   texture=/Game/Global/Textures/Default/Thinfilm/T_F_MED_Global_FB_ThinFilm_color
  }
  n9 = Add(A=Multiply(A=@n8.RGB, B=NamedRerouteUsage()), B=NamedRerouteUsage())
  n10 = Power(Base=NamedRerouteUsage(), Exp=NamedRerouteUsage())
  n11 = LinearInterpolate(Alpha=NamedRerouteUsage()) {constA=1, constB=0.01}
  n12 = Saturate(None=LinearInterpolate(A=@n10, B=Multiply(A=@n10, B=@n11), Alpha=NamedRerouteUsage()))
  NamedRerouteDeclaration(None=SetMaterialAttributes!(MaterialAttributes=NamedRerouteUsage(), BaseColor=LinearInterpolate(A=@n5, B=@n9, Alpha=@n12)) {attributeSetTypes [#1]}) {name="Material Out (MA)", nodeColor=<rgba 0 1 0.0859 1>}
  NamedRerouteDeclaration(None=@MaterialExpressionGetMaterialAttributes_2.BaseColor) {name="BaseColor In (V3)", nodeColor=<rgba 1 0 0.9913 1>}
  n13 = Convert(None=@MaterialExpressionGetMaterialAttributes_2.BentNormal) {
   nodeName=Break Float3
   convertInputs [{
     type=Vector3
     expressionInput {
      expression=&self:MaterialExpressionGetMaterialAttributes_2
      inputName=None
      mask=0
      maskA=0
      maskB=0
      maskG=0
      maskR=0
      outputIndex=3
     }
    }]
   convertOutputs [{type=Scalar}, {type=Scalar}, {type=Scalar}]
  }
  NamedRerouteDeclaration(None=@n13) {name="Thin Film Intensity (S)", nodeColor=<rgba 0 1 0.5043 1>}
  NamedRerouteDeclaration(None=@n13) {name="Thin Film Exponent (S)", nodeColor=<rgba 0 1 0.5043 1>}
  NamedRerouteDeclaration(None=@n13) {name="Roughness Influence (S)", nodeColor=<rgba 0 1 0.5043 1>}
  n14 = Convert(None=@MaterialExpressionGetMaterialAttributes_2.TransmittanceColor) {
   nodeName=Break Float3
   convertInputs [{
     type=Vector3
     expressionInput {
      expression=&self:MaterialExpressionGetMaterialAttributes_2
      inputName=None
      mask=0
      maskA=0
      maskB=0
      maskG=0
      maskR=0
      outputIndex=4
     }
    }]
   convertOutputs [{type=Scalar}, {type=Scalar}, {type=Scalar}]
  }
  NamedRerouteDeclaration(None=@n14) {name="Thin Film Wrap (S)", nodeColor=<rgba 0.2957 1 0 1>}
  NamedRerouteDeclaration(None=@n14) {name="Thin Film Scale (S)", nodeColor=<rgba 0.2957 1 0 1>}
  NamedRerouteDeclaration(None=@n14) {name="Thin Film Offset (S)", nodeColor=<rgba 0.2957 1 0 1>}
  NamedRerouteDeclaration(None=FunctionInput() {inputName=Fresnel, inputType=FunctionInput_Scalar, sortPriority=10}) {name="Fresnel In (S)", nodeColor=<rgba 0 0.9531 1 1>}
 ]}
